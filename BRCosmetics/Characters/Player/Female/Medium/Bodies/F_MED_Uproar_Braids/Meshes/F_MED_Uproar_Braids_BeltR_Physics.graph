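# Physics asset for the right-side belt straps on the female Uproar Braids cosmetic character body.
# Capsules cover the pelvis, right thigh and two strap chains (fr and br), which branch from master bodies linked to the pelvis.
# Strap segments have tight swing and twist limits, while master and thigh joints rotate freely.
# A free-rotation constraint joins the two chain tips, closing the straps into a loop.
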PhysicsAsset /BRCosmetics/Characters/Player/Female/Medium/Bodies/F_MED_Uproar_Braids/Meshes/F_MED_Uproar_Braids_BeltR_Physics
bodies {
 dyn_straps_br_01 [{shapeType=Capsule, length=2, radius=2, shapeName=dyn_straps_br_01_capsule, center=<2.215 0 0>, rotation=<rot 90 0 0>}]
 dyn_straps_br_02 [{shapeType=Capsule, length=5.5, radius=2, shapeName=dyn_straps_br_02_capsule, center=<5.1455 0 0>, rotation=<rot 90 0 0>}]
 dyn_straps_br_03 [{shapeType=Capsule, length=4.75, radius=2, shapeName=dyn_straps_br_03_capsule, center=<4.6791 0.0001 0>, rotation=<rot 90 0 0>}]
 dyn_straps_br_04 [{shapeType=Capsule, length=0.5, radius=2.5, shapeName=dyn_straps_br_04_capsule, rotation=<rot 90 0 0>}]
 dyn_straps_br_master [{shapeType=Capsule, length=0.5, radius=0.5, shapeName=dyn_straps_br_master_capsule}]
 dyn_straps_fr_01 [{shapeType=Capsule, length=2.75, radius=1, shapeName=dyn_straps_fr_01_capsule, center=<1.7993 0 0>, rotation=<rot 90 0 0>}]
 dyn_straps_fr_02 [{shapeType=Capsule, length=1, radius=1.5, shapeName=dyn_straps_fr_02_capsule, center=<2.1998 0 0>, rotation=<rot 90 0 0>}]
 dyn_straps_fr_03 [{shapeType=Capsule, length=4, radius=2, shapeName=dyn_straps_fr_03_capsule, center=<3.8268 -0.0001 0>, rotation=<rot 90 0 0>}]
 dyn_straps_fr_04 [{shapeType=Capsule, length=7.5, radius=1.75, shapeName=dyn_straps_fr_04_capsule, center=<5.8681 0 0>, rotation=<rot 90 0 0>}]
 dyn_straps_fr_05 [{shapeType=Capsule, length=0.5, radius=2.5, shapeName=dyn_straps_fr_05_capsule, rotation=<rot 90 0 0>}]
 dyn_straps_fr_master [{shapeType=Capsule, length=0.505, radius=0.505, shapeName=dyn_straps_fr_master_capsule, rotation=<rot 0 -90 0>}]
 dyn_straps_r_master [{shapeType=Capsule, length=0.5, radius=0.5, shapeName=dyn_straps_r_master_capsule}]
 pelvis [{shapeType=Capsule, length=9.0727, radius=8.7102, shapeName=pelvis_capsule, center=<0 -0.8236 0>}]
 thigh_r [{
   shapeType=Capsule
   length=20.75
   radius=6.5
   shapeName=thigh_r_capsule
   center=<10.4286 0.0436 -0.2921>
   rotation=<rot 86.6591 71.6304 71.7991>
  }]
}
constraints [
 {
  bone1Name=thigh_r
  bone2Name=pelvis
  swing1LimitDegrees=45
  swing1Motion=Free
  swing2LimitDegrees=45
  swing2Motion=Free
  twistLimitDegrees=45
  twistMotion=Free
 }
 {
  bone1Name=dyn_straps_r_master
  bone2Name=pelvis
  swing1LimitDegrees=45
  swing1Motion=Free
  swing2LimitDegrees=45
  swing2Motion=Free
  twistLimitDegrees=45
  twistMotion=Free
 }
 {
  bone1Name=dyn_straps_fr_master
  bone2Name=dyn_straps_r_master
  swing1LimitDegrees=45
  swing1Motion=Free
  swing2LimitDegrees=45
  swing2Motion=Free
  twistLimitDegrees=45
  twistMotion=Free
 }
 {bone1Name=dyn_straps_fr_01, bone2Name=dyn_straps_fr_master, swing1LimitDegrees=15, swing2LimitDegrees=10, twistLimitDegrees=5}
 {bone1Name=dyn_straps_fr_02, bone2Name=dyn_straps_fr_01, swing1LimitDegrees=15, swing2LimitDegrees=10, twistLimitDegrees=5}
 {bone1Name=dyn_straps_fr_03, bone2Name=dyn_straps_fr_02, swing1LimitDegrees=15, swing2LimitDegrees=10, twistLimitDegrees=5}
 {bone1Name=dyn_straps_fr_04, bone2Name=dyn_straps_fr_03, swing1LimitDegrees=15, swing2LimitDegrees=15, twistLimitDegrees=5}
 {bone1Name=dyn_straps_fr_05, bone2Name=dyn_straps_fr_04, swing1LimitDegrees=15, swing2LimitDegrees=15, twistLimitDegrees=5}
 {
  bone1Name=dyn_straps_br_master
  bone2Name=dyn_straps_r_master
  swing1LimitDegrees=45
  swing1Motion=Free
  swing2LimitDegrees=45
  swing2Motion=Free
  twistLimitDegrees=45
  twistMotion=Free
 }
 {bone1Name=dyn_straps_br_01, bone2Name=dyn_straps_br_master, swing1LimitDegrees=15, swing2LimitDegrees=15, twistLimitDegrees=5}
 {bone1Name=dyn_straps_br_02, bone2Name=dyn_straps_br_01, swing1LimitDegrees=15, swing2LimitDegrees=15, twistLimitDegrees=5}
 {bone1Name=dyn_straps_br_03, bone2Name=dyn_straps_br_02, swing1LimitDegrees=15, swing2LimitDegrees=15, twistLimitDegrees=5}
 {bone1Name=dyn_straps_br_04, bone2Name=dyn_straps_br_03, swing1LimitDegrees=15, swing2LimitDegrees=15, twistLimitDegrees=5}
 {
  bone1Name=dyn_straps_fr_05
  bone2Name=dyn_straps_br_04
  swing1LimitDegrees=45
  swing1Motion=Free
  swing2LimitDegrees=45
  swing2Motion=Free
  twistLimitDegrees=45
  twistMotion=Free
 }
]
