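# Physics asset for the dynamic left boot flaps on the female medium Uproar Braids player cosmetic body.
# Small capsule bodies form three bone chains (a, b and c), each constrained back to a shared boot master body.
# Joints use tight swing and twist limits, except the last joint of chain c, which swings and twists freely.
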PhysicsAsset /BRCosmetics/Characters/Player/Female/Medium/Bodies/F_MED_Uproar_Braids/Meshes/F_MED_Uproar_Braids_FootL_Physics
bodies {
 dyn_boot_flap_a_01_l [{shapeType=Capsule, length=2, radius=2, shapeName=dyn_boot_flap_a_01_l_capsule, center=<1 0 0>, rotation=<rot 90 0 0>}]
 dyn_boot_flap_a_02_l [{shapeType=Capsule, length=2, radius=2, shapeName=dyn_boot_flap_a_02_l_capsule, center=<1 0 0>, rotation=<rot 90 0 0>}]
 dyn_boot_flap_a_03_l [{shapeType=Capsule, length=2, radius=2, shapeName=dyn_boot_flap_a_03_l_capsule, center=<1 0 0>, rotation=<rot 90 0 0>}]
 dyn_boot_flap_a_04_l [{shapeType=Capsule, length=2, radius=2, shapeName=dyn_boot_flap_a_04_l_capsule, center=<1 0 0>, rotation=<rot 90 0 0>}]
 dyn_boot_flap_b_01_l [{shapeType=Capsule, length=2, radius=2, shapeName=dyn_boot_flap_b_01_l_capsule, center=<1 0 0>, rotation=<rot 90 0 0>}]
 dyn_boot_flap_b_02_l [{shapeType=Capsule, length=2, radius=2, shapeName=dyn_boot_flap_b_02_l_capsule, center=<1 0 0>, rotation=<rot 90 0 0>}]
 dyn_boot_flap_c_01_l [{shapeType=Capsule, length=2, radius=2, shapeName=dyn_boot_flap_c_01_l_capsule, center=<1 0 0>, rotation=<rot 90 0 0>}]
 dyn_boot_flap_c_02_l [{shapeType=Capsule, length=2, radius=2, shapeName=dyn_boot_flap_c_02_l_capsule, center=<1 0 0>, rotation=<rot 90 0 0>}]
 dyn_boot_flap_c_03_l [{shapeType=Capsule, length=2, radius=2, shapeName=dyn_boot_flap_c_03_l_capsule, center=<1 0 0>, rotation=<rot 90 0 0>}]
 dyn_boot_flap_c_04_l [{shapeType=Capsule, length=2, radius=2, shapeName=dyn_boot_flap_c_04_l_capsule, center=<1 0 0>, rotation=<rot 90 0 0>}]
 dyn_boot_master_l [{shapeType=Capsule, length=0.5, radius=0.5, shapeName=dyn_boot_master_l_capsule, rotation=<rot 90 0 0>}]
}
constraints [
 {bone1Name=dyn_boot_flap_a_01_l, bone2Name=dyn_boot_master_l, swing1LimitDegrees=5, swing2LimitDegrees=5, twistLimitDegrees=2}
 {bone1Name=dyn_boot_flap_a_02_l, bone2Name=dyn_boot_flap_a_01_l, swing1LimitDegrees=5, swing2LimitDegrees=5, twistLimitDegrees=2}
 {bone1Name=dyn_boot_flap_a_03_l, bone2Name=dyn_boot_flap_a_02_l, swing1LimitDegrees=5, swing2LimitDegrees=5, twistLimitDegrees=2}
 {bone1Name=dyn_boot_flap_a_04_l, bone2Name=dyn_boot_flap_a_03_l, swing1LimitDegrees=5, swing2LimitDegrees=5, twistLimitDegrees=2}
 {bone1Name=dyn_boot_flap_b_01_l, bone2Name=dyn_boot_master_l, swing1LimitDegrees=5, swing2LimitDegrees=5, twistLimitDegrees=2}
 {bone1Name=dyn_boot_flap_b_02_l, bone2Name=dyn_boot_flap_b_01_l, swing1LimitDegrees=5, swing2LimitDegrees=5, twistLimitDegrees=2}
 {bone1Name=dyn_boot_flap_c_01_l, bone2Name=dyn_boot_master_l, swing1LimitDegrees=5, swing2LimitDegrees=5, twistLimitDegrees=2}
 {bone1Name=dyn_boot_flap_c_02_l, bone2Name=dyn_boot_flap_c_01_l, swing1LimitDegrees=5, swing2LimitDegrees=5, twistLimitDegrees=2}
 {bone1Name=dyn_boot_flap_c_03_l, bone2Name=dyn_boot_flap_c_02_l, swing1LimitDegrees=5, swing2LimitDegrees=5, twistLimitDegrees=2}
 {
  bone1Name=dyn_boot_flap_c_04_l
  bone2Name=dyn_boot_flap_c_03_l
  swing1LimitDegrees=45
  swing1Motion=Free
  swing2LimitDegrees=45
  swing2Motion=Free
  twistLimitDegrees=45
  twistMotion=Free
 }
]
