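# Physics asset for the right-foot mesh of the female medium Uproar Braids body, defining physics for its boot flaps.
# Four flap chains (a to d), two to four bones long, each hang from a small boot master body.
# Every flap bone has a same-sized capsule and is constrained to the bone before it with tight swing and twist limits, so each joint bends only slightly.
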PhysicsAsset /BRCosmetics/Characters/Player/Female/Medium/Bodies/F_MED_Uproar_Braids/Meshes/F_MED_Uproar_Braids_FootR_Physics
bodies {
 dyn_boot_flap_a_01_r [{shapeType=Capsule, length=2, radius=2, shapeName=dyn_boot_flap_a_01_r_capsule, center=<1 0 0>, rotation=<rot 90 0 0>}]
 dyn_boot_flap_a_02_r [{shapeType=Capsule, length=2, radius=2, shapeName=dyn_boot_flap_a_02_r_capsule, center=<1 0 0>, rotation=<rot 90 0 0>}]
 dyn_boot_flap_a_03_r [{shapeType=Capsule, length=2, radius=2, shapeName=dyn_boot_flap_a_03_r_capsule, center=<1 0 0>, rotation=<rot 90 0 0>}]
 dyn_boot_flap_a_04_r [{shapeType=Capsule, length=2, radius=2, shapeName=dyn_boot_flap_a_04_r_capsule, center=<1 0 0>, rotation=<rot 90 0 0>}]
 dyn_boot_flap_b_01_r [{shapeType=Capsule, length=2, radius=2, shapeName=dyn_boot_flap_b_01_r_capsule, center=<1 0 0>, rotation=<rot 90 0 0>}]
 dyn_boot_flap_b_02_r [{shapeType=Capsule, length=2, radius=2, shapeName=dyn_boot_flap_b_02_r_capsule, center=<1 0 0>, rotation=<rot 90 0 0>}]
 dyn_boot_flap_c_01_r [{shapeType=Capsule, length=2, radius=2, shapeName=dyn_boot_flap_c_01_r_capsule, center=<1 0 0>, rotation=<rot 90 0 0>}]
 dyn_boot_flap_c_02_r [{shapeType=Capsule, length=2, radius=2, shapeName=dyn_boot_flap_c_02_r_capsule, center=<1 0 0>, rotation=<rot 90 0 0>}]
 dyn_boot_flap_c_03_r [{shapeType=Capsule, length=2, radius=2, shapeName=dyn_boot_flap_c_03_r_capsule, center=<1 0 0>, rotation=<rot 90 0 0>}]
 dyn_boot_flap_c_04_r [{shapeType=Capsule, length=2, radius=2, shapeName=dyn_boot_flap_c_04_r_capsule, center=<1 0 0>, rotation=<rot 90 0 0>}]
 dyn_boot_flap_d_01_r [{shapeType=Capsule, length=2, radius=2, shapeName=dyn_boot_flap_d_01_r_capsule, center=<1 0 0>, rotation=<rot 90 0 0>}]
 dyn_boot_flap_d_02_r [{shapeType=Capsule, length=2, radius=2, shapeName=dyn_boot_flap_d_02_r_capsule, center=<1 0 0>, rotation=<rot 90 0 0>}]
 dyn_boot_flap_d_03_r [{shapeType=Capsule, length=2, radius=2, shapeName=dyn_boot_flap_d_03_r_capsule, center=<1 0 0>, rotation=<rot 90 0 0>}]
 dyn_right_boot_master_r [{shapeType=Capsule, length=0.5, radius=0.5, shapeName=dyn_right_boot_master_r_capsule, rotation=<rot 90 0 0>}]
}
constraints [
 {bone1Name=dyn_boot_flap_a_01_r, bone2Name=dyn_right_boot_master_r, swing1LimitDegrees=5, swing2LimitDegrees=5, twistLimitDegrees=2}
 {bone1Name=dyn_boot_flap_a_02_r, bone2Name=dyn_boot_flap_a_01_r, swing1LimitDegrees=5, swing2LimitDegrees=5, twistLimitDegrees=2}
 {bone1Name=dyn_boot_flap_a_03_r, bone2Name=dyn_boot_flap_a_02_r, swing1LimitDegrees=5, swing2LimitDegrees=5, twistLimitDegrees=2}
 {bone1Name=dyn_boot_flap_a_04_r, bone2Name=dyn_boot_flap_a_03_r, swing1LimitDegrees=5, swing2LimitDegrees=5, twistLimitDegrees=2}
 {bone1Name=dyn_boot_flap_b_01_r, bone2Name=dyn_right_boot_master_r, swing1LimitDegrees=5, swing2LimitDegrees=5, twistLimitDegrees=2}
 {bone1Name=dyn_boot_flap_b_02_r, bone2Name=dyn_boot_flap_b_01_r, swing1LimitDegrees=5, swing2LimitDegrees=5, twistLimitDegrees=2}
 {bone1Name=dyn_boot_flap_c_01_r, bone2Name=dyn_right_boot_master_r, swing1LimitDegrees=5, swing2LimitDegrees=5, twistLimitDegrees=2}
 {bone1Name=dyn_boot_flap_c_02_r, bone2Name=dyn_boot_flap_c_01_r, swing1LimitDegrees=5, swing2LimitDegrees=5, twistLimitDegrees=2}
 {bone1Name=dyn_boot_flap_c_03_r, bone2Name=dyn_boot_flap_c_02_r, swing1LimitDegrees=5, swing2LimitDegrees=5, twistLimitDegrees=2}
 {bone1Name=dyn_boot_flap_c_04_r, bone2Name=dyn_boot_flap_c_03_r, swing1LimitDegrees=5, swing2LimitDegrees=5, twistLimitDegrees=2}
 {bone1Name=dyn_boot_flap_d_01_r, bone2Name=dyn_right_boot_master_r, swing1LimitDegrees=5, swing2LimitDegrees=5, twistLimitDegrees=2}
 {bone1Name=dyn_boot_flap_d_02_r, bone2Name=dyn_boot_flap_d_01_r, swing1LimitDegrees=5, swing2LimitDegrees=5, twistLimitDegrees=2}
 {bone1Name=dyn_boot_flap_d_03_r, bone2Name=dyn_boot_flap_d_02_r, swing1LimitDegrees=5, swing2LimitDegrees=5, twistLimitDegrees=2}
]
