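# Physics asset for the F_MED_Uproar_Braids female body cosmetic, covering three dynamic leather chains that hang from the pelvis.
# Each chain runs from a tiny master box through three larger boxes with tight swing and twist limits, though the second and third masters rotate freely on the pelvis.
# Free-rotation cross-constraints tie the first two chains at each level.
# Pelvis and right-thigh capsules provide body shapes.
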PhysicsAsset /BRCosmetics/Characters/Player/Female/Medium/Bodies/F_MED_Uproar_Braids/Meshes/F_MED_Uproar_Braids_Physics_New
bodies {
 dyn_leather2_01 [{shapeType=Box, extentX=6, extentY=2.5, extentZ=7.5, shapeName=dyn_leather_01_box}]
 dyn_leather2_02 [{shapeType=Box, extentX=6, extentY=2.5, extentZ=7.5, shapeName=dyn_leather_01_box}]
 dyn_leather2_03 [{shapeType=Box, extentX=6, extentY=2.5, extentZ=7.5, shapeName=dyn_leather_01_box}]
 dyn_leather2_master [{shapeType=Box, extentX=0.5, extentY=0.5, extentZ=0.5, shapeName=dyn_leather_01_box}]
 dyn_leather3_01 [{shapeType=Box, extentX=6, extentY=2.5, extentZ=7.5, shapeName=dyn_leather_01_box, center=<0 1.3445 0>}]
 dyn_leather3_02 [{shapeType=Box, extentX=6, extentY=2.5, extentZ=7.5, shapeName=dyn_leather_01_box, center=<0 1.3445 0>}]
 dyn_leather3_03 [{shapeType=Box, extentX=6, extentY=2.5, extentZ=7.5, shapeName=dyn_leather3_03_box, center=<0 1.3445 0>}]
 dyn_leather3_master [{shapeType=Box, extentX=0.5, extentY=0.5, extentZ=0.5, shapeName=dyn_leather_01_box}]
 dyn_leather_01 [{shapeType=Box, extentX=6, extentY=2, extentZ=7.5, shapeName=dyn_leather_01_box}]
 dyn_leather_02 [{shapeType=Box, extentX=6, extentY=2.5, extentZ=7.5, shapeName=dyn_leather_01_box, rotation=<rot 0 0 -90>}]
 dyn_leather_03 [{shapeType=Box, extentX=6, extentY=2.5, extentZ=7.5, shapeName=dyn_leather_01_box, rotation=<rot 0 0 -90>}]
 dyn_leather_master [{shapeType=Box, extentX=0.5, extentY=0.5, extentZ=0.5, shapeName=dyn_leather_01_box}]
 pelvis [{shapeType=Capsule, length=9.0727, radius=8.7102, shapeName=pelvis_capsule, center=<0 -0.8236 0>}]
 thigh_r [{shapeType=Capsule, length=20.5, radius=6.75, shapeName=thigh_r_capsule, center=<10.4113 0 -1.011>, rotation=<rot 83.7864 179.9999 179.9999>}]
}
constraints [
 {
  bone1Name=thigh_r
  bone2Name=pelvis
  swing1LimitDegrees=45
  swing1Motion=Free
  swing2LimitDegrees=45
  swing2Motion=Free
  twistLimitDegrees=45
  twistMotion=Free
 }
 {bone1Name=dyn_leather_master, bone2Name=pelvis, swing1LimitDegrees=15, swing2LimitDegrees=15, twistLimitDegrees=5}
 {bone1Name=dyn_leather_01, bone2Name=dyn_leather_master, swing1LimitDegrees=15, swing2LimitDegrees=15, twistLimitDegrees=5}
 {bone1Name=dyn_leather_02, bone2Name=dyn_leather_01, swing1LimitDegrees=15, swing2LimitDegrees=15, twistLimitDegrees=5}
 {bone1Name=dyn_leather_03, bone2Name=dyn_leather_02, swing1LimitDegrees=15, swing2LimitDegrees=15, twistLimitDegrees=5}
 {bone1Name=dyn_leather2_01, bone2Name=dyn_leather2_master, swing1LimitDegrees=15, swing2LimitDegrees=15, twistLimitDegrees=5}
 {bone1Name=dyn_leather2_02, bone2Name=dyn_leather2_01, swing1LimitDegrees=15, swing2LimitDegrees=15, twistLimitDegrees=5}
 {bone1Name=dyn_leather2_03, bone2Name=dyn_leather2_02, swing1LimitDegrees=15, swing2LimitDegrees=15, twistLimitDegrees=5}
 {bone1Name=dyn_leather3_01, bone2Name=dyn_leather3_master, swing1LimitDegrees=15, swing2LimitDegrees=15, twistLimitDegrees=5}
 {bone1Name=dyn_leather3_02, bone2Name=dyn_leather3_01, swing1LimitDegrees=15, swing2LimitDegrees=15, twistLimitDegrees=5}
 {
  bone1Name=dyn_leather2_master
  bone2Name=pelvis
  swing1LimitDegrees=45
  swing1Motion=Free
  swing2LimitDegrees=45
  swing2Motion=Free
  twistLimitDegrees=45
  twistMotion=Free
 }
 {
  bone1Name=dyn_leather3_master
  bone2Name=pelvis
  swing1LimitDegrees=45
  swing1Motion=Free
  swing2LimitDegrees=45
  swing2Motion=Free
  twistLimitDegrees=45
  twistMotion=Free
 }
 {
  bone1Name=dyn_leather_02
  bone2Name=dyn_leather2_02
  swing1LimitDegrees=45
  swing1Motion=Free
  swing2LimitDegrees=45
  swing2Motion=Free
  twistLimitDegrees=45
  twistMotion=Free
 }
 {
  bone1Name=dyn_leather_01
  bone2Name=dyn_leather2_01
  swing1LimitDegrees=45
  swing1Motion=Free
  swing2LimitDegrees=45
  swing2Motion=Free
  twistLimitDegrees=45
  twistMotion=Free
 }
 {
  bone1Name=dyn_leather2_03
  bone2Name=dyn_leather_03
  swing1LimitDegrees=45
  swing1Motion=Free
  swing2LimitDegrees=45
  swing2Motion=Free
  twistLimitDegrees=45
  twistMotion=Free
 }
 {bone1Name=dyn_leather3_03, bone2Name=dyn_leather3_02, swing1LimitDegrees=15, swing2LimitDegrees=15, twistLimitDegrees=5}
]
