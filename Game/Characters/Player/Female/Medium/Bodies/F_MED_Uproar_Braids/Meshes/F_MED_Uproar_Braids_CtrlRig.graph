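# Control Rig for the F_MED_Uproar_Braids female player body that adjusts the left-side strap bones as the left thigh moves.
# A SphericalPoseReader on thigh_l, with an asymmetric active region and a falloff band, outputs a pose weight.
# That weight drives a ModifyTransforms unit that applies preset offsets to bones including dyn_straps_fl_master and dyn_straps_bl_master.
# The graph has no pose reader for the right thigh.
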ControlRigBlueprint /Game/Characters/Player/Female/Medium/Bodies/F_MED_Uproar_Braids/Meshes/F_MED_Uproar_Braids_CtrlRig
graphs [
 RigVMFunctionLibrary {links [], nodes []}
 RigVMModel {nodes [
   RigUnit_BeginExecution = &RigUnit_BeginExecution()
   n1 = &RigUnit_SphericalPoseReader(
    ActiveRegionScaleFactors=(PositiveWidth=3.5,NegativeWidth=0.5,PositiveHeight=3,NegativeHeight=3)
    ActiveRegionSize=0.1
    Debug=(bDrawDebug=false,DebugScale=25,DebugSegments=20,DebugThickness=0.25)
    DriverAxis=(X=-1,Y=0,Z=0)
    DriverItem=(Type=Bone,Name="thigh_l")
    FalloffRegionScaleFactors=(PositiveWidth=1,NegativeWidth=1,PositiveHeight=1,NegativeHeight=1)
    FalloffSize=0.45
    OptionalParentItem=(Type=Bone)
    RotationOffset=(X=90,Y=0,Z=-15)
    ExecutePin=@RigUnit_BeginExecution.ExecutePin
   )
   > &RigUnit_ModifyTransforms(
    ItemToModify="((Item=(Type=Bone,Name=\"dyn_straps_fl_master\"),Transform=(Rotation=(X=0.0708,Y=-0.3152,Z=-0.1255,W=0.938),Translation=(X=-2,Y=4,Z=4),Scale3D=(X=1,Y=1,Z=1))),(Item=(Type=Bone,Name=\"dyn_straps_bl_master\"),Transform=(Rotation=(X=0,Y=-0.2164,Z=0,W=0.9763),Tran ...(truncated)"
    WeightMinimum=0
    Weight=@n1.OutputParam
   )
  ]}
]
hierarchy [
 root {loc=<0 0 0>, rot=<0 0 0 1>}
 attach {parent=root, loc=<0 0 0>, rot=<0 0 0 1>}
 pelvis {parent=root, loc=<0 -0.5301 98.0571>, rot=<0.0127 0.707 -0.0127 -0.707>}
 spine_01 {parent=pelvis, loc=<6.1208 -0.9328 0>, rot=<0 0 0.034 -0.9994>}
 spine_02 {parent=spine_01, loc=<6.5996 -0.2095 0>, rot=<0 0 0.0238 -0.9997>}
 spine_03 {parent=spine_02, loc=<6.4914 -0.1237 0>, rot=<0 0 0.0149 -0.9999>}
 spine_04 {parent=spine_03, loc=<6.9598 0.0794 0>, rot=<0 0 -0.0132 -0.9999>}
 spine_05 {parent=spine_04, loc=<6.9987 0 0>, rot=<0 0 -0.0015 -1>}
 clavicle_l {parent=spine_05, loc=<5.4831 -0.1053 -3.2946>, rot=<0.0465 -0.7289 -0.009 -0.6829>}
 upperarm_l {parent=clavicle_l, loc=<9.8658 -1.2136 -0.4105>, rot=<0.0164 -0.3245 -0.0269 -0.9454>}
 lowerarm_l {parent=upperarm_l, loc=<26.465 -0.8345 0.259>, rot=<0.0036 0.0161 -0.2298 -0.9731>}
 hand_l {parent=lowerarm_l, loc=<22.7355 -0.3203 -0.6105>, rot=<-0.6784 -0.0223 0.0735 -0.7307>}
 index_metacarpal_l {parent=hand_l, loc=<1.6535 0.0183 -2.96>, rot=<0.0158 -0.0124 -0.0173 -0.9996>}
 index_01_l {parent=index_metacarpal_l, loc=<8.3008 0.0357 -0.0524>, rot=<-0.0044 0.0485 0.0924 -0.9945>}
 index_02_l {parent=index_01_l, loc=<3.3587 -0.3004 0.0452>, rot=<-0.002 0.0027 0.3225 -0.9466>}
 index_03_l {parent=index_02_l, loc=<2.5578 -0.123 -0.0252>, rot=<-0.002 0.0067 0.2558 -0.9667>}
 middle_metacarpal_l {parent=hand_l, loc=<2.1856 -0.1725 -0.5984>, rot=<0.006 0.0168 -0.0329 -0.9993>}
 middle_01_l {parent=middle_metacarpal_l, loc=<7.3275 -0.2447 0.0129>, rot=<-0.0208 0.0042 0.0858 -0.9961>}
 middle_02_l {parent=middle_01_l, loc=<4.0123 0.1401 0.1175>, rot=<-0.0218 0.0258 0.3512 -0.9357>}
 middle_03_l {parent=middle_02_l, loc=<2.6077 0.0534 0.1373>, rot=<0.0082 0.0188 0.1708 -0.9851>}
 pinky_metacarpal_l {parent=hand_l, loc=<1.7616 -0.9753 3.3814>, rot=<-0.0397 0.0814 -0.069 -0.9935>}
 pinky_01_l {parent=pinky_metacarpal_l, loc=<7.1709 -0.4831 -0.1094>, rot=<-0.0654 -0.0035 0.2242 -0.9723>}
 pinky_02_l {parent=pinky_01_l, loc=<2.8274 0.0893 -0.1043>, rot=<0.0037 0.0011 0.2854 -0.9584>}
 pinky_03_l {parent=pinky_02_l, loc=<2.2046 0.0307 -0.0078>, rot=<-0.0002 0.0098 0.1445 -0.9895>}
 ring_metacarpal_l {parent=hand_l, loc=<2.2074 -0.3625 1.3638>, rot=<-0.0284 0.0722 -0.0458 -0.9959>}
 ring_01_l {parent=ring_metacarpal_l, loc=<7.0295 -0.6197 -0.1446>, rot=<-0.0489 -0.0252 0.1892 -0.9804>}
 ring_02_l {parent=ring_01_l, loc=<3.6174 0.4647 -0.1027>, rot=<0 0 0.2806 -0.9598>}
 ring_03_l {parent=ring_02_l, loc=<2.4152 0.114 0.0199>, rot=<0 0 0.1189 -0.9929>}
 thumb_01_l {parent=hand_l, loc=<1.3118 -0.3191 -2.459>, rot=<0.6546 -0.4042 0.0664 -0.6354>}
 thumb_02_l {parent=thumb_01_l, loc=<5.9336 -0.0448 0.0401>, rot=<-0.0037 0.0213 0.298 -0.9543>}
 thumb_03_l {parent=thumb_02_l, loc=<2.2223 0.3682 -0.0907>, rot=<0.0118 -0.013 0.0066 -0.9998>}
 weapon_l {parent=hand_l, loc=<-0.1974 -3.4111 0.3811>, rot=<0 0 0 1>}
 lowerarm_twist_01_l {parent=lowerarm_l, loc=<18.316 0 0>, rot=<0 0 0 1>}
 lowerarm_twist_02_l {parent=lowerarm_l, loc=<10.6707 0 0>, rot=<0 0 0 1>}
 upperarm_twist_01_l {parent=upperarm_l, loc=<7.2135 -0.1134 0>, rot=<0 0 0 1>}
 upperarm_twist_02_l {parent=upperarm_l, loc=<19.7998 -0.1134 0>, rot=<0 0 0 1>}
 clavicle_r {parent=spine_05, loc=<5.4831 -0.1053 3.2946>, rot=<0.7289 0.0465 -0.6829 0.009>}
 upperarm_r {parent=clavicle_r, loc=<-9.8658 1.2136 0.4104>, rot=<0.0164 -0.3245 -0.0269 -0.9454>}
 lowerarm_r {parent=upperarm_r, loc=<-26.4651 0.8345 -0.259>, rot=<0.0036 0.0161 -0.2298 -0.9731>}
 hand_r {parent=lowerarm_r, loc=<-22.7356 0.3203 0.6106>, rot=<-0.6784 -0.0223 0.0735 -0.7307>}
 index_metacarpal_r {parent=hand_r, loc=<-1.6534 -0.0181 2.9601>, rot=<0.0158 -0.0124 -0.0173 -0.9996>}
 index_01_r {parent=index_metacarpal_r, loc=<-8.3008 -0.0357 0.0524>, rot=<-0.0044 0.0485 0.0924 -0.9945>}
 index_02_r {parent=index_01_r, loc=<-3.3589 0.3003 -0.0453>, rot=<-0.002 0.0027 0.3225 -0.9466>}
 index_03_r {parent=index_02_r, loc=<-2.5578 0.123 0.0252>, rot=<-0.002 0.0067 0.2558 -0.9667>}
 middle_metacarpal_r {parent=hand_r, loc=<-2.1856 0.1726 0.5984>, rot=<0.006 0.0168 -0.0329 -0.9993>}
 middle_01_r {parent=middle_metacarpal_r, loc=<-7.3273 0.2447 -0.0129>, rot=<-0.0208 0.0042 0.0858 -0.9961>}
 middle_02_r {parent=middle_01_r, loc=<-4.0124 -0.14 -0.1175>, rot=<-0.0218 0.0258 0.3512 -0.9357>}
 middle_03_r {parent=middle_02_r, loc=<-2.6077 -0.0533 -0.1373>, rot=<0.0082 0.0188 0.1708 -0.9851>}
 pinky_metacarpal_r {parent=hand_r, loc=<-1.7617 0.9753 -3.3814>, rot=<-0.0397 0.0814 -0.069 -0.9935>}
 pinky_01_r {parent=pinky_metacarpal_r, loc=<-7.1707 0.4832 0.1094>, rot=<-0.0654 -0.0035 0.2242 -0.9723>}
 pinky_02_r {parent=pinky_01_r, loc=<-2.8274 -0.0892 0.1043>, rot=<0.0037 0.0011 0.2854 -0.9584>}
 pinky_03_r {parent=pinky_02_r, loc=<-2.2045 -0.031 0.0078>, rot=<-0.0002 0.0098 0.1445 -0.9895>}
 ring_metacarpal_r {parent=hand_r, loc=<-2.2073 0.3627 -1.3638>, rot=<-0.0284 0.0722 -0.0458 -0.9959>}
 ring_01_r {parent=ring_metacarpal_r, loc=<-7.0297 0.6197 0.1446>, rot=<-0.0489 -0.0252 0.1892 -0.9804>}
 ring_02_r {parent=ring_01_r, loc=<-3.6171 -0.4648 0.1027>, rot=<0 0 0.2806 -0.9598>}
 ring_03_r {parent=ring_02_r, loc=<-2.4153 -0.1141 -0.0199>, rot=<0 0 0.1189 -0.9929>}
 thumb_01_r {parent=hand_r, loc=<-1.3119 0.3192 2.459>, rot=<0.6546 -0.4042 0.0664 -0.6354>}
 thumb_02_r {parent=thumb_01_r, loc=<-5.9335 0.0447 -0.0401>, rot=<-0.0037 0.0213 0.298 -0.9543>}
 thumb_03_r {parent=thumb_02_r, loc=<-2.2223 -0.3682 0.0907>, rot=<0.0118 -0.013 0.0066 -0.9998>}
 weapon_r {parent=hand_r, loc=<0.1975 3.4112 -0.3811>, rot=<0 0 0 -1>}
 lowerarm_twist_01_r {parent=lowerarm_r, loc=<-18.3161 0 0>, rot=<0 0 0 1>}
 dyn_armband {parent=lowerarm_twist_01_r, loc=<-2.5023 0.3054 0.2058>, rot=<0.914 -0.2199 0.3332 0.0722>}
 lowerarm_twist_02_r {parent=lowerarm_r, loc=<-10.6707 0 0>, rot=<0 0 0 1>}
 upperarm_twist_01_r {parent=upperarm_r, loc=<-7.2135 0.1134 0>, rot=<0 0 0 1>}
 upperarm_twist_02_r {parent=upperarm_r, loc=<-19.7998 0.1134 0>, rot=<0 0 0 1>}
 shoulder_helper {parent=clavicle_r, loc=<-5.1586 0.7635 0.1866>, rot=<0 0 0 1>}
 neck_01 {parent=spine_05, loc=<11.3138 0.1529 0>, rot=<0 0 -0.2003 -0.9797>}
 neck_02 {parent=neck_01, loc=<4.5627 0 0>, rot=<0 0 0.0504 -0.9987>}
 head {parent=neck_02, loc=<5.3914 0 0>, rot=<0 0 0.115 -0.9934>}
 thigh_l {parent=pelvis, loc=<-1.0855 -0.2109 -8.7053>, rot=<0.0009 0.0355 0.0155 -0.9993>}
 calf_l {parent=thigh_l, loc=<-42.5284 -0.0019 -0.7297>, rot=<0.0006 -0.0161 -0.0391 -0.9991>}
 calf_twist_01_l {parent=calf_l, loc=<-35.8483 -1.9197 -0.5794>, rot=<0.0237 0.0095 -0.0035 -0.9997>}
 dyn_boot_master_l {parent=calf_twist_01_l, loc=<-0.0002 0 0>, rot=<0 0 0 1>}
 dyn_boot_flap_a_01_l {parent=dyn_boot_master_l, loc=<5.6022 3.811 -5.8157>, rot=<0.2085 -0.8512 -0.4619 -0.137>}
 dyn_boot_flap_a_02_l {parent=dyn_boot_flap_a_01_l, loc=<2.391 0 0>, rot=<0.5703 0.418 0.0216 -0.7068>}
 dyn_boot_flap_a_03_l {parent=dyn_boot_flap_a_02_l, loc=<4.0547 0 0>, rot=<-0.0888 -0.0665 -0.0199 -0.9936>}
 dyn_boot_flap_a_04_l {parent=dyn_boot_flap_a_03_l, loc=<2.04 0 0>, rot=<0 0 0 1>}
 dyn_boot_flap_b_01_l {parent=dyn_boot_master_l, loc=<8.8003 -6.7492 -0.1119>, rot=<0.1118 -0.9763 -0.1854 -0.0079>}
 dyn_boot_flap_b_02_l {parent=dyn_boot_flap_b_01_l, loc=<3.8571 0 0>, rot=<-0.0663 -0.0298 0.0946 -0.9929>}
 dyn_boot_flap_b_03_l {parent=dyn_boot_flap_b_02_l, loc=<3.3236 0 0>, rot=<0 0 0 1>}
 dyn_boot_flap_c_01_l {parent=dyn_boot_master_l, loc=<5.9386 2.3563 6.923>, rot=<0.2522 -0.957 -0.1421 -0.0192>}
 dyn_boot_flap_c_02_l {parent=dyn_boot_flap_c_01_l, loc=<2.3063 0 0>, rot=<0.8962 0.3562 0.0612 0.2573>}
 dyn_boot_flap_c_03_l {parent=dyn_boot_flap_c_02_l, loc=<3.2718 0 0>, rot=<-0.1015 -0.0208 0.0566 -0.993>}
 dyn_boot_flap_c_04_l {parent=dyn_boot_flap_c_03_l, loc=<2.1255 0 0>, rot=<0 0 0 1>}
 calf_twist_02_l {parent=calf_l, loc=<-19.1546 -1.7943 -0.2638>, rot=<0.0237 0.0095 -0.0035 -0.9997>}
 foot_l {parent=calf_l, loc=<-46.7402 -2.5014 -0.592>, rot=<0.0427 -0.0238 0.0353 -0.9982>}
 ball_l {parent=foot_l, loc=<-3.5658 10.8066 -0.1544>, rot=<0.0001 -0.0001 -0.7063 -0.7079>}
 thigh_twist_01_l {parent=thigh_l, loc=<-27.952 0 0>, rot=<0.0013 0 0 -1>}
 dyn_straps_bl_aim {parent=thigh_l, loc=<-14.8338 -9.4023 -1.8544>, rot=<-0.0519 -0.9614 -0.0134 -0.2699>}
 dyn_straps_fl_aim {parent=thigh_l, loc=<-14.0752 5.4369 -7.2518>, rot=<0.0288 -0.9783 0.0069 -0.2052>}
 thigh_r {parent=pelvis, loc=<-1.0853 -0.2109 8.7053>, rot=<0.0355 -0.0009 0.9993 0.0155>}
 calf_r {parent=thigh_r, loc=<42.5286 0.0019 0.7297>, rot=<0.0006 -0.0161 -0.0391 -0.9991>}
 calf_twist_01_r {parent=calf_r, loc=<35.8485 1.9197 0.5794>, rot=<0.0237 0.0095 -0.0035 -0.9997>}
 dyn_right_boot_master_r {parent=calf_twist_01_r, loc=<0 0 0>, rot=<0 0 0 1>}
 dyn_boot_flap_a_01_r {parent=dyn_right_boot_master_r, loc=<-5.6022 -3.811 5.8157>, rot=<0.8594 0.1523 0.1088 -0.4758>}
 dyn_boot_flap_a_02_r {parent=dyn_boot_flap_a_01_r, loc=<4.0458 0 0>, rot=<-0.5805 -0.481 0.1428 -0.6413>}
 dyn_boot_flap_a_03_r {parent=dyn_boot_flap_a_02_r, loc=<4.1159 0 0>, rot=<0.1788 0.1404 -0.2365 -0.9447>}
 dyn_boot_flap_a_04_r {parent=dyn_boot_flap_a_03_r, loc=<2.3398 0 0>, rot=<0 0 0 1>}
 dyn_boot_flap_b_01_r {parent=dyn_right_boot_master_r, loc=<-7.2515 7.1008 0.2465>, rot=<0.9684 0.1177 0.017 -0.219>}
 dyn_boot_flap_b_02_r {parent=dyn_boot_flap_b_01_r, loc=<4.1189 0 0>, rot=<0.0979 0.0449 0.0863 -0.9904>}
 dyn_boot_flap_b_03_r {parent=dyn_boot_flap_b_02_r, loc=<3.3235 0 0>, rot=<0 0 0 1>}
 dyn_boot_flap_c_01_r {parent=dyn_right_boot_master_r, loc=<-6.077 -2.4776 -6.562>, rot=<0.9913 0.1193 -0.0206 -0.0523>}
 dyn_boot_flap_c_02_r {parent=dyn_boot_flap_c_01_r, loc=<4.3973 0 0>, rot=<0.9044 0.2882 -0.0068 -0.3145>}
 dyn_boot_flap_c_03_r {parent=dyn_boot_flap_c_02_r, loc=<2.8732 0 0>, rot=<-0.0639 -0.0202 0.0197 -0.9976>}
 dyn_boot_flap_c_04_r {parent=dyn_boot_flap_c_03_r, loc=<2.1543 0 0>, rot=<0 0 0 1>}
 dyn_boot_flap_d_01_r {parent=dyn_right_boot_master_r, loc=<-4.2164 -5.5004 -0.6936>, rot=<-0.6847 0.0133 0.727 0.0496>}
 dyn_boot_flap_d_02_r {parent=dyn_boot_flap_d_01_r, loc=<-0.4712 1.5093 -1.7068>, rot=<0 0 0 1>}
 dyn_boot_flap_d_03_r {parent=dyn_boot_flap_d_02_r, loc=<-0.5855 1.3451 -1.987>, rot=<0 0 0 1>}
 calf_twist_02_r {parent=calf_r, loc=<19.1546 1.7943 0.2638>, rot=<0.0237 0.0095 -0.0035 -0.9997>}
 foot_r {parent=calf_r, loc=<46.7403 2.5014 0.592>, rot=<0.0427 -0.0238 0.0353 -0.9982>}
 ball_r {parent=foot_r, loc=<3.5658 -10.8066 0.1544>, rot=<0.0001 -0.0001 -0.7063 -0.7079>}
 thigh_twist_01_r {parent=thigh_r, loc=<27.952 0 0>, rot=<0.0013 0 0 -1>}
 dyn_straps_fr_aim {parent=thigh_r, loc=<18.1994 -3.965 6.7526>, rot=<0.9728 0.088 0.2133 0.0202>}
 dyn_straps_br_aim {parent=thigh_r, loc=<20.5308 10.2862 2.3446>, rot=<0.9697 -0.0225 0.2434 -0.0046>}
 dyn_leather_master {parent=pelvis, loc=<4.427 5.4125 9.4669>, rot=<0.0583 -0.8952 0.3716 0.2389>}
 dyn_leather_01 {parent=dyn_leather_master, loc=<0 0 0>, rot=<-0.7341 0 0 -0.6791>}
 dyn_leather_02 {parent=dyn_leather_01, loc=<5.1952 0 0>, rot=<0.6659 0.211 -0.0339 -0.7148>}
 dyn_leather_03 {parent=dyn_leather_02, loc=<3.3386 0 0>, rot=<0.0724 0 0 -0.9974>}
 dyn_leather2_master {parent=pelvis, loc=<6.5918 0.5257 12.6848>, rot=<0.2107 -0.7175 0.6577 0.0908>}
 dyn_leather2_01 {parent=dyn_leather2_master, loc=<0 0 0>, rot=<0 0 0 1>}
 dyn_leather2_02 {parent=dyn_leather2_01, loc=<4.8291 0 0>, rot=<0.1983 0.0845 -0.0741 -0.9737>}
 dyn_leather2_03 {parent=dyn_leather2_02, loc=<5.8106 0 0>, rot=<0 0 0 1>}
 dyn_leather3_master {parent=pelvis, loc=<8.1952 -6.9085 8.3618>, rot=<0.1881 -0.8354 0.4994 0.1314>}
 dyn_leather3_01 {parent=dyn_leather3_master, loc=<0 0 0>, rot=<-0.0174 -0.0105 0.066 -0.9976>}
 dyn_leather3_02 {parent=dyn_leather3_01, loc=<4.152 0 0>, rot=<-0.1351 -0.0912 -0.0165 -0.9865>}
 dyn_leather3_03 {parent=dyn_leather3_02, loc=<4.3873 0 0>, rot=<0.3514 0.1706 -0.115 -0.9133>}
 dyn_straps_r_master {parent=pelvis, loc=<0 0 0>, rot=<0 0 0 1>}
 dyn_straps_fr_master {parent=dyn_straps_r_master, loc=<1.0082 7.7652 6.0072>, rot=<0.0723 -0.9658 -0.0275 0.2473>}
 dyn_straps_fr_01 {parent=dyn_straps_fr_master, loc=<0 0 0>, rot=<0 0 0 1>}
 dyn_straps_fr_02 {parent=dyn_straps_fr_01, loc=<2.3824 0.5936 0>, rot=<0 0 0 1>}
 dyn_straps_fr_03 {parent=dyn_straps_fr_02, loc=<4.1981 0 0>, rot=<0.0021 -0.01 0.0215 -0.9997>}
 dyn_straps_fr_04 {parent=dyn_straps_fr_03, loc=<7.9823 0 0>, rot=<-0.0453 0.0713 0.33 -0.9402>}
 dyn_straps_fr_05 {parent=dyn_straps_fr_04, loc=<12.5722 0 0>, rot=<0 0 0 1>}
 dyn_straps_br_master {parent=dyn_straps_r_master, loc=<-2.6182 -11.4867 0.4218>, rot=<-0.0371 -0.96 0.0008 0.2777>}
 dyn_straps_br_01 {parent=dyn_straps_br_master, loc=<0 0 0>, rot=<0 0 0 1>}
 dyn_straps_br_02 {parent=dyn_straps_br_01, loc=<4.6266 0 0>, rot=<-0.0067 -0.0621 -0.0663 -0.9958>}
 dyn_straps_br_03 {parent=dyn_straps_br_02, loc=<9.9743 0 0>, rot=<-0.0356 -0.0952 -0.1818 -0.9781>}
 dyn_straps_br_04 {parent=dyn_straps_br_03, loc=<9.9612 0 0>, rot=<0 0 0 1>}
 dyn_straps_l_master {parent=pelvis, loc=<0 0 0>, rot=<0 0 0 1>}
 dyn_straps_bl_master {parent=dyn_straps_l_master, loc=<-2.6455 -11.3411 -2.029>, rot=<-0.066 -0.9503 -0.0102 -0.304>}
 dyn_straps_bl_01 {parent=dyn_straps_bl_master, loc=<0 0 0>, rot=<0 0 0 1>}
 dyn_straps_bl_02 {parent=dyn_straps_bl_01, loc=<5.4798 0 0>, rot=<0.0164 0.0915 -0.0743 -0.9929>}
 dyn_straps_bl_03 {parent=dyn_straps_bl_02, loc=<9.1083 0 0>, rot=<0.0246 0.0463 -0.233 -0.9711>}
 dyn_straps_bl_04 {parent=dyn_straps_bl_03, loc=<8.5566 0 0>, rot=<0 0 0 1>}
 dyn_straps_fl_master {parent=dyn_straps_l_master, loc=<2.599 6.2533 -7.9901>, rot=<0.0135 -0.9707 0.0119 -0.2396>}
 dyn_straps_fl_01 {parent=dyn_straps_fl_master, loc=<0 0 0>, rot=<0 0 0 1>}
 dyn_straps_fl_02 {parent=dyn_straps_fl_01, loc=<2.7244 0.6898 -0.5163>, rot=<0 0 0 1>}
 dyn_straps_fl_03 {parent=dyn_straps_fl_02, loc=<4.4855 0 0>, rot=<0.0014 -0.0075 0.1214 -0.9926>}
 dyn_straps_fl_04 {parent=dyn_straps_fl_03, loc=<7.9298 0 0>, rot=<0.0248 -0.0377 0.271 -0.9615>}
 dyn_straps_fl_05 {parent=dyn_straps_fl_04, loc=<9.4022 0 0>, rot=<0 0 0 1>}
 dyn_gun_master {parent=pelvis, loc=<-14.0828 -6.7204 -18.67>, rot=<0.0336 -0.9862 0.1548 -0.0471>}
 dyn_gun {parent=dyn_gun_master, loc=<5.1529 5.482 -0.0466>, rot=<0 0 0 1>}
 ik_foot_root {parent=root, loc=<0 0 0>, rot=<0 0 0 1>}
 ik_foot_l {parent=ik_foot_root, loc=<14.8262 -7.3545 8.296>, rot=<0.0357 0.7045 0.0284 -0.7082>}
 ik_foot_r {parent=ik_foot_root, loc=<-14.8262 -7.3546 8.2958>, rot=<0.7082 0.0284 -0.7045 0.0357>}
 ik_hand_root {parent=root, loc=<0 0 0>, rot=<0 0 0 1>}
 ik_hand_gun {parent=ik_hand_root, loc=<-48.1744 3.3546 104.4381>, rot=<0.6276 0.1309 0.4183 -0.6434>}
 ik_hand_l {parent=ik_hand_gun, loc=<59.3272 67.6997 34.3524>, rot=<-0.6158 -0.7027 -0.3565 0>}
 ik_hand_r {parent=ik_hand_gun, loc=<0 0 0>, rot=<0 0 0 1>}
 root {parent=root, type=Socket, loc=<0 0 0>, rot=<0 0 0 1>}
]
