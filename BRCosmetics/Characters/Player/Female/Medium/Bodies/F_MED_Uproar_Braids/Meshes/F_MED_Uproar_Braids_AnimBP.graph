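# Post-process-style Anim Blueprint for the F_MED_Uproar_Braids female body cosmetic, adding secondary motion to straps, leather, armband, gun and feet.
# It copies the attached parent's pose, constrains straps to thighs and pelvis, and rotates leather by crouch, jump and right-thigh angle.
# That cached pose feeds separate front-end and gameplay RigidBody/AnimDynamics chains, selected by bIsInFrontEnd.
# Dynamics reset shortly after emotes end; the Control Rig node is unconnected.
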AnimBlueprint /BRCosmetics/Characters/Player/Female/Medium/Bodies/F_MED_Uproar_Braids/Meshes/F_MED_Uproar_Braids_AnimBP
target_skeleton=/Game/Characters/Player/Female/Medium/Bodies/F_MED_Uproar_Braids/Meshes/F_MED_Uproar_Braids_Skeleton
graphs [
 AnimGraph {nodes [
   Control Rig!(Source, bAlphaBoolEnabled:bool=NOT Boolean(A=Get bIsCrouching())) -> (Pose) {
    alphaInputType=Bool
    alphaBoolBlend {blendInTime=0.1, blendOption=Linear, blendOutTime=0.1}
    controlRigAssetReference {blueprintRigClass=/Game/Characters/Player/Female/Medium/Bodies/F_MED_Uproar_Braids/Meshes/F_MED_Uproar_Braids_CtrlRig.F_MED_Uproar_Braids_CtrlRig_C}
   }
   n1 = Anim Dynamics - Bone: dyn_straps_bl_01(ComponentPose=Local To Component(LocalPose=Use cached pose 'Presim'()), Alpha=1, ExternalForce="0,0,0") {
    angularDampingOverride=1
    angularSpringConstant=20000
    bAngularSpring=true
    bOverrideAngularDamping=true
    bUseGravityOverride=true
    gravityScale=0.35
    simulationSpace=World
    alphaBoolBlend {blendOption=Linear}
    boundBone {boneName=dyn_straps_bl_01}
    componentLinearAccScale=<1 1 1>
    componentLinearVelScale=<1 1 1>
    gravityOverride {z=-980}
    physicsBodyDefinitions [{
      boundBone {boneName=dyn_straps_bl_01}
      boxExtents=<18 25 25>
      constraintSetup {angularLimitsMax=<1 0 4>, angularLimitsMin=<-1 0 -4>, angularTarget=<1 0 0>}
      localJointOffset=<2 0 0>
     }]
   }
   n2 = Anim Dynamics - Bone: dyn_straps_bl_02(ComponentPose=@n1, Alpha=1, ExternalForce="0,0,0") {
    angularDampingOverride=1
    angularSpringConstant=20000
    bAngularSpring=true
    bOverrideAngularDamping=true
    bUseGravityOverride=true
    gravityScale=0.35
    simulationSpace=World
    alphaBoolBlend {blendOption=Linear}
    boundBone {boneName=dyn_straps_bl_02}
    componentLinearAccScale=<1 1 1>
    componentLinearVelScale=<1 1 1>
    gravityOverride {z=-980}
    physicsBodyDefinitions [{
      boundBone {boneName=dyn_straps_bl_02}
      boxExtents=<18 25 25>
      constraintSetup {angularLimitsMax=<1 0 4>, angularLimitsMin=<-1 0 -4>, angularTarget=<1 0 0>}
      localJointOffset=<2 0 0>
     }]
   }
   n3 = Anim Dynamics - Bone: dyn_straps_bl_03(ComponentPose=@n2, Alpha=1, ExternalForce="0,0,0") {
    angularDampingOverride=1
    angularSpringConstant=20000
    bAngularSpring=true
    bOverrideAngularDamping=true
    bUseGravityOverride=true
    gravityScale=0.35
    simulationSpace=World
    alphaBoolBlend {blendOption=Linear}
    boundBone {boneName=dyn_straps_bl_03}
    componentLinearAccScale=<1 1 1>
    componentLinearVelScale=<1 1 1>
    gravityOverride {z=-980}
    physicsBodyDefinitions [{
      boundBone {boneName=dyn_straps_bl_03}
      boxExtents=<18 25 25>
      constraintSetup {angularLimitsMax=<1 0 4>, angularLimitsMin=<-1 0 -4>, angularTarget=<1 0 0>}
      localJointOffset=<2 0 0>
     }]
   }
   n4 = Anim Dynamics - Bone: dyn_straps_fl_01(ComponentPose=@n3, Alpha=1, ExternalForce="0,0,0") {
    angularDampingOverride=1
    angularSpringConstant=20000
    bAngularSpring=true
    bOverrideAngularDamping=true
    bUseGravityOverride=true
    gravityScale=0.35
    simulationSpace=World
    alphaBoolBlend {blendOption=Linear}
    boundBone {boneName=dyn_straps_fl_01}
    componentLinearAccScale=<1 1 1>
    componentLinearVelScale=<1 1 1>
    gravityOverride {z=-980}
    physicsBodyDefinitions [{
      boundBone {boneName=dyn_straps_fl_01}
      boxExtents=<18 25 25>
      constraintSetup {angularLimitsMax=<1 0 4>, angularLimitsMin=<-1 0 -4>, angularTarget=<1 0 0>}
      localJointOffset=<2 0 0>
     }]
   }
   n5 = Anim Dynamics - Bone: dyn_straps_fl_02(ComponentPose=@n4, Alpha=1, ExternalForce="0,0,0") {
    angularDampingOverride=1
    angularSpringConstant=20000
    bAngularSpring=true
    bOverrideAngularDamping=true
    bUseGravityOverride=true
    gravityScale=0.35
    simulationSpace=World
    alphaBoolBlend {blendOption=Linear}
    boundBone {boneName=dyn_straps_fl_02}
    componentLinearAccScale=<1 1 1>
    componentLinearVelScale=<1 1 1>
    gravityOverride {z=-980}
    physicsBodyDefinitions [{
      boundBone {boneName=dyn_straps_fl_02}
      boxExtents=<18 25 25>
      constraintSetup {angularLimitsMax=<1 0 4>, angularLimitsMin=<-1 0 -4>, angularTarget=<1 0 0>}
      localJointOffset=<2 0 0>
     }]
   }
   n6 = Anim Dynamics - Bone: dyn_straps_fl_03(ComponentPose=@n5, Alpha=1, ExternalForce="0,0,0") {
    angularDampingOverride=1
    angularSpringConstant=20000
    bAngularSpring=true
    bOverrideAngularDamping=true
    bUseGravityOverride=true
    gravityScale=0.35
    simulationSpace=World
    alphaBoolBlend {blendOption=Linear}
    boundBone {boneName=dyn_straps_fl_03}
    componentLinearAccScale=<1 1 1>
    componentLinearVelScale=<1 1 1>
    gravityOverride {z=-980}
    physicsBodyDefinitions [{
      boundBone {boneName=dyn_straps_fl_03}
      boxExtents=<18 25 25>
      constraintSetup {angularLimitsMax=<1 0 4>, angularLimitsMin=<-1 0 -4>, angularTarget=<1 0 0>}
      localJointOffset=<2 0 0>
     }]
   }
   n7 = RigidBody(ComponentPose=@n6) {
    bOverrideWorldGravity=true
    overridePhysicsAsset=&/F_MED_Uproar_Braids_Frontend_Physics
    simulationSpace=BaseBoneSpace
    alphaBoolBlend {blendOption=Linear}
    baseBoneRef {boneName=pelvis}
    componentAppliedLinearAccClamp=<2000 2000 2000>
    componentLinearAccScale=<0.3 0.3 0.35>
    componentLinearVelScale=<0.3 0.3 0.35>
    overrideWorldGravity {z=-980}
   }
   n8 = RigidBody(ComponentPose=@n7) {
    bOverrideWorldGravity=true
    overridePhysicsAsset=&/F_MED_Uproar_Braids_BeltR_Physics
    simulationSpace=BaseBoneSpace
    alphaBoolBlend {blendOption=Linear}
    baseBoneRef {boneName=pelvis}
    componentAppliedLinearAccClamp=<750 750 750>
    componentLinearAccScale=<0.1 0.1 0.15>
    componentLinearVelScale=<0.1 0.1 0.15>
    overrideWorldGravity {z=-980}
   }
   n9 = RigidBody(ComponentPose=@n8) {
    bOverrideWorldGravity=true
    overridePhysicsAsset=&/F_MED_Uproar_Braids_FootL_Physics
    simulationSpace=BaseBoneSpace
    alphaBoolBlend {blendOption=Linear}
    baseBoneRef {boneName=foot_l}
    componentAppliedLinearAccClamp=<2000 2000 2000>
    componentLinearAccScale=<0.3 0.3 0.35>
    componentLinearVelScale=<0.3 0.3 0.35>
    overrideWorldGravity {z=-980}
   }
   n10 = RigidBody(ComponentPose=@n9) {
    bOverrideWorldGravity=true
    overridePhysicsAsset=&/F_MED_Uproar_Braids_FootR_Physics
    simulationSpace=BaseBoneSpace
    alphaBoolBlend {blendOption=Linear}
    baseBoneRef {boneName=foot_r}
    componentAppliedLinearAccClamp=<2000 2000 2000>
    componentLinearAccScale=<0.3 0.3 0.35>
    componentLinearVelScale=<0.3 0.3 0.35>
    overrideWorldGravity {z=-980}
   }
   n11 = Anim Dynamics - Bone: dyn_armband(ComponentPose=@n10, Alpha=1, ExternalForce="0,0,0") {
    bAngularSpring=true
    bGravityOverrideInSimSpace=true
    bUseGravityOverride=true
    simulationSpace=RootRelative
    alphaBoolBlend {blendOption=Linear}
    boundBone {boneName=dyn_armband}
    componentAppliedLinearAccClamp=<2000 2000 2000>
    componentLinearAccScale=<0.3 0.3 0.35>
    componentLinearVelScale=<0.3 0.3 0.35>
    gravityOverride {z=-980}
    physicsBodyDefinitions [{
      sphereCollisionRadius=0
      boundBone {boneName=dyn_armband}
      boxExtents=<10 10 10>
      constraintSetup {angularLimitsMax=<10 10 10>, angularLimitsMin=<-10 -10 -10>, angularTarget=<0 0 0>}
      localJointOffset=<0 -1 0>
     }]
   }
   n12 = Anim Dynamics - Bone: dyn_gun!(ComponentPose=@n11, Alpha=1, GravityOverride:Vector="0,0,0", ExternalForce="0,0,0") -> (Pose) {
    angularDampingOverride=1
    angularSpringConstant=1500
    bAngularSpring=true
    bGravityOverrideInSimSpace=true
    bOverrideAngularDamping=true
    bUsePlanarLimit=false
    bUseSphericalLimits=true
    gravityScale=0.5
    simulationSpace=RootRelative
    alphaBoolBlend {blendOption=Linear}
    boundBone {boneName=dyn_gun}
    componentAppliedLinearAccClamp=<750 750 750>
    componentLinearAccScale {z=0.2}
    componentLinearVelScale {z=0.2}
    physicsBodyDefinitions [{
      collisionType=CustomSphere
      sphereCollisionRadius=1
      boundBone {boneName=dyn_gun}
      boxExtents=<40 22 22>
      constraintSetup {angularLimitsMax=<10 10 25>, angularLimitsMin=<-10 -10 -25>, angularTarget=<0 0 0>}
      localJointOffset=<10 -0.5 0>
     }]
    relativeSpaceBone {boneName=root}
   }
   n13 = "Transform (Modify) Bone - Bone: dyn_straps_bl_master"!(ComponentPose=@n12, bAlphaBoolEnabled:bool=True, Translation="0,0,0", Rotation="0,0,0", Scale="1,1,1") -> (Pose) {
    alphaInputType=Bool
    rotationMode=BMM_Additive
    translationMode=BMM_Additive
    alphaBoolBlend {blendOption=Linear}
    alphaScaleBiasClamp {interpSpeedIncreasing=15}
    boneToModify {boneName=dyn_straps_bl_master}
   }
   n14 = "Transform (Modify) Bone - Bone: dyn_straps_fl_master"!(ComponentPose=@n13, bAlphaBoolEnabled:bool=True, Translation="0,0,0", Rotation="6.265,3.252,0.474", Scale="1,1,1") -> (Pose) {
    alphaInputType=Bool
    rotationMode=BMM_Additive
    translationMode=BMM_Additive
    alphaBoolBlend {blendOption=Linear}
    alphaScaleBiasClamp {interpSpeedIncreasing=15}
    boneToModify {boneName=dyn_straps_fl_master}
    rotation=<rot 6.265 3.252 0.474>
   }
   n15 = "Transform (Modify) Bone - Bone: dyn_straps_br_master"!(ComponentPose=@n14, bAlphaBoolEnabled:bool=True, Translation="0,0,0", Rotation="0.000,-10.499,0.000", Scale="1,1,1") -> (Pose) {
    alphaInputType=Bool
    rotationMode=BMM_Additive
    rotationSpace=BCS_BoneSpace
    translationMode=BMM_Additive
    alphaBoolBlend {blendOption=Linear}
    alphaScaleBiasClamp {bInterpResult=true, interpSpeedIncreasing=15}
    boneToModify {boneName=dyn_straps_br_master}
    rotation {yaw=-10.499}
   }
   n16 = "Transform (Modify) Bone - Bone: dyn_leather2_master"!(ComponentPose=@n15, bAlphaBoolEnabled:bool=True, Translation="0,0,0", Rotation="-15,0.000,0.000", Scale="1,1,1") -> (Pose) {
    alphaInputType=Bool
    rotationMode=BMM_Additive
    alphaBoolBlend {blendOption=Linear}
    boneToModify {boneName=dyn_leather2_master}
    rotation {pitch=-15}
   }
   n17 = "Transform (Modify) Bone - Bone: dyn_leather3_master"!(ComponentPose=@n16, bAlphaBoolEnabled:bool=True, Translation="0,0,0", Rotation="-15,0.000,0.000", Scale="1,1,1") -> (Pose) {
    alphaInputType=Bool
    rotationMode=BMM_Additive
    alphaBoolBlend {blendOption=Linear}
    boneToModify {boneName=dyn_leather3_master}
    rotation {pitch=-15}
   }
   n18 = "Constraint - dyn_gun_master (PPPP)"!(
    ComponentPose=@n17
    Alpha=1
    ConstraintWeights_0=0.25
    ConstraintWeights_1:float=0.25
    ConstraintWeights_2:float=0.25
    ConstraintWeights_3:float=0.25
   ) -> (Pose) {
    alphaBoolBlend {blendOption=Linear}
    boneToModify {boneName=dyn_gun_master}
    constraintSetup [
     {targetBone {boneName=dyn_straps_bl_03}}
     {targetBone {boneName=dyn_straps_fl_05}}
     {targetBone {boneName=dyn_straps_fl_04}}
     {targetBone {boneName=dyn_straps_bl_04}}
    ]
    constraintWeights [1,1,1,1]
   }
   n19 = "Transform (Modify) Bone - Bone: upperarm_l"(ComponentPose=@n18, Alpha=1, Translation="0,0,0", Rotation="5,-3,0.000", Scale="1,1,1") {
    rotationMode=BMM_Additive
    rotationSpace=BCS_BoneSpace
    alphaBoolBlend {blendOption=Linear}
    boneToModify {boneName=upperarm_l}
    rotation {pitch=5, yaw=-3}
   }
   n20 = "Transform (Modify) Bone - Bone: upperarm_r"(ComponentPose=@n19, Alpha=1, Translation="0,0,0", Rotation="3,-3,0.000", Scale="1,1,1") {
    rotationMode=BMM_Additive
    rotationSpace=BCS_BoneSpace
    alphaBoolBlend {blendOption=Linear}
    boneToModify {boneName=upperarm_r}
    rotation {pitch=3, yaw=-3}
   }
   Output Pose(Result=Blend Poses by bool(BlendPose_0=Component To Local(ComponentPose=@n20), BlendPose_1=Use cached pose 'PostSim'(), bActiveValue=Get bIsInFrontEnd()) {blendType=Linear}) {layerGroup=None}
   n21 = Get CrouchAlpha() -> (CrouchAlpha:double)
   n22 = "Constraint - dyn_straps_fr_02 (PPP)"!(
    ComponentPose=Local To Component(LocalPose=Copy Pose From Mesh() {bUseAttachedParent=true})
    Alpha=1
    ConstraintWeights_0=0.3
    ConstraintWeights_1:float=0
    ConstraintWeights_2:float=0.7
   ) -> (Pose) {
    alphaBoolBlend {blendOption=Linear}
    boneToModify {boneName=dyn_straps_fr_02}
    constraintSetup [{targetBone {boneName=thigh_r}}, {targetBone {boneName=thigh_twist_01_r}}, {targetBone {boneName=pelvis}}]
    constraintWeights [1,1,1]
   }
   n23 = "Constraint - dyn_straps_fr_03 (PPP)"!(ComponentPose=@n22, Alpha=1, ConstraintWeights_0=0, ConstraintWeights_1:float=1, ConstraintWeights_2:float=0) -> (Pose) {
    alphaBoolBlend {blendOption=Linear}
    boneToModify {boneName=dyn_straps_fr_03}
    constraintSetup [{targetBone {boneName=thigh_r}}, {targetBone {boneName=thigh_twist_01_r}}, {targetBone {boneName=pelvis}}]
    constraintWeights [1,1,1]
   }
   n24 = "Constraint - dyn_straps_fr_04 (PPP)"!(ComponentPose=@n23, Alpha=1, ConstraintWeights_0=1, ConstraintWeights_1:float=0, ConstraintWeights_2:float=0) -> (Pose) {
    alphaBoolBlend {blendOption=Linear}
    boneToModify {boneName=dyn_straps_fr_04}
    constraintSetup [{targetBone {boneName=thigh_r}}, {targetBone {boneName=thigh_twist_01_r}}, {targetBone {boneName=pelvis}}]
    constraintWeights [1,1,1]
   }
   n25 = "Constraint - dyn_straps_fr_05 (PPP)"!(ComponentPose=@n24, Alpha=1, ConstraintWeights_0=1, ConstraintWeights_1:float=0, ConstraintWeights_2:float=0) -> (Pose) {
    alphaBoolBlend {blendOption=Linear}
    boneToModify {boneName=dyn_straps_fr_05}
    constraintSetup [{targetBone {boneName=thigh_r}}, {targetBone {boneName=thigh_twist_01_r}}, {targetBone {boneName=pelvis}}]
    constraintWeights [1,1,1]
   }
   n26 = "Constraint - dyn_straps_br_03 (PPP)"!(ComponentPose=@n25, Alpha=1, ConstraintWeights_0=0, ConstraintWeights_1:float=0.6, ConstraintWeights_2:float=0.4) -> (Pose) {
    alphaBoolBlend {blendOption=Linear}
    boneToModify {boneName=dyn_straps_br_03}
    constraintSetup [{targetBone {boneName=thigh_r}}, {targetBone {boneName=thigh_twist_01_r}}, {targetBone {boneName=pelvis}}]
    constraintWeights [1,1,1]
   }
   n27 = "Constraint - dyn_straps_br_02 (PPP)"!(ComponentPose=@n26, Alpha=1, ConstraintWeights_0=0, ConstraintWeights_1:float=0.2, ConstraintWeights_2:float=0.8) -> (Pose) {
    alphaBoolBlend {blendOption=Linear}
    boneToModify {boneName=dyn_straps_br_02}
    constraintSetup [{targetBone {boneName=thigh_r}}, {targetBone {boneName=thigh_twist_01_r}}, {targetBone {boneName=pelvis}}]
    constraintWeights [1,1,1]
   }
   n28 = "Constraint - dyn_straps_fl_02 (PPP)"!(ComponentPose=@n27, Alpha=1, ConstraintWeights_0=0.3, ConstraintWeights_1:float=0, ConstraintWeights_2:float=0.7) -> (Pose) {
    alphaBoolBlend {blendOption=Linear}
    boneToModify {boneName=dyn_straps_fl_02}
    constraintSetup [{targetBone {boneName=thigh_l}}, {targetBone {boneName=thigh_twist_01_l}}, {targetBone {boneName=pelvis}}]
    constraintWeights [1,1,1]
   }
   n29 = "Constraint - dyn_straps_fl_03 (PPP)"!(ComponentPose=@n28, Alpha=1, ConstraintWeights_0=0, ConstraintWeights_1:float=1, ConstraintWeights_2:float=0) -> (Pose) {
    alphaBoolBlend {blendOption=Linear}
    boneToModify {boneName=dyn_straps_fl_03}
    constraintSetup [{targetBone {boneName=thigh_l}}, {targetBone {boneName=thigh_twist_01_l}}, {targetBone {boneName=pelvis}}]
    constraintWeights [1,1,1]
   }
   n30 = "Constraint - dyn_straps_fl_04 (PPP)"!(ComponentPose=@n29, Alpha=1, ConstraintWeights_0=1, ConstraintWeights_1:float=0, ConstraintWeights_2:float=0) -> (Pose) {
    alphaBoolBlend {blendOption=Linear}
    boneToModify {boneName=dyn_straps_fl_04}
    constraintSetup [{targetBone {boneName=thigh_l}}, {targetBone {boneName=thigh_twist_01_l}}, {targetBone {boneName=pelvis}}]
    constraintWeights [1,1,1]
   }
   n31 = "Constraint - dyn_straps_fl_05 (PPP)"!(ComponentPose=@n30, Alpha=1, ConstraintWeights_0=1, ConstraintWeights_1:float=0, ConstraintWeights_2:float=0) -> (Pose) {
    alphaBoolBlend {blendOption=Linear}
    boneToModify {boneName=dyn_straps_fl_05}
    constraintSetup [{targetBone {boneName=thigh_l}}, {targetBone {boneName=thigh_twist_01_l}}, {targetBone {boneName=pelvis}}]
    constraintWeights [1,1,1]
   }
   n32 = "Constraint - dyn_straps_bl_03 (PPP)"!(ComponentPose=@n31, Alpha=1, ConstraintWeights_0=0, ConstraintWeights_1:float=0.6, ConstraintWeights_2:float=0.4) -> (Pose) {
    alphaBoolBlend {blendOption=Linear}
    boneToModify {boneName=dyn_straps_bl_03}
    constraintSetup [{targetBone {boneName=thigh_l}}, {targetBone {boneName=thigh_twist_01_l}}, {targetBone {boneName=pelvis}}]
    constraintWeights [1,1,1]
   }
   n33 = "Constraint - dyn_straps_bl_02 (PPP)"!(ComponentPose=@n32, Alpha=1, ConstraintWeights_0=0, ConstraintWeights_1:float=0.2, ConstraintWeights_2:float=0.8) -> (Pose) {
    alphaBoolBlend {blendOption=Linear}
    boneToModify {boneName=dyn_straps_bl_02}
    constraintSetup [{targetBone {boneName=thigh_l}}, {targetBone {boneName=thigh_twist_01_l}}, {targetBone {boneName=pelvis}}]
    constraintWeights [1,1,1]
   }
   n34 = "Transform (Modify) Bone - Bone: dyn_leather_master"(ComponentPose=@n33, Alpha=@n21, Translation="0,0,0", Rotation="-25,-7.5,-1.090", Scale="1,1,1") {
    rotationMode=BMM_Additive
    translationMode=BMM_Additive
    alphaBoolBlend {blendOption=Linear}
    alphaScaleBiasClamp {interpSpeedIncreasing=15}
    boneToModify {boneName=dyn_leather_master}
    rotation=<rot -25 -7.5 -1.09>
   }
   n35 = "Transform (Modify) Bone - Bone: dyn_leather2_master"!(ComponentPose=@n34, Alpha:float=@n21, Translation="0,0,0", Rotation="-25,0.000,0.000", Scale="1,1,1") -> (Pose) {
    rotationMode=BMM_Additive
    translationMode=BMM_Additive
    alphaBoolBlend {blendOption=Linear}
    alphaScaleBiasClamp {interpSpeedIncreasing=15}
    boneToModify {boneName=dyn_leather2_master}
    rotation {pitch=-25}
   }
   n36 = Get Thigh R Crouch Leather Rot Alpha() -> (Thigh R Crouch Leather Rot Alpha:double)
   n37 = "Transform (Modify) Bone - Bone: dyn_leather_master"(ComponentPose=@n35, Alpha=@n36, Translation="0.000,-0.000,6.400", Rotation="-40,-20,-0.000", Scale="1,1,1") {
    rotationMode=BMM_Additive
    translationMode=BMM_Additive
    alphaBoolBlend {blendOption=Linear}
    alphaScaleBiasClamp {interpSpeedIncreasing=15}
    boneToModify {boneName=dyn_leather_master}
    rotation {pitch=-40, yaw=-20}
    translation {z=6.4}
   }
   n38 = "Transform (Modify) Bone - Bone: dyn_leather2_master"(ComponentPose=@n37, Alpha=@n36, Translation="0.000,-0.000,1.548", Rotation="-60,0.000,0.000", Scale="1,1,1") {
    rotationMode=BMM_Additive
    translationMode=BMM_Additive
    alphaBoolBlend {blendOption=Linear}
    alphaScaleBiasClamp {interpSpeedIncreasing=15}
    boneToModify {boneName=dyn_leather2_master}
    rotation {pitch=-60}
    translation {z=1.548}
   }
   n39 = "Transform (Modify) Bone - Bone: dyn_leather_master"(ComponentPose=@n38, Alpha=1, Translation="0,0,0", Rotation="-30,-10.291,3.502", Scale="1,1,1") {
    rotationMode=BMM_Additive
    rotationSpace=BCS_BoneSpace
    alphaBoolBlend {blendOption=Linear}
    alphaScaleBiasClamp {bInterpResult=true, interpSpeedIncreasing=15}
    boneToModify {boneName=dyn_leather_master}
    rotation=<rot -30 -10.291 3.502>
   }
   n40 = Get Jump Alpha() -> (Jump Alpha:double)
   n41 = "Transform (Modify) Bone - Bone: dyn_leather_master"(ComponentPose=@n39, Alpha=@n40, Translation="0.000,0.000,4.277", Rotation="-5.858,-62.233,0", Scale="1,1,1") {
    rotationMode=BMM_Additive
    translationMode=BMM_Additive
    alphaBoolBlend {blendOption=Linear}
    alphaScaleBiasClamp {interpSpeedIncreasing=15}
    boneToModify {boneName=dyn_leather_master}
    rotation {pitch=-5.858, yaw=-62.233}
    translation {z=4.277}
   }
   n42 = "Transform (Modify) Bone - Bone: dyn_leather2_master"(ComponentPose=@n41, Alpha=@n40, Translation="0.000,-0.000,1.548", Rotation="-90,0,0", Scale="1,1,1") {
    rotationMode=BMM_Additive
    translationMode=BMM_Additive
    alphaBoolBlend {blendOption=Linear}
    alphaScaleBiasClamp {interpSpeedIncreasing=15}
    boneToModify {boneName=dyn_leather2_master}
    rotation {pitch=-90}
    translation {z=1.548}
   }
   n43 = "Transform (Modify) Bone - Bone: dyn_leather3_master"!(ComponentPose=@n42, Alpha:float=@n40, Translation="0.000,-0.000,0.607", Rotation="-46.086,-13.373,17.076", Scale="1,1,1") -> (Pose) {
    rotationMode=BMM_Additive
    translationMode=BMM_Additive
    alphaBoolBlend {blendOption=Linear}
    alphaScaleBiasClamp {interpSpeedIncreasing=15}
    boneToModify {boneName=dyn_leather3_master}
    rotation=<rot -46.086 -13.3731 17.0761>
    translation {z=0.607}
   }
   n44 = "Transform (Modify) Bone - Bone: dyn_straps_l_master"!(ComponentPose=@n43, bAlphaBoolEnabled:bool=Get bIsInFrontEnd(), Translation="0,0,0", Rotation="0,0,0", Scale="1,1,1") -> (Pose) {
    alphaInputType=Bool
    rotationMode=BMM_Additive
    rotationSpace=BCS_BoneSpace
    alphaBoolBlend {blendOption=Linear}
    alphaScaleBiasClamp {bInterpResult=true, interpSpeedIncreasing=15}
    boneToModify {boneName=dyn_straps_l_master}
   }
   Save cached pose 'Presim'(Pose=Component To Local(ComponentPose=@n44)) {cacheName=Presim}
   n45 = RigidBody(ComponentPose=Local To Component(LocalPose=Use cached pose 'Presim'())) {
    bOverrideWorldGravity=true
    overridePhysicsAsset=&/F_MED_Uproar_Braids_Physics_New
    simulationSpace=BaseBoneSpace
    alphaBoolBlend {blendOption=Linear}
    baseBoneRef {boneName=pelvis}
    componentAppliedLinearAccClamp=<2000 2000 2000>
    componentLinearAccScale=<0.3 0.3 0.35>
    componentLinearVelScale=<0.3 0.3 0.35>
    overrideWorldGravity {z=-980}
    simSpaceSettings {maxAngularAcceleration=1000, maxAngularVelocity=3, maxLinearAcceleration=250, maxLinearVelocity=500, velocityScaleZ=0, worldAlpha=1}
   }
   n46 = RigidBody(ComponentPose=@n45) {
    bOverrideWorldGravity=true
    overridePhysicsAsset=&F_MED_Uproar_Braids_FootL_Physics
    simulationSpace=BaseBoneSpace
    alphaBoolBlend {blendOption=Linear}
    baseBoneRef {boneName=foot_l}
    componentAppliedLinearAccClamp=<2000 2000 2000>
    componentLinearAccScale=<0.3 0.3 0.35>
    componentLinearVelScale=<0.3 0.3 0.35>
    overrideWorldGravity {z=-980}
   }
   n47 = RigidBody(ComponentPose=@n46) {
    bOverrideWorldGravity=true
    overridePhysicsAsset=&F_MED_Uproar_Braids_FootR_Physics
    simulationSpace=BaseBoneSpace
    alphaBoolBlend {blendOption=Linear}
    baseBoneRef {boneName=foot_r}
    componentAppliedLinearAccClamp=<2000 2000 2000>
    componentLinearAccScale=<0.3 0.3 0.35>
    componentLinearVelScale=<0.3 0.3 0.35>
    overrideWorldGravity {z=-980}
   }
   n48 = Anim Dynamics - Bone: dyn_armband(ComponentPose=@n47, Alpha=1, ExternalForce="0,0,0") {
    bAngularSpring=true
    bGravityOverrideInSimSpace=true
    bUseGravityOverride=true
    simulationSpace=RootRelative
    alphaBoolBlend {blendOption=Linear}
    boundBone {boneName=dyn_armband}
    componentAppliedLinearAccClamp=<2000 2000 2000>
    componentLinearAccScale=<0.3 0.3 0.35>
    componentLinearVelScale=<0.3 0.3 0.35>
    gravityOverride {z=-980}
    physicsBodyDefinitions [{
      sphereCollisionRadius=0
      boundBone {boneName=dyn_armband}
      boxExtents=<10 10 10>
      constraintSetup {angularLimitsMax=<10 10 10>, angularLimitsMin=<-10 -10 -10>, angularTarget=<0 0 0>}
      localJointOffset=<0 -1 0>
     }]
   }
   n49 = Anim Dynamics - Bone: dyn_straps_fr_03(ComponentPose=@n48, Alpha=1, ExternalForce="0,0,0") {
    angularDampingOverride=0.999
    angularSpringConstant=100
    bAngularSpring=true
    bOverrideAngularDamping=true
    bUseGravityOverride=true
    simulationSpace=RootRelative
    alphaBoolBlend {blendOption=Linear}
    boundBone {boneName=dyn_straps_fr_03}
    gravityOverride {z=-980}
    physicsBodyDefinitions [{
      boundBone {boneName=dyn_straps_fr_03}
      boxExtents=<15 15 15>
      constraintSetup {angularLimitsMax=<0 30 20>, angularLimitsMin=<0 0 0>, angularTarget=<1 0 0>}
      localJointOffset=<2 0 0>
     }]
   }
   n50 = Anim Dynamics - Bone: dyn_straps_fr_04(ComponentPose=@n49, Alpha=1, ExternalForce="0,0,0") {
    angularDampingOverride=0.999
    angularSpringConstant=100
    bAngularSpring=true
    bOverrideAngularDamping=true
    bUseGravityOverride=true
    simulationSpace=RootRelative
    alphaBoolBlend {blendOption=Linear}
    boundBone {boneName=dyn_straps_fr_04}
    gravityOverride {z=-980}
    physicsBodyDefinitions [{
      boundBone {boneName=dyn_straps_fr_04}
      boxExtents=<15 15 15>
      constraintSetup {angularLimitsMax=<0 0 30>, angularLimitsMin=<0 0 0>, angularTarget=<1 0 0>}
      localJointOffset=<2 0 0>
     }]
   }
   n51 = Anim Dynamics - Bone: dyn_straps_br_01(ComponentPose=@n50, Alpha=1, ExternalForce="0,0,0") {
    angularDampingOverride=0.999
    angularSpringConstant=100
    bAngularSpring=true
    bOverrideAngularDamping=true
    bUseGravityOverride=true
    simulationSpace=RootRelative
    alphaBoolBlend {blendOption=Linear}
    boundBone {boneName=dyn_straps_br_01}
    gravityOverride {z=-980}
    physicsBodyDefinitions [{
      boundBone {boneName=dyn_straps_br_01}
      boxExtents=<15 15 15>
      constraintSetup {angularLimitsMax=<0 30 0>, angularLimitsMin=<0 0 -30>, angularTarget=<1 0 0>}
      localJointOffset=<2 0 0>
     }]
   }
   n52 = Anim Dynamics - Bone: dyn_straps_br_02(ComponentPose=@n51, Alpha=1, ExternalForce="0,0,0") {
    angularDampingOverride=0.999
    angularSpringConstant=100
    bAngularSpring=true
    bOverrideAngularDamping=true
    bUseGravityOverride=true
    simulationSpace=RootRelative
    alphaBoolBlend {blendOption=Linear}
    boundBone {boneName=dyn_straps_br_02}
    gravityOverride {z=-980}
    physicsBodyDefinitions [{
      boundBone {boneName=dyn_straps_br_02}
      boxExtents=<15 15 15>
      constraintSetup {angularLimitsMax=<0 30 0>, angularLimitsMin=<0 0 -30>, angularTarget=<1 0 0>}
      localJointOffset=<2 0 0>
     }]
   }
   n53 = Anim Dynamics - Bone: dyn_straps_br_03(ComponentPose=@n52, Alpha=1, ExternalForce="0,0,0") {
    angularDampingOverride=0.999
    angularSpringConstant=100
    bAngularSpring=true
    bOverrideAngularDamping=true
    bUseGravityOverride=true
    simulationSpace=RootRelative
    alphaBoolBlend {blendOption=Linear}
    boundBone {boneName=dyn_straps_br_03}
    gravityOverride {z=-980}
    physicsBodyDefinitions [{
      boundBone {boneName=dyn_straps_br_03}
      boxExtents=<15 15 15>
      constraintSetup {angularLimitsMax=<0 30 0>, angularLimitsMin=<0 0 -30>, angularTarget=<1 0 0>}
      localJointOffset=<2 0 0>
     }]
   }
   n54 = Anim Dynamics - Bone: dyn_straps_fl_03(ComponentPose=@n53, Alpha=1, ExternalForce="0,0,0") {
    angularDampingOverride=0.999
    angularSpringConstant=100
    bAngularSpring=true
    bOverrideAngularDamping=true
    bUseGravityOverride=true
    simulationSpace=RootRelative
    alphaBoolBlend {blendOption=Linear}
    boundBone {boneName=dyn_straps_fl_03}
    gravityOverride {z=-980}
    physicsBodyDefinitions [{
      boundBone {boneName=dyn_straps_fl_03}
      boxExtents=<15 15 15>
      constraintSetup {angularLimitsMax=<0 0 20>, angularLimitsMin=<0 -30 0>, angularTarget=<1 0 0>}
      localJointOffset=<2 0 0>
     }]
   }
   n55 = Anim Dynamics - Bone: dyn_straps_fl_04(ComponentPose=@n54, Alpha=1, ExternalForce="0,0,0") {
    angularDampingOverride=0.999
    angularSpringConstant=100
    bAngularSpring=true
    bOverrideAngularDamping=true
    bUseGravityOverride=true
    simulationSpace=RootRelative
    alphaBoolBlend {blendOption=Linear}
    boundBone {boneName=dyn_straps_fl_04}
    gravityOverride {z=-980}
    physicsBodyDefinitions [{
      boundBone {boneName=dyn_straps_fl_04}
      boxExtents=<15 15 15>
      constraintSetup {angularLimitsMax=<0 0 30>, angularLimitsMin=<0 0 0>, angularTarget=<1 0 0>}
      localJointOffset=<2 0 0>
     }]
   }
   n56 = Anim Dynamics - Bone: dyn_straps_bl_01(ComponentPose=@n55, Alpha=1, ExternalForce="0,0,0") {
    angularDampingOverride=0.999
    angularSpringConstant=100
    bAngularSpring=true
    bOverrideAngularDamping=true
    bUseGravityOverride=true
    simulationSpace=RootRelative
    alphaBoolBlend {blendOption=Linear}
    boundBone {boneName=dyn_straps_bl_01}
    gravityOverride {z=-980}
    physicsBodyDefinitions [{
      boundBone {boneName=dyn_straps_bl_01}
      boxExtents=<15 15 15>
      constraintSetup {angularLimitsMax=<0 0 0>, angularLimitsMin=<0 -30 -30>, angularTarget=<1 0 0>}
      localJointOffset=<2 0 0>
     }]
   }
   n57 = Anim Dynamics - Bone: dyn_straps_bl_02(ComponentPose=@n56, Alpha=1, ExternalForce="0,0,0") {
    angularDampingOverride=0.999
    angularSpringConstant=100
    bAngularSpring=true
    bOverrideAngularDamping=true
    bUseGravityOverride=true
    simulationSpace=RootRelative
    alphaBoolBlend {blendOption=Linear}
    boundBone {boneName=dyn_straps_bl_02}
    gravityOverride {z=-980}
    physicsBodyDefinitions [{
      boundBone {boneName=dyn_straps_bl_02}
      boxExtents=<15 15 15>
      constraintSetup {angularLimitsMax=<0 0 0>, angularLimitsMin=<0 -30 -30>, angularTarget=<1 0 0>}
      localJointOffset=<2 0 0>
     }]
   }
   n58 = Anim Dynamics - Bone: dyn_straps_bl_03(ComponentPose=@n57, Alpha=1, ExternalForce="0,0,0") {
    angularDampingOverride=0.999
    angularSpringConstant=100
    bAngularSpring=true
    bOverrideAngularDamping=true
    bUseGravityOverride=true
    simulationSpace=RootRelative
    alphaBoolBlend {blendOption=Linear}
    boundBone {boneName=dyn_straps_bl_03}
    gravityOverride {z=-980}
    physicsBodyDefinitions [{
      boundBone {boneName=dyn_straps_bl_03}
      boxExtents=<15 15 15>
      constraintSetup {angularLimitsMax=<0 0 0>, angularLimitsMin=<0 -30 -30>, angularTarget=<1 0 0>}
      localJointOffset=<2 0 0>
     }]
   }
   n59 = "Constraint - dyn_gun_master (PPPPPP)"!(
    ComponentPose=@n58
    Alpha=1
    ConstraintWeights_0=0.0024
    ConstraintWeights_1:float=0.1372
    ConstraintWeights_2:float=0.2991
    ConstraintWeights_3:float=0.1435
    ConstraintWeights_4:float=0.4101
    ConstraintWeights_5:float=0.0077
   ) -> (Pose) {
    alphaBoolBlend {blendOption=Linear}
    boneToModify {boneName=dyn_gun_master}
    constraintSetup [
     {targetBone {boneName=dyn_straps_fl_03}}
     {targetBone {boneName=dyn_straps_fl_05}}
     {targetBone {boneName=dyn_straps_fl_04}}
     {targetBone {boneName=dyn_straps_bl_04}}
     {targetBone {boneName=dyn_straps_bl_03}}
     {targetBone {boneName=pelvis}}
    ]
    constraintWeights [1,1,1,1,1,1]
   }
   n60 = Anim Dynamics - Bone: dyn_gun(ComponentPose=@n59, Alpha=1, GravityOverride="0,0,0", ExternalForce="0,0,0") {
    angularDampingOverride=1
    angularSpringConstant=1500
    bAngularSpring=true
    bGravityOverrideInSimSpace=true
    bOverrideAngularDamping=true
    bUsePlanarLimit=false
    bUseSphericalLimits=true
    gravityScale=0.5
    simulationSpace=RootRelative
    alphaBoolBlend {blendOption=Linear}
    boundBone {boneName=dyn_gun}
    componentAppliedLinearAccClamp=<750 750 750>
    componentLinearAccScale {z=0.2}
    componentLinearVelScale {z=0.2}
    physicsBodyDefinitions [{
      collisionType=CustomSphere
      sphereCollisionRadius=1
      boundBone {boneName=dyn_gun}
      boxExtents=<21 21 21>
      constraintSetup {angularLimitsMax=<20 10 20>, angularLimitsMin=<-20 -20 -20>, angularTarget=<0 0 0>}
      localJointOffset=<10 -0.5 0>
     }]
    relativeSpaceBone {boneName=root}
   }
   Save cached pose 'PostSim'(Pose=Component To Local(ComponentPose=@n60)) {cacheName=PostSim}
  ]}
 EventGraph {nodes [
   Event BlueprintUpdateAnimation()
   > n1 = Sequence!(execute) -> (then_0, then_1, then_2:Exec)
   n2 = Branch(execute=@n1.then_2, Condition=AND Boolean(A=Get bIsInFrontEnd(), B=Get bIsPlayingEmote()))
   > Set StrapDynAlpha(StrapDynAlpha:double=0.25) -> (Output_Get:double)
   Set StrapDynAlpha(execute=@n2.else, StrapDynAlpha=0.1)
   Set Jump Falling(execute=@n1.then_0, Jump Falling:bool=OR Boolean(A=Get bIsJumping(), B=Get bIsFalling())) -> (Output_Get:bool)
   > n3 = Branch(Condition=Get Jump Falling() -> (Jump Falling:bool))
   n4 = BreakTransform(InTransform=GetDeltaTransformFromRefPose(self=GetOwnerSkeletalMeshComponent(), BoneName=thigh_r, BaseName=pelvis))
   > Set Jump Alpha(Jump Alpha:double=MapRangeClamped(Value=BreakRotator(InRot=@n4.Rotation).Yaw, InRangeA=10, InRangeB=-100)) -> (Output_Get:double)
   Set Jump Alpha(execute=@n3.else, Jump Alpha=0)
   n5 = Branch(execute=@n1.then_1, Condition=Get bIsCrouching())
   > Set CrouchAlpha(CrouchAlpha:double=1) -> (Output_Get:double)
   n6 = BreakTransform(InTransform=GetDeltaTransformFromRefPose(self=GetOwnerSkeletalMeshComponent(), BoneName=thigh_r, BaseName=pelvis))
   > Set Thigh R Crouch Leather Rot Alpha(Thigh R Crouch Leather Rot Alpha:double=MapRangeClamped(Value=BreakRotator(InRot=@n6.Rotation).Yaw, InRangeA=-25, InRangeB=-105)) -> (Output_Get:double)
   Set CrouchAlpha(execute=@n5.else, CrouchAlpha=0)
   > Set Thigh R Crouch Leather Rot Alpha(Thigh R Crouch Leather Rot Alpha=0)
   Event OnEndPlayingEmote() {is=title K2Node_Event}
   > Delay(Duration=0.2)
   > ResetDynamics(InTeleportType=ResetPhysics)
  ]}
]
variables [
 CrouchAlpha {type=number}
 Custom Idle {type=number}
 frontend_not_emote {type=boolean}
 Jump Alpha {type=number}
 Jump Falling {type=boolean}
 "rban  disabled" {type=boolean}
 Skydiving {type=boolean}
 Spine05_twist {type=number}
 StrapDynAlpha {type=number}
 Thigh L Crouch Pitch Alpha {type=number}
 Thigh L Crouch Pitch In Alpha {type=number}
 Thigh L Crouch Rot Alpha {type=number}
 Thigh R Crouch Belt Rot Alpha {type=number}
 Thigh R Crouch Leather Rot Alpha {type=number}
 Thigh R Crouch Pitch Alpha {type=number}
 Thigh R Crouch Rot Roll Alpha {type=number}
]
