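# Physics asset for the braided hair of the female Uproar Braids character's face accessory mesh, giving two braids their collision and joint setup.
# Capsule bodies cover the head, a hair master bone, and left and right chains that each run from a master through ten braid bones.
# Upper joints, from the head through segment 05, swing and twist freely.
# Segments 06 to 010 have tight angular limits and mostly fatter capsules, stiffening the lower braids.
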PhysicsAsset /BRCosmetics/Characters/Player/Female/Medium/Bodies/F_MED_Uproar_Braids/Meshes/Parts/F_MED_Uproar_Braids_FaceAcc_Physics
bodies {
 dyn_Hair_Braid_MSTR [{
   shapeType=Capsule
   length=6.6556
   radius=2.4865
   shapeName=dyn_Hair_Braid_MSTR_capsule
   center=<-0.9124 -2.7495 0.1839>
   rotation=<rot -0.8346 -90 4.5755>
  }]
 dyn_l_braid_01 [{
   shapeType=Capsule
   length=4.3943
   radius=3.5451
   shapeName=dyn_l_braid_01_capsule
   center=<4.5518 -0.8108 -0.7484>
   rotation=<rot 0 -87.4534 -84.7177>
  }]
 dyn_l_braid_010 [{shapeType=Capsule, length=10.8202, radius=6.6723, shapeName=dyn_l_braid_010_capsule, center=<8.9345 0 -0.0001>, rotation=<rot 90 0 0>}]
 dyn_l_braid_02 [{
   shapeType=Capsule
   length=5.3664
   radius=2.4336
   shapeName=dyn_l_braid_02_capsule
   center=<5.2226 -0.25 0.066>
   rotation=<rot 0 89.4982 -89.1636>
  }]
 dyn_l_braid_03 [{
   shapeType=Capsule
   length=6.8743
   radius=2.221
   shapeName=dyn_l_braid_03_capsule
   center=<6.8916 -0.1234 0.0249>
   rotation=<rot 0 -90.2467 -89.8915>
  }]
 dyn_l_braid_04 [{
   shapeType=Capsule
   length=6.1601
   radius=2.1324
   shapeName=dyn_l_braid_04_capsule
   center=<6.1827 0.0259 -0.031>
   rotation=<rot 0 -92.5787 -88.2269>
  }]
 dyn_l_braid_05 [{
   shapeType=Capsule
   length=6.7809
   radius=2.2498
   shapeName=dyn_l_braid_05_capsule
   center=<7.0037 -0.0767 -0.0465>
   rotation=<rot 0 89.7017 -89.576>
  }]
 dyn_l_braid_06 [{shapeType=Capsule, length=3.3095, radius=5.3395, shapeName=dyn_l_braid_06_capsule, center=<8.9345 0 -0.0001>, rotation=<rot 90 0 0>}]
 dyn_l_braid_07 [{shapeType=Capsule, length=3.2806, radius=5.4435, shapeName=dyn_l_braid_07_capsule, center=<8.9345 0 -0.0001>, rotation=<rot 90 0 0>}]
 dyn_l_braid_08 [{shapeType=Capsule, length=4.4217, radius=5.4287, shapeName=dyn_l_braid_08_capsule, center=<8.9345 0 -0.0001>, rotation=<rot 90 0 0>}]
 dyn_l_braid_09 [{shapeType=Capsule, length=3.1237, radius=5.3734, shapeName=dyn_l_braid_09_capsule, center=<8.9345 0 -0.0001>, rotation=<rot 90 0 0>}]
 dyn_l_braid_MSTR [{
   shapeType=Capsule
   length=4.3208
   radius=0.9627
   shapeName=dyn_l_braid_MSTR_capsule
   center=<0.9839 -0.6962 -0.1191>
   rotation=<rot 0 -16.3447 -48.6398>
  }]
 dyn_r_braid_01 [{
   shapeType=Capsule
   length=5.2966
   radius=3.8444
   shapeName=dyn_r_braid_01_capsule
   center=<3.6837 0.6379 -0.7614>
   rotation=<rot 0 99.3919 -86.7593>
  }]
 dyn_r_braid_010 [{shapeType=Capsule, length=10.7565, radius=6.5709, shapeName=dyn_r_braid_010_capsule, center=<8.9345 0 -0.0001>, rotation=<rot 90 0 0>}]
 dyn_r_braid_02 [{
   shapeType=Capsule
   length=5.4498
   radius=2.6444
   shapeName=dyn_r_braid_02_capsule
   center=<5.3093 0.1117 -0.0159>
   rotation=<rot 0 88.1864 -88.6425>
  }]
 dyn_r_braid_03 [{
   shapeType=Capsule
   length=6.8077
   radius=2.196
   shapeName=dyn_r_braid_03_capsule
   center=<7.0181 0.0085 0.0128>
   rotation=<rot 0 89.7896 -89.9552>
  }]
 dyn_r_braid_04 [{
   shapeType=Capsule
   length=5.9167
   radius=2.3569
   shapeName=dyn_r_braid_04_capsule
   center=<6.3233 0.0538 0.2585>
   rotation=<rot 0 -88.728 -88.9467>
  }]
 dyn_r_braid_05 [{
   shapeType=Capsule
   length=6.9533
   radius=2.3355
   shapeName=dyn_r_braid_05_capsule
   center=<7.0129 0.0247 0.288>
   rotation=<rot 0 90.3761 -89.6679>
  }]
 dyn_r_braid_06 [{shapeType=Capsule, length=3.4723, radius=5.3531, shapeName=dyn_r_braid_06_capsule, center=<8.9345 0 -0.0001>, rotation=<rot 90 0 0>}]
 dyn_r_braid_07 [{shapeType=Capsule, length=2.5744, radius=5.3926, shapeName=dyn_r_braid_07_capsule, center=<8.9345 0 -0.0001>, rotation=<rot 90 0 0>}]
 dyn_r_braid_08 [{shapeType=Capsule, length=3.0216, radius=5.3344, shapeName=dyn_r_braid_08_capsule, center=<8.9345 0 -0.0001>, rotation=<rot 90 0 0>}]
 dyn_r_braid_09 [{shapeType=Capsule, length=3.1086, radius=5.2138, shapeName=dyn_r_braid_09_capsule, center=<8.9345 0 -0.0001>, rotation=<rot 90 0 0>}]
 dyn_r_braid_MSTR [{
   shapeType=Capsule
   length=4.4066
   radius=1.8521
   shapeName=dyn_r_braid_MSTR_capsule
   center=<1.0193 0.255 -0.0281>
   rotation=<rot 40.93 -90 -12.3523>
  }]
 head [{shapeType=Capsule, length=9.2143, radius=8.9361, center=<8.2663 -0.6444 -0.3977>, rotation=<rot 76.5285 70.3006 160.8025>}]
}
constraints [
 {
  bone1Name=dyn_Hair_Braid_MSTR
  bone2Name=head
  swing1LimitDegrees=45
  swing1Motion=Free
  swing2LimitDegrees=45
  swing2Motion=Free
  twistLimitDegrees=45
  twistMotion=Free
 }
 {
  bone1Name=dyn_r_braid_MSTR
  bone2Name=dyn_Hair_Braid_MSTR
  swing1LimitDegrees=45
  swing1Motion=Free
  swing2LimitDegrees=45
  swing2Motion=Free
  twistLimitDegrees=45
  twistMotion=Free
 }
 {
  bone1Name=dyn_r_braid_01
  bone2Name=dyn_r_braid_MSTR
  swing1LimitDegrees=45
  swing1Motion=Free
  swing2LimitDegrees=45
  swing2Motion=Free
  twistLimitDegrees=45
  twistMotion=Free
 }
 {
  bone1Name=dyn_r_braid_02
  bone2Name=dyn_r_braid_01
  swing1LimitDegrees=45
  swing1Motion=Free
  swing2LimitDegrees=45
  swing2Motion=Free
  twistLimitDegrees=45
  twistMotion=Free
 }
 {
  bone1Name=dyn_r_braid_03
  bone2Name=dyn_r_braid_02
  swing1LimitDegrees=45
  swing1Motion=Free
  swing2LimitDegrees=45
  swing2Motion=Free
  twistLimitDegrees=45
  twistMotion=Free
 }
 {
  bone1Name=dyn_r_braid_04
  bone2Name=dyn_r_braid_03
  swing1LimitDegrees=45
  swing1Motion=Free
  swing2LimitDegrees=45
  swing2Motion=Free
  twistLimitDegrees=45
  twistMotion=Free
 }
 {
  bone1Name=dyn_r_braid_05
  bone2Name=dyn_r_braid_04
  swing1LimitDegrees=45
  swing1Motion=Free
  swing2LimitDegrees=45
  swing2Motion=Free
  twistLimitDegrees=45
  twistMotion=Free
 }
 {bone1Name=dyn_r_braid_06, bone2Name=dyn_r_braid_05, swing1LimitDegrees=15, swing2LimitDegrees=15, twistLimitDegrees=5}
 {bone1Name=dyn_r_braid_07, bone2Name=dyn_r_braid_06, swing1LimitDegrees=15, swing2LimitDegrees=15, twistLimitDegrees=5}
 {bone1Name=dyn_r_braid_08, bone2Name=dyn_r_braid_07, swing1LimitDegrees=15, swing2LimitDegrees=15, twistLimitDegrees=5}
 {bone1Name=dyn_r_braid_09, bone2Name=dyn_r_braid_08, swing1LimitDegrees=15, swing2LimitDegrees=15, twistLimitDegrees=5}
 {bone1Name=dyn_r_braid_010, bone2Name=dyn_r_braid_09, swing1LimitDegrees=15, swing2LimitDegrees=15, twistLimitDegrees=5}
 {
  bone1Name=dyn_l_braid_MSTR
  bone2Name=dyn_Hair_Braid_MSTR
  swing1LimitDegrees=45
  swing1Motion=Free
  swing2LimitDegrees=45
  swing2Motion=Free
  twistLimitDegrees=45
  twistMotion=Free
 }
 {
  bone1Name=dyn_l_braid_01
  bone2Name=dyn_l_braid_MSTR
  swing1LimitDegrees=45
  swing1Motion=Free
  swing2LimitDegrees=45
  swing2Motion=Free
  twistLimitDegrees=45
  twistMotion=Free
 }
 {
  bone1Name=dyn_l_braid_02
  bone2Name=dyn_l_braid_01
  swing1LimitDegrees=45
  swing1Motion=Free
  swing2LimitDegrees=45
  swing2Motion=Free
  twistLimitDegrees=45
  twistMotion=Free
 }
 {
  bone1Name=dyn_l_braid_03
  bone2Name=dyn_l_braid_02
  swing1LimitDegrees=45
  swing1Motion=Free
  swing2LimitDegrees=45
  swing2Motion=Free
  twistLimitDegrees=45
  twistMotion=Free
 }
 {
  bone1Name=dyn_l_braid_04
  bone2Name=dyn_l_braid_03
  swing1LimitDegrees=45
  swing1Motion=Free
  swing2LimitDegrees=45
  swing2Motion=Free
  twistLimitDegrees=45
  twistMotion=Free
 }
 {
  bone1Name=dyn_l_braid_05
  bone2Name=dyn_l_braid_04
  swing1LimitDegrees=45
  swing1Motion=Free
  swing2LimitDegrees=45
  swing2Motion=Free
  twistLimitDegrees=45
  twistMotion=Free
 }
 {bone1Name=dyn_l_braid_06, bone2Name=dyn_l_braid_05, swing1LimitDegrees=15, swing2LimitDegrees=15, twistLimitDegrees=5}
 {bone1Name=dyn_l_braid_07, bone2Name=dyn_l_braid_06, swing1LimitDegrees=15, swing2LimitDegrees=15, twistLimitDegrees=5}
 {bone1Name=dyn_l_braid_08, bone2Name=dyn_l_braid_07, swing1LimitDegrees=15, swing2LimitDegrees=15, twistLimitDegrees=5}
 {bone1Name=dyn_l_braid_09, bone2Name=dyn_l_braid_08, swing1LimitDegrees=15, swing2LimitDegrees=15, twistLimitDegrees=5}
 {bone1Name=dyn_l_braid_010, bone2Name=dyn_l_braid_09, swing1LimitDegrees=15, swing2LimitDegrees=15, twistLimitDegrees=5}
]
