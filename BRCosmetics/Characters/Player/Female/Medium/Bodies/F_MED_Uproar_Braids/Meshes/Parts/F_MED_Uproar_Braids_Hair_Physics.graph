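# Physics asset for the hair part of the F_MED_Uproar_Braids female player cosmetic, giving the hair bones capsule bodies for physical motion.
# A front group and a top group branch from dyn_Hair_Fr_MSTR, each with a base chain and shorter side chains splitting off its first bone.
# Tight swing and twist limit angles are set along each chain.
# An extra constraint with free swing and twist cross-links the two groups' first base bones.
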PhysicsAsset /BRCosmetics/Characters/Player/Female/Medium/Bodies/F_MED_Uproar_Braids/Meshes/Parts/F_MED_Uproar_Braids_Hair_Physics
bodies {
 dyn_Hair_Fr_MSTR [{shapeType=Capsule, length=3.4481, radius=3.0321, shapeName=dyn_Hair_Fr_MSTR_capsule, rotation=<rot 90 0 0>}]
 dyn_hair_fr_a_01 [{shapeType=Capsule, length=2.75, radius=2.25, shapeName=dyn_hair_fr_a_01_capsule, center=<2.6622 1.2461 0>, rotation=<rot 90 0 0>}]
 dyn_hair_fr_a_02 [{shapeType=Capsule, length=3.75, radius=2.75, shapeName=dyn_hair_fr_a_02_capsule, center=<2.3128 1.2461 0>, rotation=<rot 90 0 0>}]
 dyn_hair_fr_base_01 [{shapeType=Capsule, length=5, radius=2.25, shapeName=dyn_hair_fr_base_01_capsule, center=<3.8279 -0.7421 -0.2834>, rotation=<rot 90 0 0>}]
 dyn_hair_fr_base_02 [{shapeType=Capsule, length=6, radius=3, shapeName=dyn_hair_fr_base_02_capsule, center=<3.7879 -0.742 0.6183>, rotation=<rot 90 0 0>}]
 dyn_hair_fr_base_03 [{shapeType=Capsule, length=2.75, radius=2.75, shapeName=dyn_hair_fr_base_03_capsule, center=<1.8748 -0.742 0.6183>, rotation=<rot 90 0 0>}]
 dyn_hair_top_a_01 [{shapeType=Capsule, length=2.75, radius=1.75, shapeName=dyn_hair_top_a_01_capsule, center=<2.2761 0 0.5264>, rotation=<rot 90 0 0>}]
 dyn_hair_top_a_02 [{shapeType=Capsule, length=2, radius=1.75, shapeName=dyn_hair_top_a_02_capsule, center=<1.7412 0 0.5264>, rotation=<rot 90 0 0>}]
 dyn_hair_top_b_01 [{shapeType=Capsule, length=3.25, radius=2.25, shapeName=dyn_hair_top_b_01_capsule, center=<3.3234 0 -0.6993>, rotation=<rot 90 0 0>}]
 dyn_hair_top_b_02 [{shapeType=Capsule, length=2, radius=2, shapeName=dyn_hair_top_b_02_capsule, center=<1.7149 0 -0.6993>, rotation=<rot 90 0 0>}]
 dyn_hair_top_base_01 [{shapeType=Capsule, length=3.5, radius=2.5, shapeName=dyn_hair_top_base_01_capsule, center=<3.6838 0 -0.5615>, rotation=<rot 90 0 0>}]
 dyn_hair_top_base_02 [{shapeType=Capsule, length=3.5, radius=2.25, shapeName=dyn_hair_top_base_02_capsule, center=<2.8727 0 1.1475>, rotation=<rot 90 0 0>}]
 dyn_hair_top_base_03 [{shapeType=Capsule, length=2.5, radius=2, shapeName=dyn_hair_top_base_03_capsule, center=<1.9324 0 1.1476>, rotation=<rot 90 0 0>}]
}
constraints [
 {bone1Name=dyn_hair_fr_base_01, bone2Name=dyn_Hair_Fr_MSTR, swing1LimitDegrees=20, swing2LimitDegrees=15, twistLimitDegrees=5}
 {bone1Name=dyn_hair_fr_base_02, bone2Name=dyn_hair_fr_base_01, swing1LimitDegrees=15, swing2LimitDegrees=25, twistLimitDegrees=5}
 {bone1Name=dyn_hair_fr_base_03, bone2Name=dyn_hair_fr_base_02, swing1LimitDegrees=15, swing2LimitDegrees=15, twistLimitDegrees=5}
 {bone1Name=dyn_hair_fr_a_01, bone2Name=dyn_hair_fr_base_01, swing1LimitDegrees=15, swing2LimitDegrees=25, twistLimitDegrees=5}
 {bone1Name=dyn_hair_fr_a_02, bone2Name=dyn_hair_fr_a_01, swing1LimitDegrees=15, swing2LimitDegrees=15, twistLimitDegrees=5}
 {bone1Name=dyn_hair_top_base_01, bone2Name=dyn_Hair_Fr_MSTR, swing1LimitDegrees=30, swing2LimitDegrees=15, twistLimitDegrees=5}
 {bone1Name=dyn_hair_top_base_02, bone2Name=dyn_hair_top_base_01, swing1LimitDegrees=15, swing2LimitDegrees=15, twistLimitDegrees=5}
 {bone1Name=dyn_hair_top_base_03, bone2Name=dyn_hair_top_base_02, swing1LimitDegrees=15, swing2LimitDegrees=15, twistLimitDegrees=5}
 {bone1Name=dyn_hair_top_a_01, bone2Name=dyn_hair_top_base_01, swing1LimitDegrees=15, swing2LimitDegrees=15, twistLimitDegrees=5}
 {bone1Name=dyn_hair_top_a_02, bone2Name=dyn_hair_top_a_01, swing1LimitDegrees=15, swing2LimitDegrees=15, twistLimitDegrees=5}
 {bone1Name=dyn_hair_top_b_01, bone2Name=dyn_hair_top_base_01, swing1LimitDegrees=15, swing2LimitDegrees=15, twistLimitDegrees=5}
 {bone1Name=dyn_hair_top_b_02, bone2Name=dyn_hair_top_b_01, swing1LimitDegrees=15, swing2LimitDegrees=15, twistLimitDegrees=5}
 {
  bone1Name=dyn_hair_fr_base_01
  bone2Name=dyn_hair_top_base_01
  swing1LimitDegrees=45
  swing1Motion=Free
  swing2LimitDegrees=45
  swing2Motion=Free
  twistLimitDegrees=45
  twistMotion=Free
 }
]
