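# Post-process-style secondary motion AnimBP for the Uproar Braids female cosmetic, animating hair and two long braids.
# It copies the parent mesh pose, runs head-space RigidBody hair, offsets a braid look-at bone by player state, then caches the result.
# Control Rig feeds per-bone AnimDynamics braid chains, using spherical limits on spine and pelvis against clipping.
# Bool blends select skydive, cape and DBNO variants, or Control Rig alone without dynamics.
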AnimBlueprint /BRCosmetics/Characters/Player/Female/Medium/Bodies/F_MED_Uproar_Braids/Meshes/Parts/F_MED_Uproar_Braids_FaceAcc_AnimBP
target_skeleton=&/F_MED_Uproar_Braids_FaceAcc_Skeleton
graphs [
 AnimGraph {nodes [
   "Constraint - dyn_braids_lookat (T)"(Alpha=1, ConstraintWeights_0=1) {
    alphaBoolBlend {blendOption=Linear}
    boneToModify {boneName=dyn_braids_lookat}
    constraintSetup [{transformType=Translation, targetBone {boneName=pelvis}}]
    constraintWeights [1]
   }
   n1 = "Transform (Modify) Bone - Bone: dyn_r_braid_MSTR"(Alpha=1, Translation="0,0,0", Rotation="-10,-15,-0.007", Scale="0.9,1,1") {
    rotationMode=BMM_Additive
    rotationSpace=BCS_BoneSpace
    translationMode=BMM_Additive
    alphaBoolBlend {blendOption=Linear}
    boneToModify {boneName=dyn_r_braid_MSTR}
    rotation=<rot -10 -15 -0.007>
    scale {x=0.9}
   }
   n2 = "Transform (Modify) Bone - Bone: dyn_l_braid_MSTR"(ComponentPose=@n1, Alpha=1, Translation="0,0,0", Rotation="-15,0,0", Scale="0.9,1,1") {
    rotationMode=BMM_Additive
    rotationSpace=BCS_BoneSpace
    translationMode=BMM_Additive
    alphaBoolBlend {blendOption=Linear}
    boneToModify {boneName=dyn_l_braid_MSTR}
    rotation {pitch=-15}
    scale {x=0.9}
   }
   n3 = "Transform (Modify) Bone - Bone: dyn_r_braid_03"(ComponentPose=@n2, Alpha=1, Translation="0,0,0", Rotation="-10,0,0", Scale="0.9,1,1") {
    rotationMode=BMM_Additive
    rotationSpace=BCS_BoneSpace
    translationMode=BMM_Additive
    alphaBoolBlend {blendOption=Linear}
    boneToModify {boneName=dyn_r_braid_03}
    rotation {pitch=-10}
    scale {x=0.9}
   }
   "Transform (Modify) Bone - Bone: dyn_l_braid_03"(ComponentPose=@n3, Alpha=1, Translation="0,0,0", Rotation="-10,5.335,0", Scale="0.9,1,1") {
    rotationMode=BMM_Additive
    rotationSpace=BCS_BoneSpace
    translationMode=BMM_Additive
    alphaBoolBlend {blendOption=Linear}
    boneToModify {boneName=dyn_l_braid_03}
    rotation {pitch=-10, yaw=5.335}
    scale {x=0.9}
   }
   n4 = "Transform (Modify) Bone - Bone: dyn_r_braid_05"(Alpha=1, Translation="0,0,0", Rotation="0,0,0", Scale="0.9,1,1") {
    rotationMode=BMM_Additive
    rotationSpace=BCS_BoneSpace
    translationMode=BMM_Additive
    alphaBoolBlend {blendOption=Linear}
    boneToModify {boneName=dyn_r_braid_05}
    scale {x=0.9}
   }
   "Transform (Modify) Bone - Bone: dyn_l_braid_05"(ComponentPose=@n4, Alpha=1, Translation="0,0,0", Rotation="0,0,0", Scale="0.9,1,1") {
    rotationMode=BMM_Additive
    rotationSpace=BCS_BoneSpace
    translationMode=BMM_Additive
    alphaBoolBlend {blendOption=Linear}
    boneToModify {boneName=dyn_l_braid_05}
    scale {x=0.9}
   }
   n5 = Use cached pose 'rigids disabled'()
   n6 = Blend Poses by bool(
    BlendPose_0=Component To Local(ComponentPose=Local To Component(LocalPose=@n5))
    BlendPose_1=Use cached pose 'rigids disabled'()
    bActiveValue=Get bIsInFrontEnd()
   ) {blendType=Linear}
   n7 = Use cached pose 'CapePhysics'()
   n8 = Get WearingCape() -> (WearingCape:bool)
   n9 = Control Rig(Source=Use cached pose 'Prebraid sim'()) {
    alphaBoolBlend {blendOption=Linear}
    controlRigAssetReference {blueprintRigClass=&/F_MED_Uproar_Braids_FaceAcc_CtrlRig.F_MED_Uproar_Braids_FaceAcc_CtrlRig_C}
   }
   n10 = "Transform (Modify) Bone - Bone: dyn_r_braid_04"(ComponentPose=Local To Component(LocalPose=@n9), Alpha=1, Translation="0,0,0", Rotation="0,0,0", Scale="0.9,1,1") {
    rotationMode=BMM_Additive
    translationMode=BMM_Additive
    alphaBoolBlend {blendOption=Linear}
    boneToModify {boneName=dyn_r_braid_04}
    scale {x=0.9}
   }
   n11 = "Transform (Modify) Bone - Bone: dyn_r_braid_06"(ComponentPose=@n10, Alpha=1, Translation="0,0,0", Rotation="0,0,0", Scale="0.9,1,1") {
    rotationMode=BMM_Additive
    translationMode=BMM_Additive
    alphaBoolBlend {blendOption=Linear}
    boneToModify {boneName=dyn_r_braid_06}
    scale {x=0.9}
   }
   n12 = "Transform (Modify) Bone - Bone: dyn_r_braid_08"(ComponentPose=@n11, Alpha=1, Translation="0,0,0", Rotation="0,0,0", Scale="0.9,1,1") {
    rotationMode=BMM_Additive
    translationMode=BMM_Additive
    alphaBoolBlend {blendOption=Linear}
    boneToModify {boneName=dyn_r_braid_08}
    scale {x=0.9}
   }
   n13 = "Transform (Modify) Bone - Bone: dyn_r_braid_010"(ComponentPose=@n12, Alpha=1, Translation="0,0,0", Rotation="0,0,0", Scale="0.9,1,1") {
    rotationMode=BMM_Additive
    translationMode=BMM_Additive
    alphaBoolBlend {blendOption=Linear}
    boneToModify {boneName=dyn_r_braid_010}
    scale {x=0.9}
   }
   n14 = "Transform (Modify) Bone - Bone: dyn_l_braid_04"(ComponentPose=@n13, Alpha=1, Translation="0,0,0", Rotation="0,0,0", Scale="0.9,1,1") {
    rotationMode=BMM_Additive
    translationMode=BMM_Additive
    alphaBoolBlend {blendOption=Linear}
    boneToModify {boneName=dyn_l_braid_04}
    scale {x=0.9}
   }
   n15 = "Transform (Modify) Bone - Bone: dyn_l_braid_06"(ComponentPose=@n14, Alpha=1, Translation="0,0,0", Rotation="0,0,0", Scale="0.9,1,1") {
    rotationMode=BMM_Additive
    translationMode=BMM_Additive
    alphaBoolBlend {blendOption=Linear}
    boneToModify {boneName=dyn_l_braid_06}
    scale {x=0.9}
   }
   n16 = "Transform (Modify) Bone - Bone: dyn_l_braid_08"(ComponentPose=@n15, Alpha=1, Translation="0,0,0", Rotation="0,0,0", Scale="0.9,1,1") {
    rotationMode=BMM_Additive
    translationMode=BMM_Additive
    alphaBoolBlend {blendOption=Linear}
    boneToModify {boneName=dyn_l_braid_08}
    scale {x=0.9}
   }
   n17 = "Transform (Modify) Bone - Bone: dyn_l_braid_010"(ComponentPose=@n16, Alpha=1, Translation="0,0,0", Rotation="0,0,0", Scale="0.9,1,1") {
    rotationMode=BMM_Additive
    translationMode=BMM_Additive
    alphaBoolBlend {blendOption=Linear}
    boneToModify {boneName=dyn_l_braid_010}
    scale {x=0.9}
   }
   n18 = Get simulateBase() -> (simulateBase:double)
   n19 = Anim Dynamics - Bone: dyn_l_braid_02!(ComponentPose=@n17, Alpha=@n18, GravityOverride:Vector="0,0,0", ExternalForce="0,0,0") -> (Pose) {
    angularDampingOverride=1
    angularSpringConstant=2500
    bAngularSpring=true
    bGravityOverrideInSimSpace=true
    bOverrideAngularDamping=true
    bUseGravityOverride=true
    bUseSphericalLimits=true
    gravityScale=0.35
    simulationSpace=RootRelative
    alphaBoolBlend {blendOption=Linear}
    alphaScaleBiasClamp {scale=0.4}
    boundBone {boneName=dyn_l_braid_02}
    componentAppliedLinearAccClamp=<5000 5000 5000>
    componentLinearAccScale {z=0.44}
    componentLinearVelScale {z=0.44}
    physicsBodyDefinitions [{
      collisionType=CustomSphere
      sphereCollisionRadius=1
      boundBone {boneName=dyn_l_braid_02}
      boxExtents=<12 22 22>
      constraintSetup {angularLimitsMax=<15 15 25>, angularLimitsMin=<15 -15 -5>, angularTarget=<0 0 0>}
      localJointOffset=<10 0 0>
     }]
    sphericalLimits [
     {limitRadius=60, drivingBone {boneName=spine_05}, sphereLocalOffset=<-12.8628 52 0>}
     {limitRadius=50, drivingBone {boneName=pelvis}, sphereLocalOffset=<-2.6826 38.079 0>}
     {limitRadius=28, drivingBone {boneName=spine_02}, sphereLocalOffset=<2.6549 15 0>}
     {limitRadius=1, drivingBone {boneName=pelvis}, sphereLocalOffset=<-30 0 0>}
    ]
   }
   n20 = Get GravityOverride()
   n21 = Anim Dynamics - Bone: dyn_l_braid_03!(ComponentPose=@n19, Alpha=@n18, GravityOverride:Vector=@n20, ExternalForce="0,0,0") -> (Pose) {
    angularDampingOverride=1
    angularSpringConstant=2500
    bAngularSpring=true
    bGravityOverrideInSimSpace=true
    bOverrideAngularDamping=true
    bUseGravityOverride=true
    bUseSphericalLimits=true
    gravityScale=0.35
    simulationSpace=RootRelative
    alphaBoolBlend {blendOption=Linear}
    alphaScaleBiasClamp {scale=0.4}
    boundBone {boneName=dyn_l_braid_03}
    componentAppliedLinearAccClamp=<5000 5000 5000>
    componentLinearAccScale {z=0.44}
    componentLinearVelScale {z=0.44}
    physicsBodyDefinitions [{
      collisionType=CustomSphere
      sphereCollisionRadius=1
      boundBone {boneName=dyn_l_braid_03}
      boxExtents=<12 22 22>
      constraintSetup {angularLimitsMax=<15 15 25>, angularLimitsMin=<15 -15 -5>, angularTarget=<0 0 0>}
      localJointOffset=<10 0 0>
     }]
    sphericalLimits [
     {limitRadius=60, drivingBone {boneName=spine_05}, sphereLocalOffset=<-12.8628 52 0>}
     {limitRadius=50, drivingBone {boneName=pelvis}, sphereLocalOffset=<-2.6826 38.079 0>}
     {limitRadius=28, drivingBone {boneName=spine_02}, sphereLocalOffset=<2.6549 15 0>}
     {limitRadius=1, drivingBone {boneName=pelvis}, sphereLocalOffset=<-30 0 0>}
    ]
   }
   n22 = Anim Dynamics - Bone: dyn_l_braid_04!(ComponentPose=@n21, Alpha=1, GravityOverride:Vector=@n20, ExternalForce="0,0,0") -> (Pose) {
    angularDampingOverride=1
    angularSpringConstant=2500
    bAngularSpring=true
    bGravityOverrideInSimSpace=true
    bOverrideAngularDamping=true
    bUseGravityOverride=true
    bUseSphericalLimits=true
    gravityScale=0.35
    simulationSpace=RootRelative
    alphaBoolBlend {blendOption=Linear}
    alphaScaleBiasClamp {scale=0.4}
    boundBone {boneName=dyn_l_braid_04}
    componentAppliedLinearAccClamp=<5000 5000 5000>
    componentLinearAccScale {z=0.44}
    componentLinearVelScale {z=0.44}
    physicsBodyDefinitions [{
      collisionType=CustomSphere
      sphereCollisionRadius=1
      boundBone {boneName=dyn_l_braid_04}
      boxExtents=<12 22 22>
      constraintSetup {angularLimitsMax=<15 15 25>, angularLimitsMin=<15 -15 -5>, angularTarget=<0 0 0>}
      localJointOffset=<10 0 0>
     }]
    sphericalLimits [
     {limitRadius=60, drivingBone {boneName=spine_05}, sphereLocalOffset=<-12.8628 52 0>}
     {limitRadius=50, drivingBone {boneName=pelvis}, sphereLocalOffset=<-2.6826 38.079 0>}
     {limitRadius=28, drivingBone {boneName=spine_02}, sphereLocalOffset=<2.6549 15 0>}
     {limitRadius=1, drivingBone {boneName=pelvis}, sphereLocalOffset=<-30 0 0>}
    ]
   }
   n23 = Anim Dynamics - Bone: dyn_l_braid_05!(ComponentPose=@n22, Alpha=1, GravityOverride:Vector=@n20, ExternalForce="0,0,0") -> (Pose) {
    angularDampingOverride=1
    angularSpringConstant=2500
    bAngularSpring=true
    bGravityOverrideInSimSpace=true
    bOverrideAngularDamping=true
    bUseGravityOverride=true
    bUseSphericalLimits=true
    gravityScale=0.35
    simulationSpace=RootRelative
    alphaBoolBlend {blendOption=Linear}
    alphaScaleBiasClamp {scale=0.4}
    boundBone {boneName=dyn_l_braid_05}
    componentAppliedLinearAccClamp=<5000 5000 5000>
    componentLinearAccScale {z=0.44}
    componentLinearVelScale {z=0.44}
    physicsBodyDefinitions [{
      collisionType=CustomSphere
      sphereCollisionRadius=1
      boundBone {boneName=dyn_l_braid_05}
      boxExtents=<12 22 22>
      constraintSetup {angularLimitsMax=<15 15 25>, angularLimitsMin=<15 -15 -5>, angularTarget=<0 0 0>}
      localJointOffset=<10 0 0>
     }]
    sphericalLimits [
     {limitRadius=65.9176, drivingBone {boneName=spine_05}, sphereLocalOffset=<-12.8628 52 0>}
     {limitRadius=50, drivingBone {boneName=pelvis}, sphereLocalOffset=<-2.6826 38.079 0>}
     {limitRadius=28, drivingBone {boneName=spine_02}, sphereLocalOffset=<2.6549 15 0>}
     {limitRadius=1, drivingBone {boneName=pelvis}, sphereLocalOffset=<-30 0 0>}
    ]
   }
   n24 = Anim Dynamics - Bone: dyn_l_braid_06!(ComponentPose=@n23, Alpha=1, GravityOverride:Vector=@n20, ExternalForce="0,0,0") -> (Pose) {
    angularDampingOverride=1
    angularSpringConstant=2500
    bAngularSpring=true
    bGravityOverrideInSimSpace=true
    bOverrideAngularDamping=true
    bUseGravityOverride=true
    bUseSphericalLimits=true
    gravityScale=0.35
    simulationSpace=RootRelative
    alphaBoolBlend {blendOption=Linear}
    alphaScaleBiasClamp {scale=0.4}
    boundBone {boneName=dyn_l_braid_06}
    componentAppliedLinearAccClamp=<5000 5000 5000>
    componentLinearAccScale {z=0.44}
    componentLinearVelScale {z=0.44}
    physicsBodyDefinitions [{
      collisionType=CustomSphere
      sphereCollisionRadius=1
      boundBone {boneName=dyn_l_braid_06}
      boxExtents=<12 22 22>
      constraintSetup {angularLimitsMax=<15 15 25>, angularLimitsMin=<15 -15 -5>, angularTarget=<0 0 0>}
      localJointOffset=<10 0 0>
     }]
    sphericalLimits [
     {limitRadius=65.9176, drivingBone {boneName=spine_05}, sphereLocalOffset=<-12.8628 52 0>}
     {limitRadius=50, drivingBone {boneName=pelvis}, sphereLocalOffset=<-2.6826 38.079 0>}
     {limitRadius=28, drivingBone {boneName=spine_02}, sphereLocalOffset=<2.6549 15 0>}
     {limitRadius=1, drivingBone {boneName=pelvis}, sphereLocalOffset=<-30 0 0>}
    ]
   }
   n25 = Anim Dynamics - Bone: dyn_l_braid_07!(ComponentPose=@n24, Alpha=1, GravityOverride:Vector=@n20, ExternalForce="0,0,0") -> (Pose) {
    angularDampingOverride=1
    angularSpringConstant=2500
    bAngularSpring=true
    bGravityOverrideInSimSpace=true
    bOverrideAngularDamping=true
    bUseGravityOverride=true
    bUseSphericalLimits=true
    gravityScale=0.35
    simulationSpace=RootRelative
    alphaBoolBlend {blendOption=Linear}
    alphaScaleBiasClamp {scale=0.4}
    boundBone {boneName=dyn_l_braid_07}
    componentAppliedLinearAccClamp=<5000 5000 5000>
    componentLinearAccScale {z=0.46}
    componentLinearVelScale {z=0.46}
    physicsBodyDefinitions [{
      collisionType=CustomSphere
      sphereCollisionRadius=1
      boundBone {boneName=dyn_l_braid_07}
      boxExtents=<12 22 22>
      constraintSetup {angularLimitsMax=<15 15 25>, angularLimitsMin=<15 -15 -25>, angularTarget=<0 0 0>}
      localJointOffset=<10 0 0>
     }]
    sphericalLimits [
     {limitRadius=65.9176, drivingBone {boneName=spine_05}, sphereLocalOffset=<-12.8628 62 0>}
     {limitRadius=50, drivingBone {boneName=pelvis}, sphereLocalOffset=<-2.6826 38.079 0>}
     {limitRadius=28, drivingBone {boneName=spine_02}, sphereLocalOffset=<2.6549 15 0>}
     {limitRadius=1, drivingBone {boneName=pelvis}, sphereLocalOffset=<-30 0 0>}
    ]
   }
   n26 = Anim Dynamics - Bone: dyn_l_braid_08!(ComponentPose=@n25, Alpha=1, GravityOverride:Vector=@n20, ExternalForce="0,0,0") -> (Pose) {
    angularDampingOverride=1
    angularSpringConstant=2500
    bAngularSpring=true
    bGravityOverrideInSimSpace=true
    bOverrideAngularDamping=true
    bUseGravityOverride=true
    bUseSphericalLimits=true
    gravityScale=0.35
    simulationSpace=RootRelative
    alphaBoolBlend {blendOption=Linear}
    alphaScaleBiasClamp {scale=0.5}
    boundBone {boneName=dyn_l_braid_08}
    componentAppliedLinearAccClamp=<5000 5000 5000>
    componentLinearAccScale {z=0.46}
    componentLinearVelScale {z=0.46}
    physicsBodyDefinitions [{
      collisionType=CustomSphere
      sphereCollisionRadius=1
      boundBone {boneName=dyn_l_braid_08}
      boxExtents=<12 22 22>
      constraintSetup {angularLimitsMax=<15 15 25>, angularLimitsMin=<15 -15 -25>, angularTarget=<0 0 0>}
      localJointOffset=<10 0 0>
     }]
    sphericalLimits [
     {limitRadius=65.9176, drivingBone {boneName=spine_05}, sphereLocalOffset=<-12.8628 62 0>}
     {limitRadius=50, drivingBone {boneName=pelvis}, sphereLocalOffset=<-2.6826 38.079 0>}
     {limitRadius=28, drivingBone {boneName=spine_02}, sphereLocalOffset=<2.6549 15 0>}
     {limitRadius=1, drivingBone {boneName=pelvis}, sphereLocalOffset=<-30 0 0>}
    ]
   }
   n27 = Anim Dynamics - Bone: dyn_l_braid_09!(ComponentPose=@n26, Alpha=1, GravityOverride:Vector=@n20, ExternalForce="0,0,0") -> (Pose) {
    angularDampingOverride=1
    angularSpringConstant=2500
    bAngularSpring=true
    bGravityOverrideInSimSpace=true
    bOverrideAngularDamping=true
    bUseGravityOverride=true
    bUseSphericalLimits=true
    gravityScale=0.35
    simulationSpace=RootRelative
    alphaBoolBlend {blendOption=Linear}
    alphaScaleBiasClamp {scale=0.5}
    boundBone {boneName=dyn_l_braid_09}
    componentAppliedLinearAccClamp=<5000 5000 5000>
    componentLinearAccScale {z=0.46}
    componentLinearVelScale {z=0.46}
    physicsBodyDefinitions [{
      collisionType=CustomSphere
      sphereCollisionRadius=1
      boundBone {boneName=dyn_l_braid_09}
      boxExtents=<12 22 22>
      constraintSetup {angularLimitsMax=<15 15 25>, angularLimitsMin=<15 -15 -25>, angularTarget=<0 0 0>}
      localJointOffset=<10 0 0>
     }]
    sphericalLimits [
     {limitRadius=65.9176, drivingBone {boneName=spine_05}, sphereLocalOffset=<-12.8628 62 0>}
     {limitRadius=50, drivingBone {boneName=pelvis}, sphereLocalOffset=<-2.6826 38.079 0>}
     {limitRadius=28, drivingBone {boneName=spine_02}, sphereLocalOffset=<2.6549 15 0>}
     {limitRadius=1, drivingBone {boneName=pelvis}, sphereLocalOffset=<-30 0 0>}
    ]
   }
   n28 = Anim Dynamics - Bone: dyn_l_braid_010!(ComponentPose=@n27, Alpha=1, GravityOverride:Vector=@n20, ExternalForce="0,0,0") -> (Pose) {
    angularDampingOverride=1
    angularSpringConstant=2500
    bAngularSpring=true
    bGravityOverrideInSimSpace=true
    bOverrideAngularDamping=true
    bUseGravityOverride=true
    bUseSphericalLimits=true
    gravityScale=0.35
    simulationSpace=RootRelative
    alphaBoolBlend {blendOption=Linear}
    alphaScaleBiasClamp {scale=0.5}
    boundBone {boneName=dyn_l_braid_010}
    componentAppliedLinearAccClamp=<5000 5000 5000>
    componentLinearAccScale {z=0.46}
    componentLinearVelScale {z=0.46}
    physicsBodyDefinitions [{
      collisionType=CustomSphere
      sphereCollisionRadius=1
      boundBone {boneName=dyn_l_braid_010}
      boxExtents=<12 22 22>
      constraintSetup {angularLimitsMax=<15 15 25>, angularLimitsMin=<15 -15 -25>, angularTarget=<0 0 0>}
      localJointOffset=<10 0 0>
     }]
    sphericalLimits [
     {limitRadius=65.9176, drivingBone {boneName=spine_05}, sphereLocalOffset=<-12.8628 62 0>}
     {limitRadius=50, drivingBone {boneName=pelvis}, sphereLocalOffset=<-2.6826 38.079 0>}
     {limitRadius=28, drivingBone {boneName=spine_02}, sphereLocalOffset=<2.6549 15 0>}
     {limitRadius=1, drivingBone {boneName=pelvis}, sphereLocalOffset=<-30 0 0>}
    ]
   }
   n29 = Anim Dynamics - Bone: dyn_l_braid_011!(ComponentPose=@n28, Alpha=1, GravityOverride:Vector=@n20, ExternalForce="0,0,0") -> (Pose) {
    angularDampingOverride=1
    angularSpringConstant=2500
    bAngularSpring=true
    bGravityOverrideInSimSpace=true
    bOverrideAngularDamping=true
    bUseGravityOverride=true
    bUseSphericalLimits=true
    gravityScale=0.35
    simulationSpace=RootRelative
    alphaBoolBlend {blendOption=Linear}
    alphaScaleBiasClamp {scale=0.5}
    boundBone {boneName=dyn_l_braid_011}
    componentAppliedLinearAccClamp=<5000 5000 5000>
    componentLinearAccScale {z=0.46}
    componentLinearVelScale {z=0.46}
    physicsBodyDefinitions [{
      collisionType=CustomSphere
      sphereCollisionRadius=1
      boundBone {boneName=dyn_l_braid_011}
      boxExtents=<12 22 22>
      constraintSetup {angularLimitsMax=<15 15 25>, angularLimitsMin=<15 -15 -25>, angularTarget=<0 0 0>}
      localJointOffset=<10 0 0>
     }]
    sphericalLimits [
     {limitRadius=65.9176, drivingBone {boneName=spine_05}, sphereLocalOffset=<-12.8628 62 0>}
     {limitRadius=50, drivingBone {boneName=pelvis}, sphereLocalOffset=<-2.6826 38.079 0>}
     {limitRadius=28, drivingBone {boneName=spine_02}, sphereLocalOffset=<2.6549 15 0>}
     {limitRadius=1, drivingBone {boneName=pelvis}, sphereLocalOffset=<-30 0 0>}
    ]
   }
   n30 = Anim Dynamics - Bone: dyn_r_braid_02!(ComponentPose=@n29, Alpha=@n18, GravityOverride:Vector="0,0,0", ExternalForce="0,0,0") -> (Pose) {
    angularDampingOverride=1
    angularSpringConstant=2500
    bAngularSpring=true
    bGravityOverrideInSimSpace=true
    bOverrideAngularDamping=true
    bUseGravityOverride=true
    bUseSphericalLimits=true
    gravityScale=0.35
    simulationSpace=RootRelative
    alphaBoolBlend {blendOption=Linear}
    alphaScaleBiasClamp {scale=0.4}
    boundBone {boneName=dyn_r_braid_02}
    componentAppliedLinearAccClamp=<5000 5000 5000>
    componentLinearAccScale {z=0.44}
    componentLinearVelScale {z=0.44}
    physicsBodyDefinitions [{
      collisionType=CustomSphere
      sphereCollisionRadius=1
      boundBone {boneName=dyn_r_braid_02}
      boxExtents=<12 22 22>
      constraintSetup {angularLimitsMax=<15 15 5>, angularLimitsMin=<15 -15 -25>, angularTarget=<0 0 0>}
      localJointOffset=<10 0 0>
     }]
    sphericalLimits [
     {limitRadius=60, drivingBone {boneName=spine_05}, sphereLocalOffset=<-12.8628 52 0>}
     {limitRadius=45, drivingBone {boneName=pelvis}, sphereLocalOffset=<-2.6826 38.079 0>}
     {limitRadius=28, drivingBone {boneName=spine_02}, sphereLocalOffset=<2.6549 15 0>}
     {limitRadius=1, drivingBone {boneName=pelvis}, sphereLocalOffset=<-30 0 0>}
    ]
   }
   n31 = Anim Dynamics - Bone: dyn_r_braid_03!(ComponentPose=@n30, Alpha=@n18, GravityOverride:Vector=@n20, ExternalForce="0,0,0") -> (Pose) {
    angularDampingOverride=1
    angularSpringConstant=2500
    bAngularSpring=true
    bGravityOverrideInSimSpace=true
    bOverrideAngularDamping=true
    bUseGravityOverride=true
    bUseSphericalLimits=true
    gravityScale=0.35
    simulationSpace=RootRelative
    alphaBoolBlend {blendOption=Linear}
    alphaScaleBiasClamp {scale=0.4}
    boundBone {boneName=dyn_r_braid_03}
    componentAppliedLinearAccClamp=<5000 5000 5000>
    componentLinearAccScale {z=0.44}
    componentLinearVelScale {z=0.44}
    physicsBodyDefinitions [{
      collisionType=CustomSphere
      sphereCollisionRadius=1
      boundBone {boneName=dyn_r_braid_03}
      boxExtents=<12 22 22>
      constraintSetup {angularLimitsMax=<15 15 5>, angularLimitsMin=<15 -15 -25>, angularTarget=<0 0 0>}
      localJointOffset=<10 0 0>
     }]
    sphericalLimits [
     {limitRadius=60, drivingBone {boneName=spine_05}, sphereLocalOffset=<-12.8628 52 0>}
     {limitRadius=45, drivingBone {boneName=pelvis}, sphereLocalOffset=<-2.6826 38.079 0>}
     {limitRadius=28, drivingBone {boneName=spine_02}, sphereLocalOffset=<2.6549 15 0>}
     {limitRadius=1, drivingBone {boneName=pelvis}, sphereLocalOffset=<-30 0 0>}
    ]
   }
   n32 = Anim Dynamics - Bone: dyn_r_braid_04!(ComponentPose=@n31, Alpha=1, GravityOverride:Vector=@n20, ExternalForce="0,0,0") -> (Pose) {
    angularDampingOverride=1
    angularSpringConstant=2500
    bAngularSpring=true
    bGravityOverrideInSimSpace=true
    bOverrideAngularDamping=true
    bUseGravityOverride=true
    bUseSphericalLimits=true
    gravityScale=0.35
    simulationSpace=RootRelative
    alphaBoolBlend {blendOption=Linear}
    alphaScaleBiasClamp {scale=0.4}
    boundBone {boneName=dyn_r_braid_04}
    componentAppliedLinearAccClamp=<5000 5000 5000>
    componentLinearAccScale {z=0.44}
    componentLinearVelScale {z=0.44}
    physicsBodyDefinitions [{
      collisionType=CustomSphere
      sphereCollisionRadius=1
      boundBone {boneName=dyn_r_braid_04}
      boxExtents=<12 22 22>
      constraintSetup {angularLimitsMax=<15 15 5>, angularLimitsMin=<15 -15 -25>, angularTarget=<0 0 0>}
      localJointOffset=<10 0 0>
     }]
    sphericalLimits [
     {limitRadius=60, drivingBone {boneName=spine_05}, sphereLocalOffset=<-12.8628 52 0>}
     {limitRadius=50, drivingBone {boneName=pelvis}, sphereLocalOffset=<-2.6826 38.079 0>}
     {limitRadius=28, drivingBone {boneName=spine_02}, sphereLocalOffset=<2.6549 15 0>}
     {limitRadius=1, drivingBone {boneName=pelvis}, sphereLocalOffset=<-30 0 0>}
    ]
   }
   n33 = Anim Dynamics - Bone: dyn_r_braid_05!(ComponentPose=@n32, Alpha=1, GravityOverride:Vector=@n20, ExternalForce="0,0,0") -> (Pose) {
    angularDampingOverride=1
    angularSpringConstant=2500
    bAngularSpring=true
    bGravityOverrideInSimSpace=true
    bOverrideAngularDamping=true
    bUseGravityOverride=true
    bUseSphericalLimits=true
    gravityScale=0.35
    simulationSpace=RootRelative
    alphaBoolBlend {blendOption=Linear}
    alphaScaleBiasClamp {scale=0.4}
    boundBone {boneName=dyn_r_braid_05}
    componentAppliedLinearAccClamp=<5000 5000 5000>
    componentLinearAccScale {z=0.44}
    componentLinearVelScale {z=0.44}
    physicsBodyDefinitions [{
      collisionType=CustomSphere
      sphereCollisionRadius=1
      boundBone {boneName=dyn_r_braid_05}
      boxExtents=<12 22 22>
      constraintSetup {angularLimitsMax=<15 15 5>, angularLimitsMin=<15 -15 -25>, angularTarget=<0 0 0>}
      localJointOffset=<10 0 0>
     }]
    sphericalLimits [
     {limitRadius=60, drivingBone {boneName=spine_05}, sphereLocalOffset=<-12.8628 52 0>}
     {limitRadius=50, drivingBone {boneName=pelvis}, sphereLocalOffset=<-2.6826 38.079 0>}
     {limitRadius=28, drivingBone {boneName=spine_02}, sphereLocalOffset=<2.6549 15 0>}
     {limitRadius=1, drivingBone {boneName=pelvis}, sphereLocalOffset=<-30 0 0>}
    ]
   }
   n34 = Anim Dynamics - Bone: dyn_r_braid_06!(ComponentPose=@n33, Alpha=1, GravityOverride:Vector=@n20, ExternalForce="0,0,0") -> (Pose) {
    angularDampingOverride=1
    angularSpringConstant=2500
    bAngularSpring=true
    bGravityOverrideInSimSpace=true
    bOverrideAngularDamping=true
    bUseGravityOverride=true
    bUseSphericalLimits=true
    gravityScale=0.35
    simulationSpace=RootRelative
    alphaBoolBlend {blendOption=Linear}
    alphaScaleBiasClamp {scale=0.4}
    boundBone {boneName=dyn_r_braid_06}
    componentAppliedLinearAccClamp=<5000 5000 5000>
    componentLinearAccScale {z=0.44}
    componentLinearVelScale {z=0.44}
    physicsBodyDefinitions [{
      collisionType=CustomSphere
      sphereCollisionRadius=1
      boundBone {boneName=dyn_r_braid_06}
      boxExtents=<12 22 22>
      constraintSetup {angularLimitsMax=<15 15 5>, angularLimitsMin=<15 -15 -25>, angularTarget=<0 0 0>}
      localJointOffset=<10 0 0>
     }]
    sphericalLimits [
     {limitRadius=65.9176, drivingBone {boneName=spine_05}, sphereLocalOffset=<-12.8628 52 0>}
     {limitRadius=50, drivingBone {boneName=pelvis}, sphereLocalOffset=<-2.6826 38.079 0>}
     {limitRadius=28, drivingBone {boneName=spine_02}, sphereLocalOffset=<2.6549 15 0>}
     {limitRadius=1, drivingBone {boneName=pelvis}, sphereLocalOffset=<-30 0 0>}
    ]
   }
   n35 = Anim Dynamics - Bone: dyn_r_braid_07!(ComponentPose=@n34, Alpha=1, GravityOverride:Vector=@n20, ExternalForce="0,0,0") -> (Pose) {
    angularDampingOverride=1
    angularSpringConstant=2500
    bAngularSpring=true
    bGravityOverrideInSimSpace=true
    bOverrideAngularDamping=true
    bUseGravityOverride=true
    bUseSphericalLimits=true
    gravityScale=0.35
    simulationSpace=RootRelative
    alphaBoolBlend {blendOption=Linear}
    alphaScaleBiasClamp {scale=0.4}
    boundBone {boneName=dyn_r_braid_07}
    componentAppliedLinearAccClamp=<5000 5000 5000>
    componentLinearAccScale {z=0.46}
    componentLinearVelScale {z=0.46}
    physicsBodyDefinitions [{
      collisionType=CustomSphere
      sphereCollisionRadius=1
      boundBone {boneName=dyn_r_braid_07}
      boxExtents=<12 22 22>
      constraintSetup {angularLimitsMax=<15 15 25>, angularLimitsMin=<15 -15 -25>, angularTarget=<0 0 0>}
      localJointOffset=<10 0 0>
     }]
    sphericalLimits [
     {limitRadius=65.9176, drivingBone {boneName=spine_05}, sphereLocalOffset=<-12.8628 62 0>}
     {limitRadius=50, drivingBone {boneName=pelvis}, sphereLocalOffset=<-2.6826 38.079 0>}
     {limitRadius=28, drivingBone {boneName=spine_02}, sphereLocalOffset=<2.6549 15 0>}
     {limitRadius=1, drivingBone {boneName=pelvis}, sphereLocalOffset=<-30 0 0>}
    ]
   }
   n36 = Anim Dynamics - Bone: dyn_r_braid_08!(ComponentPose=@n35, Alpha=1, GravityOverride:Vector=@n20, ExternalForce="0,0,0") -> (Pose) {
    angularDampingOverride=1
    angularSpringConstant=2500
    bAngularSpring=true
    bGravityOverrideInSimSpace=true
    bOverrideAngularDamping=true
    bUseGravityOverride=true
    bUseSphericalLimits=true
    gravityScale=0.35
    simulationSpace=RootRelative
    alphaBoolBlend {blendOption=Linear}
    alphaScaleBiasClamp {scale=0.5}
    boundBone {boneName=dyn_r_braid_08}
    componentAppliedLinearAccClamp=<5000 5000 5000>
    componentLinearAccScale {z=0.46}
    componentLinearVelScale {z=0.46}
    physicsBodyDefinitions [{
      collisionType=CustomSphere
      sphereCollisionRadius=1
      boundBone {boneName=dyn_r_braid_08}
      boxExtents=<12 22 22>
      constraintSetup {angularLimitsMax=<15 15 25>, angularLimitsMin=<15 -15 -25>, angularTarget=<0 0 0>}
      localJointOffset=<10 0 0>
     }]
    sphericalLimits [
     {limitRadius=65.9176, drivingBone {boneName=spine_05}, sphereLocalOffset=<-12.8628 62 0>}
     {limitRadius=50, drivingBone {boneName=pelvis}, sphereLocalOffset=<-2.6826 38.079 0>}
     {limitRadius=28, drivingBone {boneName=spine_02}, sphereLocalOffset=<2.6549 15 0>}
     {limitRadius=1, drivingBone {boneName=pelvis}, sphereLocalOffset=<-30 0 0>}
    ]
   }
   n37 = Anim Dynamics - Bone: dyn_r_braid_09!(ComponentPose=@n36, Alpha=1, GravityOverride:Vector=@n20, ExternalForce="0,0,0") -> (Pose) {
    angularDampingOverride=1
    angularSpringConstant=2500
    bAngularSpring=true
    bGravityOverrideInSimSpace=true
    bOverrideAngularDamping=true
    bUseGravityOverride=true
    bUseSphericalLimits=true
    gravityScale=0.35
    simulationSpace=RootRelative
    alphaBoolBlend {blendOption=Linear}
    alphaScaleBiasClamp {scale=0.5}
    boundBone {boneName=dyn_r_braid_09}
    componentAppliedLinearAccClamp=<5000 5000 5000>
    componentLinearAccScale {z=0.46}
    componentLinearVelScale {z=0.46}
    physicsBodyDefinitions [{
      collisionType=CustomSphere
      sphereCollisionRadius=1
      boundBone {boneName=dyn_r_braid_09}
      boxExtents=<12 22 22>
      constraintSetup {angularLimitsMax=<15 15 25>, angularLimitsMin=<15 -15 -25>, angularTarget=<0 0 0>}
      localJointOffset=<10 0 0>
     }]
    sphericalLimits [
     {limitRadius=65.9176, drivingBone {boneName=spine_05}, sphereLocalOffset=<-12.8628 62 0>}
     {limitRadius=50, drivingBone {boneName=pelvis}, sphereLocalOffset=<-2.6826 38.079 0>}
     {limitRadius=28, drivingBone {boneName=spine_02}, sphereLocalOffset=<2.6549 15 0>}
     {limitRadius=1, drivingBone {boneName=pelvis}, sphereLocalOffset=<-30 0 0>}
    ]
   }
   n38 = Anim Dynamics - Bone: dyn_r_braid_010!(ComponentPose=@n37, Alpha=1, GravityOverride:Vector=@n20, ExternalForce="0,0,0") -> (Pose) {
    angularDampingOverride=1
    angularSpringConstant=2500
    bAngularSpring=true
    bGravityOverrideInSimSpace=true
    bOverrideAngularDamping=true
    bUseGravityOverride=true
    bUseSphericalLimits=true
    gravityScale=0.35
    simulationSpace=RootRelative
    alphaBoolBlend {blendOption=Linear}
    alphaScaleBiasClamp {scale=0.5}
    boundBone {boneName=dyn_r_braid_010}
    componentAppliedLinearAccClamp=<5000 5000 5000>
    componentLinearAccScale {z=0.46}
    componentLinearVelScale {z=0.46}
    physicsBodyDefinitions [{
      collisionType=CustomSphere
      sphereCollisionRadius=1
      boundBone {boneName=dyn_r_braid_010}
      boxExtents=<12 22 22>
      constraintSetup {angularLimitsMax=<15 15 25>, angularLimitsMin=<15 -15 -25>, angularTarget=<0 0 0>}
      localJointOffset=<10 0 0>
     }]
    sphericalLimits [
     {limitRadius=65.9176, drivingBone {boneName=spine_05}, sphereLocalOffset=<-12.8628 62 0>}
     {limitRadius=50, drivingBone {boneName=pelvis}, sphereLocalOffset=<-2.6826 38.079 0>}
     {limitRadius=28, drivingBone {boneName=spine_02}, sphereLocalOffset=<2.6549 15 0>}
     {limitRadius=1, drivingBone {boneName=pelvis}, sphereLocalOffset=<-30 0 0>}
    ]
   }
   n39 = Anim Dynamics - Bone: dyn_r_braid_011!(ComponentPose=@n38, Alpha=1, GravityOverride:Vector=@n20, ExternalForce="0,0,0") -> (Pose) {
    angularDampingOverride=1
    angularSpringConstant=2500
    bAngularSpring=true
    bGravityOverrideInSimSpace=true
    bOverrideAngularDamping=true
    bUseGravityOverride=true
    bUseSphericalLimits=true
    gravityScale=0.35
    simulationSpace=RootRelative
    alphaBoolBlend {blendOption=Linear}
    alphaScaleBiasClamp {scale=0.5}
    boundBone {boneName=dyn_r_braid_011}
    componentAppliedLinearAccClamp=<5000 5000 5000>
    componentLinearAccScale {z=0.46}
    componentLinearVelScale {z=0.46}
    physicsBodyDefinitions [{
      collisionType=CustomSphere
      sphereCollisionRadius=1
      boundBone {boneName=dyn_r_braid_011}
      boxExtents=<12 22 22>
      constraintSetup {angularLimitsMax=<15 15 25>, angularLimitsMin=<15 -15 -25>, angularTarget=<0 0 0>}
      localJointOffset=<10 0 0>
     }]
    sphericalLimits [
     {limitRadius=65.9176, drivingBone {boneName=spine_05}, sphereLocalOffset=<-12.8628 62 0>}
     {limitRadius=50, drivingBone {boneName=pelvis}, sphereLocalOffset=<-2.6826 38.079 0>}
     {limitRadius=28, drivingBone {boneName=spine_02}, sphereLocalOffset=<2.6549 15 0>}
     {limitRadius=1, drivingBone {boneName=pelvis}, sphereLocalOffset=<-30 0 0>}
    ]
   }
   n40 = Blend Poses by bool(
    BlendPose_0=Use cached pose 'Cape Diving'()
    BlendPose_1=Blend Poses by bool(
     BlendPose_0=Component To Local(ComponentPose=@n39)
     BlendPose_1=Blend Poses by bool(BlendPose_0=@n7, BlendPose_1=@n6, bActiveValue=@n8) {blendType=Linear}
     BlendTime_0=1
     bActiveValue=Get bIsSkyDiving()
    ) {blendType=Linear}
    BlendTime_0=1
    bActiveValue=Get CapeSkydive() -> (CapeSkydive:bool)
   ) {blendType=Linear}
   n41 = Use cached pose 'dbno'()
   n42 = Get bIsDBNO()
   n43 = Control Rig(Source=Use cached pose 'Prebraid sim'()) {
    alphaBoolBlend {blendOption=Linear}
    controlRigAssetReference {blueprintRigClass=&F_MED_Uproar_Braids_FaceAcc_CtrlRig_C}
   }
   n44 = Local To Component(LocalPose=@n43)
   Output Pose(Result=Blend Poses by bool(
     BlendPose_0=Blend Poses by bool(BlendPose_0=@n41, BlendPose_1=@n40, BlendTime_0=1, bActiveValue=@n42) {blendType=Linear}
     BlendPose_1=Component To Local(ComponentPose=@n44)
     BlendTime_0=1
     bActiveValue=Get animDynamicsEnabledOrFrontend() -> (animDynamicsEnabledOrFrontend:bool)
    ) {blendType=Linear}) {layerGroup=None}
   n45 = Get Skel_Mesh()
   n46 = RigidBody(ComponentPose=Local To Component(LocalPose=Copy Pose From Mesh(SourceMeshComponent=@n45) {bUseAttachedParent=true})) {
    bOverrideWorldGravity=true
    overridePhysicsAsset=&/F_MED_Uproar_Braids_Hair_Physics
    simulationSpace=BaseBoneSpace
    alphaBoolBlend {blendOption=Linear}
    baseBoneRef {boneName=head}
    componentAppliedLinearAccClamp=<2000 2000 2000>
    componentLinearAccScale=<0.3 0.3 0.35>
    componentLinearVelScale=<0.3 0.3 0.35>
    overrideWorldGravity {z=-980}
   }
   n47 = "Transform (Modify) Bone - Bone: dyn_hair_fr_base_01"(ComponentPose=@n46, Alpha=1, Translation="0,0,0", Rotation="-10,0.000,0.000", Scale="1,1,1") {
    rotationMode=BMM_Additive
    translationMode=BMM_Additive
    alphaBoolBlend {blendInTime=0.3, blendOption=Linear, blendOutTime=0.3}
    boneToModify {boneName=dyn_hair_fr_base_01}
    rotation {pitch=-10}
   }
   n48 = "Transform (Modify) Bone - Bone: dyn_hair_top_base_01"(ComponentPose=@n47, Alpha=1, Translation="0,0,0", Rotation="-10,0.000,0.000", Scale="1,1,1") {
    rotationMode=BMM_Additive
    translationMode=BMM_Additive
    alphaBoolBlend {blendInTime=0.3, blendOption=Linear, blendOutTime=0.3}
    boneToModify {boneName=dyn_hair_top_base_01}
    rotation {pitch=-10}
   }
   n49 = "Transform (Modify) Bone - Bone: dyn_braids_lookat"!(ComponentPose=@n48, bAlphaBoolEnabled:bool=Get bIsSkyDiving(), Translation="0,10,-20", Rotation="0,0,0", Scale="1,1,1") -> (Pose) {
    alphaInputType=Bool
    translationMode=BMM_Additive
    alphaBoolBlend {blendInTime=0.3, blendOption=Linear, blendOutTime=0.3}
    boneToModify {boneName=dyn_braids_lookat}
    translation {y=10, z=-20}
   }
   n50 = "Transform (Modify) Bone - Bone: dyn_braids_lookat"(ComponentPose=@n49, bAlphaBoolEnabled=Get bIsParachuteOpened(), Translation="0,-60,60", Rotation="0,0,0", Scale="1,1,1") {
    alphaInputType=Bool
    translationMode=BMM_Additive
    alphaBoolBlend {blendInTime=0.3, blendOption=Linear, blendOutTime=0.3}
    boneToModify {boneName=dyn_braids_lookat}
    translation {y=-60, z=60}
   }
   n51 = "Transform (Modify) Bone - Bone: dyn_braids_lookat"(ComponentPose=@n50, bAlphaBoolEnabled=Get bIsDBNOCarried(), Translation="0,25,0", Rotation="0,0,0", Scale="1,1,1") {
    alphaInputType=Bool
    translationMode=BMM_Additive
    translationSpace=BCS_BoneSpace
    alphaBoolBlend {blendInTime=0.3, blendOption=Linear, blendOutTime=0.3}
    boneToModify {boneName=dyn_braids_lookat}
    translation {y=25}
   }
   n52 = "Transform (Modify) Bone - Bone: dyn_braids_lookat"(ComponentPose=@n51, bAlphaBoolEnabled=Get bIsInFrontEnd(), Translation="15,0,0", Rotation="0,0,0", Scale="1,1,1") {
    alphaInputType=Bool
    translationMode=BMM_Additive
    alphaBoolBlend {blendInTime=0.3, blendOption=Linear, blendOutTime=0.3}
    boneToModify {boneName=dyn_braids_lookat}
    translation {x=15}
   }
   Save cached pose 'Prebraid sim'(Pose=Component To Local(ComponentPose=@n52)) {cacheName=Prebraid sim}
   n53 = Control Rig(Source=Use cached pose 'Prebraid sim'()) {
    alphaBoolBlend {blendOption=Linear}
    controlRigAssetReference {blueprintRigClass=&F_MED_Uproar_Braids_FaceAcc_CtrlRig_C}
   }
   n54 = "Transform (Modify) Bone - Bone: dyn_r_braid_04"(ComponentPose=Local To Component(LocalPose=@n53), Alpha=1, Translation="0,0,0", Rotation="0,0,0", Scale="0.9,1,1") {
    rotationMode=BMM_Additive
    translationMode=BMM_Additive
    alphaBoolBlend {blendOption=Linear}
    boneToModify {boneName=dyn_r_braid_04}
    scale {x=0.9}
   }
   n55 = "Transform (Modify) Bone - Bone: dyn_r_braid_06"(ComponentPose=@n54, Alpha=1, Translation="0,0,0", Rotation="0,0,0", Scale="0.9,1,1") {
    rotationMode=BMM_Additive
    translationMode=BMM_Additive
    alphaBoolBlend {blendOption=Linear}
    boneToModify {boneName=dyn_r_braid_06}
    scale {x=0.9}
   }
   n56 = "Transform (Modify) Bone - Bone: dyn_r_braid_08"(ComponentPose=@n55, Alpha=1, Translation="0,0,0", Rotation="0,0,0", Scale="0.9,1,1") {
    rotationMode=BMM_Additive
    translationMode=BMM_Additive
    alphaBoolBlend {blendOption=Linear}
    boneToModify {boneName=dyn_r_braid_08}
    scale {x=0.9}
   }
   n57 = "Transform (Modify) Bone - Bone: dyn_r_braid_010"(ComponentPose=@n56, Alpha=1, Translation="0,0,0", Rotation="0,0,0", Scale="0.9,1,1") {
    rotationMode=BMM_Additive
    translationMode=BMM_Additive
    alphaBoolBlend {blendOption=Linear}
    boneToModify {boneName=dyn_r_braid_010}
    scale {x=0.9}
   }
   n58 = "Transform (Modify) Bone - Bone: dyn_l_braid_04"(ComponentPose=@n57, Alpha=1, Translation="0,0,0", Rotation="0,0,0", Scale="0.9,1,1") {
    rotationMode=BMM_Additive
    translationMode=BMM_Additive
    alphaBoolBlend {blendOption=Linear}
    boneToModify {boneName=dyn_l_braid_04}
    scale {x=0.9}
   }
   n59 = "Transform (Modify) Bone - Bone: dyn_l_braid_06"(ComponentPose=@n58, Alpha=1, Translation="0,0,0", Rotation="0,0,0", Scale="0.9,1,1") {
    rotationMode=BMM_Additive
    translationMode=BMM_Additive
    alphaBoolBlend {blendOption=Linear}
    boneToModify {boneName=dyn_l_braid_06}
    scale {x=0.9}
   }
   n60 = "Transform (Modify) Bone - Bone: dyn_l_braid_08"(ComponentPose=@n59, Alpha=1, Translation="0,0,0", Rotation="0,0,0", Scale="0.9,1,1") {
    rotationMode=BMM_Additive
    translationMode=BMM_Additive
    alphaBoolBlend {blendOption=Linear}
    boneToModify {boneName=dyn_l_braid_08}
    scale {x=0.9}
   }
   n61 = "Transform (Modify) Bone - Bone: dyn_l_braid_010"(ComponentPose=@n60, Alpha=1, Translation="0,0,0", Rotation="0,0,0", Scale="0.9,1,1") {
    rotationMode=BMM_Additive
    translationMode=BMM_Additive
    alphaBoolBlend {blendOption=Linear}
    boneToModify {boneName=dyn_l_braid_010}
    scale {x=0.9}
   }
   n62 = Anim Dynamics - Bone: dyn_l_braid_04(ComponentPose=@n61, Alpha=1, GravityOverride="0,0,0", ExternalForce="0,0,0") {
    angularDampingOverride=1
    angularSpringConstant=2500
    bAngularSpring=true
    bGravityOverrideInSimSpace=true
    bOverrideAngularDamping=true
    bUseSphericalLimits=true
    gravityScale=0.5
    simulationSpace=RootRelative
    alphaBoolBlend {blendOption=Linear}
    alphaScaleBiasClamp {scale=0.3}
    boundBone {boneName=dyn_l_braid_04}
    componentAppliedLinearAccClamp=<2500 2500 2500>
    componentLinearAccScale {z=0.35}
    componentLinearVelScale {z=0.7}
    physicsBodyDefinitions [{
      collisionType=CustomSphere
      sphereCollisionRadius=1
      boundBone {boneName=dyn_l_braid_04}
      boxExtents=<12 22 22>
      constraintSetup {angularLimitsMax=<10 10 20>, angularLimitsMin=<-10 -10 -20>, angularTarget=<0 0 0>}
      localJointOffset=<10 0 0>
     }]
    sphericalLimits [
     {limitRadius=65.9176, drivingBone {boneName=spine_05}, sphereLocalOffset=<-12.8628 55 0>}
     {limitRadius=50, drivingBone {boneName=pelvis}, sphereLocalOffset=<-2.6826 40 0>}
     {limitRadius=28, drivingBone {boneName=spine_02}, sphereLocalOffset=<2.6549 15 0>}
     {limitRadius=1, drivingBone {boneName=pelvis}, sphereLocalOffset=<-30 0 0>}
    ]
   }
   n63 = Anim Dynamics - Bone: dyn_l_braid_05(ComponentPose=@n62, Alpha=1, GravityOverride="0,0,0", ExternalForce="0,0,0") {
    angularDampingOverride=1
    angularSpringConstant=2500
    bAngularSpring=true
    bGravityOverrideInSimSpace=true
    bOverrideAngularDamping=true
    bUseSphericalLimits=true
    gravityScale=0.5
    simulationSpace=RootRelative
    alphaBoolBlend {blendOption=Linear}
    alphaScaleBiasClamp {scale=0.3}
    boundBone {boneName=dyn_l_braid_05}
    componentAppliedLinearAccClamp=<2500 2500 2500>
    componentLinearAccScale {z=0.35}
    componentLinearVelScale {z=0.7}
    physicsBodyDefinitions [{
      collisionType=CustomSphere
      sphereCollisionRadius=1
      boundBone {boneName=dyn_l_braid_05}
      boxExtents=<12 22 22>
      constraintSetup {angularLimitsMax=<10 10 20>, angularLimitsMin=<-10 -10 -20>, angularTarget=<0 0 0>}
      localJointOffset=<10 0 0>
     }]
    sphericalLimits [
     {limitRadius=65.9176, drivingBone {boneName=spine_05}, sphereLocalOffset=<-12.8628 55 0>}
     {limitRadius=50, drivingBone {boneName=pelvis}, sphereLocalOffset=<-2.6826 35 0>}
     {limitRadius=28, drivingBone {boneName=spine_02}, sphereLocalOffset=<2.6549 15 0>}
     {limitRadius=1, drivingBone {boneName=pelvis}, sphereLocalOffset=<-30 0 0>}
    ]
   }
   n64 = Anim Dynamics - Bone: dyn_l_braid_06(ComponentPose=@n63, Alpha=1, GravityOverride="0,0,0", ExternalForce="0,0,0") {
    angularDampingOverride=1
    angularSpringConstant=2500
    bAngularSpring=true
    bGravityOverrideInSimSpace=true
    bOverrideAngularDamping=true
    bUsePlanarLimit=false
    bUseSphericalLimits=true
    gravityScale=0.5
    simulationSpace=RootRelative
    alphaBoolBlend {blendOption=Linear}
    alphaScaleBiasClamp {scale=0.3}
    boundBone {boneName=dyn_l_braid_06}
    componentAppliedLinearAccClamp=<1500 1500 1500>
    componentLinearAccScale {z=0.35}
    componentLinearVelScale {z=0.35}
    physicsBodyDefinitions [{
      collisionType=CustomSphere
      sphereCollisionRadius=1
      boundBone {boneName=dyn_l_braid_06}
      boxExtents=<12 22 22>
      constraintSetup {angularLimitsMax=<10 10 20>, angularLimitsMin=<-10 -5 -20>, angularTarget=<0 0 0>}
      localJointOffset=<10 0 0>
     }]
    sphericalLimits [
     {limitRadius=65.9176, drivingBone {boneName=spine_05}, sphereLocalOffset=<-12.8628 55 0>}
     {limitRadius=50, drivingBone {boneName=pelvis}, sphereLocalOffset=<-2.6826 35 0>}
     {limitRadius=28, drivingBone {boneName=spine_02}, sphereLocalOffset=<2.6549 15 0>}
     {limitRadius=1, drivingBone {boneName=pelvis}, sphereLocalOffset=<-30 0 0>}
    ]
   }
   n65 = Anim Dynamics - Bone: dyn_l_braid_07(ComponentPose=@n64, Alpha=1, GravityOverride="0,0,0", ExternalForce="0,0,0") {
    angularDampingOverride=1
    angularSpringConstant=2500
    bAngularSpring=true
    bGravityOverrideInSimSpace=true
    bOverrideAngularDamping=true
    bUsePlanarLimit=false
    bUseSphericalLimits=true
    gravityScale=0.5
    simulationSpace=RootRelative
    alphaBoolBlend {blendOption=Linear}
    alphaScaleBiasClamp {scale=0.3}
    boundBone {boneName=dyn_l_braid_07}
    componentAppliedLinearAccClamp=<1500 1500 1500>
    componentLinearAccScale {z=0.35}
    componentLinearVelScale {z=0.35}
    physicsBodyDefinitions [{
      collisionType=CustomSphere
      sphereCollisionRadius=1
      boundBone {boneName=dyn_l_braid_07}
      boxExtents=<12 22 22>
      constraintSetup {angularLimitsMax=<10 10 40>, angularLimitsMin=<-10 -5 -40>, angularTarget=<0 0 0>}
      localJointOffset=<10 0 0>
     }]
    sphericalLimits [
     {limitRadius=65.9176, drivingBone {boneName=spine_05}, sphereLocalOffset=<-12.8628 65 0>}
     {limitRadius=50, drivingBone {boneName=pelvis}, sphereLocalOffset=<-2.6826 35 0>}
     {limitRadius=28, drivingBone {boneName=spine_02}, sphereLocalOffset=<2.6549 17 0>}
     {limitRadius=1, drivingBone {boneName=pelvis}, sphereLocalOffset=<-30 0 0>}
    ]
   }
   n66 = Anim Dynamics - Bone: dyn_l_braid_08(ComponentPose=@n65, Alpha=1, GravityOverride="0,0,0", ExternalForce="0,0,0") {
    angularDampingOverride=1
    angularSpringConstant=2500
    bAngularSpring=true
    bGravityOverrideInSimSpace=true
    bOverrideAngularDamping=true
    bUsePlanarLimit=false
    bUseSphericalLimits=true
    gravityScale=0.5
    simulationSpace=RootRelative
    alphaBoolBlend {blendOption=Linear}
    alphaScaleBiasClamp {scale=0.3}
    boundBone {boneName=dyn_l_braid_08}
    componentAppliedLinearAccClamp=<1500 1500 1500>
    componentLinearAccScale {z=0.35}
    componentLinearVelScale {z=0.35}
    physicsBodyDefinitions [{
      collisionType=CustomSphere
      sphereCollisionRadius=1
      boundBone {boneName=dyn_l_braid_08}
      boxExtents=<12 22 22>
      constraintSetup {angularLimitsMax=<10 10 40>, angularLimitsMin=<-10 -5 -40>, angularTarget=<0 0 0>}
      localJointOffset=<10 0 0>
     }]
    sphericalLimits [
     {limitRadius=65.9176, drivingBone {boneName=spine_05}, sphereLocalOffset=<-12.8628 65 0>}
     {limitRadius=50, drivingBone {boneName=pelvis}, sphereLocalOffset=<-2.6826 40 0>}
     {limitRadius=28, drivingBone {boneName=spine_02}, sphereLocalOffset=<2.6549 17 0>}
     {limitRadius=1, drivingBone {boneName=pelvis}, sphereLocalOffset=<-30 0 0>}
    ]
   }
   n67 = Anim Dynamics - Bone: dyn_l_braid_09(ComponentPose=@n66, Alpha=1, GravityOverride="0,0,0", ExternalForce="0,0,0") {
    angularDampingOverride=1
    angularSpringConstant=2500
    bAngularSpring=true
    bGravityOverrideInSimSpace=true
    bOverrideAngularDamping=true
    bUsePlanarLimit=false
    bUseSphericalLimits=true
    gravityScale=0.5
    simulationSpace=RootRelative
    alphaBoolBlend {blendOption=Linear}
    alphaScaleBiasClamp {scale=0.5}
    boundBone {boneName=dyn_l_braid_09}
    componentAppliedLinearAccClamp=<1500 1500 1500>
    componentLinearAccScale {z=0.35}
    componentLinearVelScale {z=0.35}
    physicsBodyDefinitions [{
      collisionType=CustomSphere
      sphereCollisionRadius=1
      boundBone {boneName=dyn_l_braid_09}
      boxExtents=<12 22 22>
      constraintSetup {angularLimitsMax=<10 10 40>, angularLimitsMin=<-10 -5 -40>, angularTarget=<0 0 0>}
      localJointOffset=<10 0 0>
     }]
    sphericalLimits [
     {limitRadius=65.9176, drivingBone {boneName=spine_05}, sphereLocalOffset=<-12.8628 65 0>}
     {limitRadius=50, drivingBone {boneName=pelvis}, sphereLocalOffset=<-2.6826 40 0>}
     {limitRadius=28, drivingBone {boneName=spine_02}, sphereLocalOffset=<2.6549 17 0>}
     {limitRadius=1, drivingBone {boneName=pelvis}, sphereLocalOffset=<-30 0 0>}
    ]
   }
   n68 = Anim Dynamics - Bone: dyn_l_braid_010(ComponentPose=@n67, Alpha=1, GravityOverride="0,0,0", ExternalForce="0,0,0") {
    angularDampingOverride=1
    angularSpringConstant=2500
    bAngularSpring=true
    bGravityOverrideInSimSpace=true
    bOverrideAngularDamping=true
    bUsePlanarLimit=false
    bUseSphericalLimits=true
    gravityScale=0.5
    simulationSpace=RootRelative
    alphaBoolBlend {blendOption=Linear}
    alphaScaleBiasClamp {scale=0.5}
    boundBone {boneName=dyn_l_braid_010}
    componentAppliedLinearAccClamp=<1500 1500 1500>
    componentLinearAccScale {z=0.35}
    componentLinearVelScale {z=0.35}
    physicsBodyDefinitions [{
      collisionType=CustomSphere
      sphereCollisionRadius=1
      boundBone {boneName=dyn_l_braid_010}
      boxExtents=<12 22 22>
      constraintSetup {angularLimitsMax=<10 10 40>, angularLimitsMin=<-10 -5 -40>, angularTarget=<0 0 0>}
      localJointOffset=<10 0 0>
     }]
    sphericalLimits [
     {limitRadius=65.9176, drivingBone {boneName=spine_05}, sphereLocalOffset=<-12.8628 65 0>}
     {limitRadius=50, drivingBone {boneName=pelvis}, sphereLocalOffset=<-2.6826 40 0>}
     {limitRadius=28, drivingBone {boneName=spine_02}, sphereLocalOffset=<2.6549 17 0>}
     {limitRadius=1, drivingBone {boneName=pelvis}, sphereLocalOffset=<-30 0 0>}
     {limitRadius=0, drivingBone {boneName=None}, sphereLocalOffset=<0 0 0>}
    ]
   }
   n69 = Anim Dynamics - Bone: dyn_l_braid_011(ComponentPose=@n68, Alpha=1, GravityOverride="0,0,0", ExternalForce="0,0,0") {
    angularDampingOverride=1
    angularSpringConstant=2500
    bAngularSpring=true
    bGravityOverrideInSimSpace=true
    bOverrideAngularDamping=true
    bUsePlanarLimit=false
    bUseSphericalLimits=true
    gravityScale=0.5
    simulationSpace=RootRelative
    alphaBoolBlend {blendOption=Linear}
    alphaScaleBiasClamp {scale=0.5}
    boundBone {boneName=dyn_l_braid_011}
    componentAppliedLinearAccClamp=<1500 1500 1500>
    componentLinearAccScale {z=0.35}
    componentLinearVelScale {z=0.35}
    physicsBodyDefinitions [{
      collisionType=CustomSphere
      sphereCollisionRadius=1
      boundBone {boneName=dyn_l_braid_011}
      boxExtents=<12 22 22>
      constraintSetup {angularLimitsMax=<10 10 40>, angularLimitsMin=<-10 -5 -40>, angularTarget=<0 0 0>}
      localJointOffset=<10 0 0>
     }]
    sphericalLimits [
     {limitRadius=65.9176, drivingBone {boneName=spine_05}, sphereLocalOffset=<-12.8628 65 0>}
     {limitRadius=50, drivingBone {boneName=pelvis}, sphereLocalOffset=<-2.6826 40 0>}
     {limitRadius=28, drivingBone {boneName=spine_02}, sphereLocalOffset=<2.6549 17 0>}
     {limitRadius=1, drivingBone {boneName=pelvis}, sphereLocalOffset=<-30 0 0>}
     {limitRadius=0, drivingBone {boneName=None}, sphereLocalOffset=<0 0 0>}
    ]
   }
   n70 = Anim Dynamics - Bone: dyn_r_braid_04(ComponentPose=@n69, Alpha=1, GravityOverride="0,0,0", ExternalForce="0,0,0") {
    angularDampingOverride=1
    angularSpringConstant=2500
    bAngularSpring=true
    bGravityOverrideInSimSpace=true
    bOverrideAngularDamping=true
    bUseSphericalLimits=true
    gravityScale=0.5
    simulationSpace=RootRelative
    alphaBoolBlend {blendOption=Linear}
    alphaScaleBiasClamp {scale=0.35}
    boundBone {boneName=dyn_r_braid_04}
    componentAppliedLinearAccClamp=<2500 2500 2500>
    componentLinearAccScale {z=0.35}
    componentLinearVelScale {z=0.7}
    physicsBodyDefinitions [{
      collisionType=CustomSphere
      sphereCollisionRadius=1
      boundBone {boneName=dyn_r_braid_04}
      boxExtents=<14 28 28>
      constraintSetup {angularLimitsMax=<10 10 20>, angularLimitsMin=<-10 -10 -20>, angularTarget=<0 0 0>}
      localJointOffset=<10 -0.5 0>
     }]
    sphericalLimits [
     {limitRadius=65.9176, drivingBone {boneName=spine_05}, sphereLocalOffset=<-12.8628 55 0>}
     {limitRadius=50, drivingBone {boneName=pelvis}, sphereLocalOffset=<-2.6826 40 0>}
     {limitRadius=28, drivingBone {boneName=spine_02}, sphereLocalOffset=<2.6549 15 0>}
     {limitRadius=1, drivingBone {boneName=pelvis}, sphereLocalOffset=<-30 0 0>}
    ]
   }
   n71 = Anim Dynamics - Bone: dyn_r_braid_05(ComponentPose=@n70, Alpha=1, GravityOverride="0,0,0", ExternalForce="0,0,0") {
    angularDampingOverride=1
    angularSpringConstant=2500
    bAngularSpring=true
    bGravityOverrideInSimSpace=true
    bOverrideAngularDamping=true
    bUseSphericalLimits=true
    gravityScale=0.5
    simulationSpace=RootRelative
    alphaBoolBlend {blendOption=Linear}
    alphaScaleBiasClamp {scale=0.35}
    boundBone {boneName=dyn_r_braid_05}
    componentAppliedLinearAccClamp=<2500 2500 2500>
    componentLinearAccScale {z=0.35}
    componentLinearVelScale {z=0.7}
    physicsBodyDefinitions [{
      collisionType=CustomSphere
      sphereCollisionRadius=1
      boundBone {boneName=dyn_r_braid_05}
      boxExtents=<14 28 28>
      constraintSetup {angularLimitsMax=<10 10 20>, angularLimitsMin=<-10 -10 -20>, angularTarget=<0 0 0>}
      localJointOffset=<10 -0.5 0>
     }]
    sphericalLimits [
     {limitRadius=65.9176, drivingBone {boneName=spine_05}, sphereLocalOffset=<-12.8628 55 0>}
     {limitRadius=50, drivingBone {boneName=pelvis}, sphereLocalOffset=<-2.6826 40 0>}
     {limitRadius=28, drivingBone {boneName=spine_02}, sphereLocalOffset=<2.6549 15 0>}
     {limitRadius=1, drivingBone {boneName=pelvis}, sphereLocalOffset=<-30 0 0>}
    ]
   }
   n72 = Anim Dynamics - Bone: dyn_r_braid_06(ComponentPose=@n71, Alpha=1, GravityOverride="0,0,0", ExternalForce="0,0,0") {
    angularDampingOverride=1
    angularSpringConstant=2500
    bAngularSpring=true
    bGravityOverrideInSimSpace=true
    bOverrideAngularDamping=true
    bUsePlanarLimit=false
    bUseSphericalLimits=true
    gravityScale=0.5
    simulationSpace=RootRelative
    alphaBoolBlend {blendOption=Linear}
    alphaScaleBiasClamp {scale=0.35}
    boundBone {boneName=dyn_r_braid_06}
    componentAppliedLinearAccClamp=<1500 1500 1500>
    componentLinearAccScale {z=0.35}
    componentLinearVelScale {z=0.35}
    physicsBodyDefinitions [{
      collisionType=CustomSphere
      sphereCollisionRadius=1
      boundBone {boneName=dyn_r_braid_06}
      boxExtents=<14 28 28>
      constraintSetup {angularLimitsMax=<10 10 20>, angularLimitsMin=<-10 -5 -20>, angularTarget=<0 0 0>}
      localJointOffset=<10 -0.5 0>
     }]
    sphericalLimits [
     {limitRadius=65.9176, drivingBone {boneName=spine_05}, sphereLocalOffset=<-12.8628 55 0>}
     {limitRadius=50, drivingBone {boneName=pelvis}, sphereLocalOffset=<-2.6826 35 0>}
     {limitRadius=28, drivingBone {boneName=spine_02}, sphereLocalOffset=<2.6549 15 0>}
     {limitRadius=1, drivingBone {boneName=pelvis}, sphereLocalOffset=<-30 0 0>}
    ]
   }
   n73 = Anim Dynamics - Bone: dyn_r_braid_07(ComponentPose=@n72, Alpha=1, GravityOverride="0,0,0", ExternalForce="0,0,0") {
    angularDampingOverride=1
    angularSpringConstant=2500
    bAngularSpring=true
    bGravityOverrideInSimSpace=true
    bOverrideAngularDamping=true
    bUsePlanarLimit=false
    bUseSphericalLimits=true
    gravityScale=0.5
    simulationSpace=RootRelative
    alphaBoolBlend {blendOption=Linear}
    alphaScaleBiasClamp {scale=0.4}
    boundBone {boneName=dyn_r_braid_07}
    componentAppliedLinearAccClamp=<1500 1500 1500>
    componentLinearAccScale {z=0.35}
    componentLinearVelScale {z=0.35}
    physicsBodyDefinitions [{
      collisionType=CustomSphere
      sphereCollisionRadius=1
      boundBone {boneName=dyn_r_braid_07}
      boxExtents=<14 28 28>
      constraintSetup {angularLimitsMax=<10 10 40>, angularLimitsMin=<-10 -5 -40>, angularTarget=<0 0 0>}
      localJointOffset=<10 -0.5 0>
     }]
   }
   n74 = Anim Dynamics - Bone: dyn_r_braid_08(ComponentPose=@n73, Alpha=1, GravityOverride="0,0,0", ExternalForce="0,0,0") {
    angularDampingOverride=1
    angularSpringConstant=2500
    bAngularSpring=true
    bGravityOverrideInSimSpace=true
    bOverrideAngularDamping=true
    bUsePlanarLimit=false
    bUseSphericalLimits=true
    gravityScale=0.5
    simulationSpace=RootRelative
    alphaBoolBlend {blendOption=Linear}
    alphaScaleBiasClamp {scale=0.5}
    boundBone {boneName=dyn_r_braid_08}
    componentAppliedLinearAccClamp=<1500 1500 1500>
    componentLinearAccScale {z=0.35}
    componentLinearVelScale {z=0.35}
    physicsBodyDefinitions [{
      collisionType=CustomSphere
      sphereCollisionRadius=1
      boundBone {boneName=dyn_r_braid_08}
      boxExtents=<14 28 28>
      constraintSetup {angularLimitsMax=<10 10 40>, angularLimitsMin=<-10 -5 -40>, angularTarget=<0 0 0>}
      localJointOffset=<10 -0.5 0>
     }]
   }
   n75 = Anim Dynamics - Bone: dyn_r_braid_09(ComponentPose=@n74, Alpha=1, GravityOverride="0,0,0", ExternalForce="0,0,0") {
    angularDampingOverride=1
    angularSpringConstant=2500
    bAngularSpring=true
    bGravityOverrideInSimSpace=true
    bOverrideAngularDamping=true
    bUsePlanarLimit=false
    bUseSphericalLimits=true
    gravityScale=0.5
    simulationSpace=RootRelative
    alphaBoolBlend {blendOption=Linear}
    alphaScaleBiasClamp {scale=0.5}
    boundBone {boneName=dyn_r_braid_09}
    componentAppliedLinearAccClamp=<1500 1500 1500>
    componentLinearAccScale {z=0.35}
    componentLinearVelScale {z=0.35}
    physicsBodyDefinitions [{
      collisionType=CustomSphere
      sphereCollisionRadius=1
      boundBone {boneName=dyn_r_braid_09}
      boxExtents=<14 28 28>
      constraintSetup {angularLimitsMax=<10 10 40>, angularLimitsMin=<-10 -5 -40>, angularTarget=<0 0 0>}
      localJointOffset=<10 -0.5 0>
     }]
   }
   n76 = Anim Dynamics - Bone: dyn_r_braid_010(ComponentPose=@n75, Alpha=1, GravityOverride="0,0,0", ExternalForce="0,0,0") {
    angularDampingOverride=1
    angularSpringConstant=2500
    bAngularSpring=true
    bGravityOverrideInSimSpace=true
    bOverrideAngularDamping=true
    bUsePlanarLimit=false
    bUseSphericalLimits=true
    gravityScale=0.5
    simulationSpace=RootRelative
    alphaBoolBlend {blendOption=Linear}
    alphaScaleBiasClamp {scale=0.5}
    boundBone {boneName=dyn_r_braid_010}
    componentAppliedLinearAccClamp=<1500 1500 1500>
    componentLinearAccScale {z=0.35}
    componentLinearVelScale {z=0.35}
    physicsBodyDefinitions [{
      collisionType=CustomSphere
      sphereCollisionRadius=1
      boundBone {boneName=dyn_r_braid_010}
      boxExtents=<14 28 28>
      constraintSetup {angularLimitsMax=<10 10 40>, angularLimitsMin=<-10 -5 -40>, angularTarget=<0 0 0>}
      localJointOffset=<10 -0.5 0>
     }]
   }
   n77 = Anim Dynamics - Bone: dyn_r_braid_011(ComponentPose=@n76, Alpha=1, GravityOverride="0,0,0", ExternalForce="0,0,0") {
    angularDampingOverride=1
    angularSpringConstant=2500
    bAngularSpring=true
    bGravityOverrideInSimSpace=true
    bOverrideAngularDamping=true
    bUsePlanarLimit=false
    bUseSphericalLimits=true
    gravityScale=0.5
    simulationSpace=RootRelative
    alphaBoolBlend {blendOption=Linear}
    alphaScaleBiasClamp {scale=0.5}
    boundBone {boneName=dyn_r_braid_011}
    componentAppliedLinearAccClamp=<1500 1500 1500>
    componentLinearAccScale {z=0.35}
    componentLinearVelScale {z=0.35}
    physicsBodyDefinitions [{
      collisionType=CustomSphere
      sphereCollisionRadius=1
      boundBone {boneName=dyn_r_braid_011}
      boxExtents=<14 28 28>
      constraintSetup {angularLimitsMax=<10 10 40>, angularLimitsMin=<-10 -5 -40>, angularTarget=<0 0 0>}
      localJointOffset=<10 -0.5 0>
     }]
   }
   n78 = "Transform (Modify) Bone - Bone: dyn_l_braid_02"(ComponentPose=@n77, Alpha=1, Translation="0,0,0", Rotation="0,0,-5", Scale="1,1,1") {rotationMode=BMM_Additive, alphaBoolBlend {blendOption=Linear}, boneToModify {boneName=dyn_l_braid_02}, rotation {roll=-5}}
   n79 = "Transform (Modify) Bone - Bone: dyn_r_braid_02"(ComponentPose=@n78, Alpha=1, Translation="0,0,0", Rotation="0,0,-5", Scale="1,1,1") {rotationMode=BMM_Additive, alphaBoolBlend {blendOption=Linear}, boneToModify {boneName=dyn_r_braid_02}, rotation {roll=-5}}
   n80 = "Transform (Modify) Bone - Bone: dyn_l_braid_04"(ComponentPose=@n79, Alpha=1, Translation="0,0,0", Rotation="0,0,10", Scale="1,1,1") {rotationMode=BMM_Additive, alphaBoolBlend {blendOption=Linear}, boneToModify {boneName=dyn_l_braid_04}, rotation {roll=10}}
   n81 = "Transform (Modify) Bone - Bone: dyn_r_braid_04"(ComponentPose=@n80, Alpha=1, Translation="0,0,0", Rotation="0,0,10", Scale="1,1,1") {rotationMode=BMM_Additive, alphaBoolBlend {blendOption=Linear}, boneToModify {boneName=dyn_r_braid_04}, rotation {roll=10}}
   n82 = "Transform (Modify) Bone - Bone: dyn_l_braid_05"(ComponentPose=@n81, Alpha=1, Translation="0,0,0", Rotation="0,0,-5", Scale="1,1,1") {rotationMode=BMM_Additive, alphaBoolBlend {blendOption=Linear}, boneToModify {boneName=dyn_l_braid_05}, rotation {roll=-5}}
   n83 = "Transform (Modify) Bone - Bone: dyn_r_braid_05"(ComponentPose=@n82, Alpha=1, Translation="0,0,0", Rotation="0,0,-5", Scale="1,1,1") {rotationMode=BMM_Additive, alphaBoolBlend {blendOption=Linear}, boneToModify {boneName=dyn_r_braid_05}, rotation {roll=-5}}
   n84 = Get kunaiAiming() -> (kunaiAiming:double)
   n85 = "Transform (Modify) Bone - Bone: dyn_r_braid_02"(ComponentPose=@n83, Alpha=@n84, Translation="0,0,0", Rotation="10,10,0", Scale="1,1,1") {
    rotationMode=BMM_Additive
    rotationSpace=BCS_BoneSpace
    alphaBoolBlend {blendInTime=1, blendOption=Linear, blendOutTime=0.5}
    alphaScaleBiasClamp {bInterpResult=true, interpSpeedIncreasing=20}
    boneToModify {boneName=dyn_r_braid_02}
    rotation {pitch=10, yaw=10}
   }
   n86 = "Transform (Modify) Bone - Bone: dyn_r_braid_04"(ComponentPose=@n85, Alpha=@n84, Translation="0,0,0", Rotation="-15,-30,0.000", Scale="1,1,1") {
    rotationMode=BMM_Additive
    rotationSpace=BCS_BoneSpace
    alphaBoolBlend {blendInTime=1, blendOption=Linear, blendOutTime=0.5}
    alphaScaleBiasClamp {bInterpResult=true, interpSpeedIncreasing=20}
    boneToModify {boneName=dyn_r_braid_04}
    rotation {pitch=-15, yaw=-30}
   }
   n87 = "Transform (Modify) Bone - Bone: dyn_r_braid_05"(ComponentPose=@n86, Alpha=@n84, Translation="0,0,0", Rotation="10,0,0.000", Scale="1,1,1") {
    rotationMode=BMM_Additive
    rotationSpace=BCS_BoneSpace
    alphaBoolBlend {blendInTime=1, blendOption=Linear, blendOutTime=0.5}
    alphaScaleBiasClamp {bInterpResult=true, interpSpeedIncreasing=20}
    boneToModify {boneName=dyn_r_braid_05}
    rotation {pitch=10}
   }
   Save cached pose 'CapePhysics'(Pose=Component To Local(ComponentPose=@n87)) {cacheName=CapePhysics}
   n88 = Control Rig(Source=Use cached pose 'Prebraid sim'()) {
    alphaBoolBlend {blendOption=Linear}
    controlRigAssetReference {blueprintRigClass=&F_MED_Uproar_Braids_FaceAcc_CtrlRig_C}
   }
   n89 = Anim Dynamics - Bone: dyn_l_braid_06(ComponentPose=Local To Component(LocalPose=@n88), Alpha=1, GravityOverride="0,0,0", ExternalForce="0,0,0") {
    angularDampingOverride=1
    angularSpringConstant=2500
    bAngularSpring=true
    bGravityOverrideInSimSpace=true
    bOverrideAngularDamping=true
    bUsePlanarLimit=false
    gravityScale=0.5
    simulationSpace=BoneRelative
    alphaBoolBlend {blendOption=Linear}
    alphaScaleBiasClamp {scale=0.4}
    boundBone {boneName=dyn_l_braid_06}
    componentAppliedLinearAccClamp=<2000 2000 2000>
    componentLinearAccScale=<0.1 0.1 0.7>
    componentLinearVelScale=<0.1 0.1 0.7>
    physicsBodyDefinitions [{
      collisionType=CustomSphere
      sphereCollisionRadius=1
      boundBone {boneName=dyn_l_braid_06}
      boxExtents=<14 28 28>
      constraintSetup {angularLimitsMax=<25 45 25>, angularLimitsMin=<-25 -45 -25>, angularTarget=<0 0 0>}
      localJointOffset=<10 -0.5 0>
     }]
    planarLimits [{drivingBone {boneName=root}, planeTransform {location=<0 0 5>, rotation=<rot 0 0 0>}}]
    relativeSpaceBone {boneName=dyn_r_braid_05}
    sphericalLimits [
     {limitRadius=65.9176, drivingBone {boneName=spine_05}, sphereLocalOffset=<-12.8628 55 0>}
     {limitRadius=50, drivingBone {boneName=pelvis}, sphereLocalOffset=<-2.6826 35 0>}
     {limitRadius=28, drivingBone {boneName=spine_02}, sphereLocalOffset=<2.6549 15 0>}
     {limitRadius=1, drivingBone {boneName=pelvis}, sphereLocalOffset=<-30 0 0>}
    ]
   }
   n90 = Anim Dynamics - Bone: dyn_r_braid_07(ComponentPose=@n89, Alpha=1, GravityOverride="0,0,0", ExternalForce="0,0,0") {
    angularDampingOverride=1
    angularSpringConstant=2500
    bAngularSpring=true
    bGravityOverrideInSimSpace=true
    bOverrideAngularDamping=true
    bUsePlanarLimit=false
    gravityScale=0.5
    simulationSpace=BoneRelative
    alphaBoolBlend {blendOption=Linear}
    alphaScaleBiasClamp {scale=0.4}
    boundBone {boneName=dyn_r_braid_07}
    componentAppliedLinearAccClamp=<2000 2000 2000>
    componentLinearAccScale=<0.1 0.1 0.7>
    componentLinearVelScale=<0.1 0.1 0.7>
    physicsBodyDefinitions [{
      collisionType=CustomSphere
      sphereCollisionRadius=1
      boundBone {boneName=dyn_r_braid_07}
      boxExtents=<12 22 22>
      constraintSetup {angularLimitsMax=<25 45 25>, angularLimitsMin=<-25 -90 -25>, angularTarget=<0 0 0>}
      localJointOffset=<10 0 0>
     }]
    planarLimits [{drivingBone {boneName=root}, planeTransform {location=<0 0 3>, rotation=<rot 0 0 0>}}]
    relativeSpaceBone {boneName=dyn_r_braid_05}
    sphericalLimits [
     {limitRadius=65.9176, drivingBone {boneName=spine_05}, sphereLocalOffset=<-12.8628 65 0>}
     {limitRadius=50, drivingBone {boneName=pelvis}, sphereLocalOffset=<-2.6826 35 0>}
     {limitRadius=28, drivingBone {boneName=spine_02}, sphereLocalOffset=<2.6549 17 0>}
     {limitRadius=1, drivingBone {boneName=pelvis}, sphereLocalOffset=<-30 0 0>}
    ]
   }
   n91 = Anim Dynamics - Bone: dyn_l_braid_08(ComponentPose=@n90, Alpha=1, GravityOverride="0,0,0", ExternalForce="0,0,0") {
    angularDampingOverride=1
    angularSpringConstant=2500
    bAngularSpring=true
    bGravityOverrideInSimSpace=true
    bOverrideAngularDamping=true
    bUsePlanarLimit=false
    gravityScale=0.5
    simulationSpace=BoneRelative
    alphaBoolBlend {blendOption=Linear}
    alphaScaleBiasClamp {scale=0.5}
    boundBone {boneName=dyn_l_braid_08}
    componentAppliedLinearAccClamp=<2000 2000 2000>
    componentLinearAccScale=<0.1 0.1 0.7>
    componentLinearVelScale=<0.1 0.1 0.7>
    physicsBodyDefinitions [{
      collisionType=CustomSphere
      sphereCollisionRadius=1
      boundBone {boneName=dyn_l_braid_08}
      boxExtents=<12 22 22>
      constraintSetup {angularLimitsMax=<25 45 25>, angularLimitsMin=<-25 -90 -25>, angularTarget=<0 0 0>}
      localJointOffset=<10 0 0>
     }]
    planarLimits [{drivingBone {boneName=root}, planeTransform {location=<0 0 3>, rotation=<rot 0 0 0>}}]
    relativeSpaceBone {boneName=dyn_r_braid_05}
    sphericalLimits [
     {limitRadius=65.9176, drivingBone {boneName=spine_05}, sphereLocalOffset=<-12.8628 65 0>}
     {limitRadius=50, drivingBone {boneName=pelvis}, sphereLocalOffset=<-2.6826 40 0>}
     {limitRadius=28, drivingBone {boneName=spine_02}, sphereLocalOffset=<2.6549 17 0>}
     {limitRadius=1, drivingBone {boneName=pelvis}, sphereLocalOffset=<-30 0 0>}
    ]
   }
   n92 = Anim Dynamics - Bone: dyn_l_braid_09(ComponentPose=@n91, Alpha=1, GravityOverride="0,0,0", ExternalForce="0,0,0") {
    angularDampingOverride=1
    angularSpringConstant=2500
    bAngularSpring=true
    bGravityOverrideInSimSpace=true
    bOverrideAngularDamping=true
    bUsePlanarLimit=false
    gravityScale=0.5
    simulationSpace=BoneRelative
    alphaBoolBlend {blendOption=Linear}
    alphaScaleBiasClamp {scale=0.5}
    boundBone {boneName=dyn_l_braid_09}
    componentAppliedLinearAccClamp=<2000 2000 2000>
    componentLinearAccScale=<0.1 0.1 0.7>
    componentLinearVelScale=<0.1 0.1 0.7>
    physicsBodyDefinitions [{
      collisionType=CustomSphere
      sphereCollisionRadius=1
      boundBone {boneName=dyn_l_braid_09}
      boxExtents=<12 22 22>
      constraintSetup {angularLimitsMax=<25 45 25>, angularLimitsMin=<-25 -90 -25>, angularTarget=<0 0 0>}
      localJointOffset=<1 0 0>
     }]
    planarLimits [{drivingBone {boneName=root}, planeTransform {location=<0 0 3>, rotation=<rot 0 0 0>}}]
    relativeSpaceBone {boneName=dyn_r_braid_05}
    sphericalLimits [
     {limitRadius=65.9176, drivingBone {boneName=spine_05}, sphereLocalOffset=<-12.8628 65 0>}
     {limitRadius=50, drivingBone {boneName=pelvis}, sphereLocalOffset=<-2.6826 40 0>}
     {limitRadius=28, drivingBone {boneName=spine_02}, sphereLocalOffset=<2.6549 17 0>}
     {limitRadius=1, drivingBone {boneName=pelvis}, sphereLocalOffset=<-30 0 0>}
    ]
   }
   n93 = Anim Dynamics - Bone: dyn_l_braid_010(ComponentPose=@n92, Alpha=1, GravityOverride="0,0,0", ExternalForce="0,0,0") {
    angularDampingOverride=1
    angularSpringConstant=2500
    bAngularSpring=true
    bGravityOverrideInSimSpace=true
    bOverrideAngularDamping=true
    bUsePlanarLimit=false
    gravityScale=0.5
    simulationSpace=BoneRelative
    alphaBoolBlend {blendOption=Linear}
    alphaScaleBiasClamp {scale=0.5}
    boundBone {boneName=dyn_l_braid_010}
    componentAppliedLinearAccClamp=<2000 2000 2000>
    componentLinearAccScale=<0.1 0.1 0.7>
    componentLinearVelScale=<0.1 0.1 0.7>
    physicsBodyDefinitions [{
      collisionType=CustomSphere
      sphereCollisionRadius=1
      boundBone {boneName=dyn_l_braid_010}
      boxExtents=<12 22 22>
      constraintSetup {angularLimitsMax=<25 45 25>, angularLimitsMin=<-25 -45 -25>, angularTarget=<0 0 0>}
      localJointOffset=<5 0 0>
     }]
    planarLimits [{drivingBone {boneName=root}, planeTransform {location=<0 0 5>, rotation=<rot 0 0 0>}}]
    relativeSpaceBone {boneName=dyn_r_braid_05}
    sphericalLimits [
     {limitRadius=65.9176, drivingBone {boneName=spine_05}, sphereLocalOffset=<-12.8628 65 0>}
     {limitRadius=50, drivingBone {boneName=pelvis}, sphereLocalOffset=<-2.6826 40 0>}
     {limitRadius=28, drivingBone {boneName=spine_02}, sphereLocalOffset=<2.6549 17 0>}
     {limitRadius=1, drivingBone {boneName=pelvis}, sphereLocalOffset=<-30 0 0>}
    ]
   }
   n94 = Anim Dynamics - Bone: dyn_l_braid_011(ComponentPose=@n93, Alpha=1, GravityOverride="0,0,0", ExternalForce="0,0,0") {
    angularDampingOverride=1
    angularSpringConstant=2500
    bAngularSpring=true
    bGravityOverrideInSimSpace=true
    bOverrideAngularDamping=true
    bUsePlanarLimit=false
    gravityScale=0.5
    simulationSpace=BoneRelative
    alphaBoolBlend {blendOption=Linear}
    alphaScaleBiasClamp {scale=0.5}
    boundBone {boneName=dyn_l_braid_011}
    componentAppliedLinearAccClamp=<2000 2000 2000>
    componentLinearAccScale=<0.1 0.1 0.7>
    componentLinearVelScale=<0.1 0.1 0.7>
    physicsBodyDefinitions [{
      collisionType=CustomSphere
      sphereCollisionRadius=1
      boundBone {boneName=dyn_l_braid_011}
      boxExtents=<12 22 22>
      constraintSetup {angularLimitsMax=<25 45 25>, angularLimitsMin=<-25 -45 -25>, angularTarget=<0 0 0>}
      localJointOffset=<5 0 0>
     }]
    planarLimits [{drivingBone {boneName=root}, planeTransform {location=<0 0 5>, rotation=<rot 0 0 0>}}]
    relativeSpaceBone {boneName=dyn_r_braid_05}
    sphericalLimits [
     {limitRadius=65.9176, drivingBone {boneName=spine_05}, sphereLocalOffset=<-12.8628 65 0>}
     {limitRadius=50, drivingBone {boneName=pelvis}, sphereLocalOffset=<-2.6826 40 0>}
     {limitRadius=28, drivingBone {boneName=spine_02}, sphereLocalOffset=<2.6549 17 0>}
     {limitRadius=1, drivingBone {boneName=pelvis}, sphereLocalOffset=<-30 0 0>}
     {limitRadius=0, drivingBone {boneName=None}, sphereLocalOffset=<0 0 0>}
    ]
   }
   n95 = Anim Dynamics - Bone: dyn_r_braid_06(ComponentPose=@n94, Alpha=1, GravityOverride="0,0,0", ExternalForce="0,0,0") {
    angularDampingOverride=1
    angularSpringConstant=2500
    bAngularSpring=true
    bGravityOverrideInSimSpace=true
    bOverrideAngularDamping=true
    bUsePlanarLimit=false
    bUseSphericalLimits=true
    gravityScale=0.5
    simulationSpace=BoneRelative
    alphaBoolBlend {blendOption=Linear}
    alphaScaleBiasClamp {scale=0.4}
    boundBone {boneName=dyn_r_braid_06}
    componentAppliedLinearAccClamp=<2000 2000 2000>
    componentLinearAccScale=<0.1 0.1 0.7>
    componentLinearVelScale=<0.1 0.1 0.7>
    physicsBodyDefinitions [{
      collisionType=CustomSphere
      sphereCollisionRadius=1
      boundBone {boneName=dyn_r_braid_06}
      boxExtents=<14 28 28>
      constraintSetup {angularLimitsMax=<25 45 25>, angularLimitsMin=<-25 -45 -25>, angularTarget=<0 0 0>}
      localJointOffset=<10 -0.5 0>
     }]
    planarLimits [{drivingBone {boneName=root}, planeTransform {location=<0 0 5>, rotation=<rot 0 0 0>}}]
    relativeSpaceBone {boneName=dyn_r_braid_05}
    sphericalLimits [
     {limitRadius=60, drivingBone {boneName=spine_05}, sphereLocalOffset=<-12.8628 55 0>}
     {limitRadius=45, drivingBone {boneName=pelvis}, sphereLocalOffset=<-2.6826 35 0>}
     {limitRadius=28, drivingBone {boneName=spine_02}, sphereLocalOffset=<2.6549 15 0>}
     {limitRadius=1, drivingBone {boneName=pelvis}, sphereLocalOffset=<-30 0 0>}
    ]
   }
   n96 = Anim Dynamics - Bone: dyn_r_braid_07(ComponentPose=@n95, Alpha=1, GravityOverride="0,0,0", ExternalForce="0,0,0") {
    angularDampingOverride=1
    angularSpringConstant=2500
    bAngularSpring=true
    bGravityOverrideInSimSpace=true
    bOverrideAngularDamping=true
    bUsePlanarLimit=false
    gravityScale=0.5
    simulationSpace=BoneRelative
    alphaBoolBlend {blendOption=Linear}
    alphaScaleBiasClamp {scale=0.4}
    boundBone {boneName=dyn_r_braid_07}
    componentAppliedLinearAccClamp=<2000 2000 2000>
    componentLinearAccScale=<0.1 0.1 0.7>
    componentLinearVelScale=<0.1 0.1 0.7>
    physicsBodyDefinitions [{
      collisionType=CustomSphere
      sphereCollisionRadius=1
      boundBone {boneName=dyn_r_braid_07}
      boxExtents=<12 22 22>
      constraintSetup {angularLimitsMax=<25 45 25>, angularLimitsMin=<-25 -90 -25>, angularTarget=<0 0 0>}
      localJointOffset=<10 0 0>
     }]
    planarLimits [{drivingBone {boneName=root}, planeTransform {location=<0 0 3>, rotation=<rot 0 0 0>}}]
    relativeSpaceBone {boneName=dyn_r_braid_05}
    sphericalLimits [
     {limitRadius=65.9176, drivingBone {boneName=spine_05}, sphereLocalOffset=<-12.8628 65 0>}
     {limitRadius=50, drivingBone {boneName=pelvis}, sphereLocalOffset=<-2.6826 35 0>}
     {limitRadius=28, drivingBone {boneName=spine_02}, sphereLocalOffset=<2.6549 17 0>}
     {limitRadius=1, drivingBone {boneName=pelvis}, sphereLocalOffset=<-30 0 0>}
    ]
   }
   n97 = Anim Dynamics - Bone: dyn_r_braid_08(ComponentPose=@n96, Alpha=1, GravityOverride="0,0,0", ExternalForce="0,0,0") {
    angularDampingOverride=1
    angularSpringConstant=2500
    bAngularSpring=true
    bGravityOverrideInSimSpace=true
    bOverrideAngularDamping=true
    bUsePlanarLimit=false
    gravityScale=0.5
    simulationSpace=BoneRelative
    alphaBoolBlend {blendOption=Linear}
    alphaScaleBiasClamp {scale=0.5}
    boundBone {boneName=dyn_r_braid_08}
    componentAppliedLinearAccClamp=<2000 2000 2000>
    componentLinearAccScale=<0.1 0.1 0.7>
    componentLinearVelScale=<0.1 0.1 0.7>
    physicsBodyDefinitions [{
      collisionType=CustomSphere
      sphereCollisionRadius=1
      boundBone {boneName=dyn_r_braid_08}
      boxExtents=<12 22 22>
      constraintSetup {angularLimitsMax=<25 45 25>, angularLimitsMin=<-25 -90 -25>, angularTarget=<0 0 0>}
      localJointOffset=<10 0 0>
     }]
    planarLimits [{drivingBone {boneName=root}, planeTransform {location=<0 0 3>, rotation=<rot 0 0 0>}}]
    relativeSpaceBone {boneName=dyn_r_braid_05}
    sphericalLimits [
     {limitRadius=65.9176, drivingBone {boneName=spine_05}, sphereLocalOffset=<-12.8628 65 0>}
     {limitRadius=50, drivingBone {boneName=pelvis}, sphereLocalOffset=<-2.6826 40 0>}
     {limitRadius=28, drivingBone {boneName=spine_02}, sphereLocalOffset=<2.6549 17 0>}
     {limitRadius=1, drivingBone {boneName=pelvis}, sphereLocalOffset=<-30 0 0>}
    ]
   }
   n98 = Anim Dynamics - Bone: dyn_r_braid_09(ComponentPose=@n97, Alpha=1, GravityOverride="0,0,0", ExternalForce="0,0,0") {
    angularDampingOverride=1
    angularSpringConstant=2500
    bAngularSpring=true
    bGravityOverrideInSimSpace=true
    bOverrideAngularDamping=true
    bUsePlanarLimit=false
    gravityScale=0.5
    simulationSpace=BoneRelative
    alphaBoolBlend {blendOption=Linear}
    alphaScaleBiasClamp {scale=0.5}
    boundBone {boneName=dyn_r_braid_09}
    componentAppliedLinearAccClamp=<2000 2000 2000>
    componentLinearAccScale=<0.1 0.1 0.7>
    componentLinearVelScale=<0.1 0.1 0.7>
    physicsBodyDefinitions [{
      collisionType=CustomSphere
      sphereCollisionRadius=1
      boundBone {boneName=dyn_r_braid_09}
      boxExtents=<12 22 22>
      constraintSetup {angularLimitsMax=<25 45 25>, angularLimitsMin=<-25 -90 -25>, angularTarget=<0 0 0>}
      localJointOffset=<1 0 0>
     }]
    planarLimits [{drivingBone {boneName=root}, planeTransform {location=<0 0 3>, rotation=<rot 0 0 0>}}]
    relativeSpaceBone {boneName=dyn_r_braid_05}
    sphericalLimits [
     {limitRadius=65.9176, drivingBone {boneName=spine_05}, sphereLocalOffset=<-12.8628 65 0>}
     {limitRadius=50, drivingBone {boneName=pelvis}, sphereLocalOffset=<-2.6826 40 0>}
     {limitRadius=28, drivingBone {boneName=spine_02}, sphereLocalOffset=<2.6549 17 0>}
     {limitRadius=1, drivingBone {boneName=pelvis}, sphereLocalOffset=<-30 0 0>}
    ]
   }
   n99 = Anim Dynamics - Bone: dyn_r_braid_010(ComponentPose=@n98, Alpha=1, GravityOverride="0,0,0", ExternalForce="0,0,0") {
    angularDampingOverride=1
    angularSpringConstant=2500
    bAngularSpring=true
    bGravityOverrideInSimSpace=true
    bOverrideAngularDamping=true
    bUsePlanarLimit=false
    gravityScale=0.5
    simulationSpace=BoneRelative
    alphaBoolBlend {blendOption=Linear}
    alphaScaleBiasClamp {scale=0.5}
    boundBone {boneName=dyn_r_braid_010}
    componentAppliedLinearAccClamp=<2000 2000 2000>
    componentLinearAccScale=<0.1 0.1 0.7>
    componentLinearVelScale=<0.1 0.1 0.7>
    physicsBodyDefinitions [{
      collisionType=CustomSphere
      sphereCollisionRadius=1
      boundBone {boneName=dyn_r_braid_010}
      boxExtents=<12 22 22>
      constraintSetup {angularLimitsMax=<25 45 25>, angularLimitsMin=<-25 -45 -25>, angularTarget=<0 0 0>}
      localJointOffset=<5 0 0>
     }]
    planarLimits [{drivingBone {boneName=root}, planeTransform {location=<0 0 5>, rotation=<rot 0 0 0>}}]
    relativeSpaceBone {boneName=dyn_r_braid_05}
    sphericalLimits [
     {limitRadius=65.9176, drivingBone {boneName=spine_05}, sphereLocalOffset=<-12.8628 65 0>}
     {limitRadius=50, drivingBone {boneName=pelvis}, sphereLocalOffset=<-2.6826 40 0>}
     {limitRadius=28, drivingBone {boneName=spine_02}, sphereLocalOffset=<2.6549 17 0>}
     {limitRadius=1, drivingBone {boneName=pelvis}, sphereLocalOffset=<-30 0 0>}
    ]
   }
   n100 = Anim Dynamics - Bone: dyn_r_braid_011(ComponentPose=@n99, Alpha=1, GravityOverride="0,0,0", ExternalForce="0,0,0") {
    angularDampingOverride=1
    angularSpringConstant=2500
    bAngularSpring=true
    bGravityOverrideInSimSpace=true
    bOverrideAngularDamping=true
    bUsePlanarLimit=false
    gravityScale=0.5
    simulationSpace=BoneRelative
    alphaBoolBlend {blendOption=Linear}
    alphaScaleBiasClamp {scale=0.5}
    boundBone {boneName=dyn_r_braid_011}
    componentAppliedLinearAccClamp=<2000 2000 2000>
    componentLinearAccScale=<0.1 0.1 0.7>
    componentLinearVelScale=<0.1 0.1 0.7>
    physicsBodyDefinitions [{
      collisionType=CustomSphere
      sphereCollisionRadius=1
      boundBone {boneName=dyn_r_braid_011}
      boxExtents=<12 22 22>
      constraintSetup {angularLimitsMax=<25 45 25>, angularLimitsMin=<-25 -45 -25>, angularTarget=<0 0 0>}
      localJointOffset=<5 0 0>
     }]
    planarLimits [{drivingBone {boneName=root}, planeTransform {location=<0 0 5>, rotation=<rot 0 0 0>}}]
    relativeSpaceBone {boneName=dyn_r_braid_05}
    sphericalLimits [
     {limitRadius=65.9176, drivingBone {boneName=spine_05}, sphereLocalOffset=<-12.8628 65 0>}
     {limitRadius=50, drivingBone {boneName=pelvis}, sphereLocalOffset=<-2.6826 40 0>}
     {limitRadius=28, drivingBone {boneName=spine_02}, sphereLocalOffset=<2.6549 17 0>}
     {limitRadius=1, drivingBone {boneName=pelvis}, sphereLocalOffset=<-30 0 0>}
     {limitRadius=0, drivingBone {boneName=None}, sphereLocalOffset=<0 0 0>}
    ]
   }
   n101 = Control Rig(Source=Use cached pose 'Prebraid sim'()) {
    alphaBoolBlend {blendOption=Linear}
    controlRigAssetReference {blueprintRigClass=&F_MED_Uproar_Braids_FaceAcc_CtrlRig_C}
   }
   n102 = RigidBody(ComponentPose=Local To Component(LocalPose=@n101), Alpha=0.3) {
    bOverrideWorldGravity=true
    overridePhysicsAsset=&/F_MED_Uproar_Braids_FaceAcc_Physics
    simulationSpace=BaseBoneSpace
    alphaBoolBlend {blendOption=Linear}
    baseBoneRef {boneName=head}
    componentAppliedLinearAccClamp=<1000 1000 1000>
    componentLinearAccScale=<0.2 0.2 0.3>
    componentLinearVelScale=<0.2 0.2 0.3>
    overrideWorldGravity {z=-980}
   }
   n103 = Blend Poses by bool(
    BlendPose_0=Component To Local(ComponentPose=@n102)
    BlendPose_1=Component To Local(ComponentPose=@n100)
    bActiveValue=Get bRigidBodyEnabled()
   ) {blendType=Linear}
   n104 = Control Rig(Source=Use cached pose 'Prebraid sim'()) {
    alphaBoolBlend {blendOption=Linear}
    controlRigAssetReference {blueprintRigClass=&F_MED_Uproar_Braids_FaceAcc_CtrlRig_C}
   }
   n105 = Anim Dynamics - Bone: dyn_l_braid_05(ComponentPose=Local To Component(LocalPose=@n104), Alpha=1, GravityOverride="0,0,0", ExternalForce="0,0,0") {
    angularDampingOverride=1
    angularSpringConstant=2500
    bAngularSpring=true
    bGravityOverrideInSimSpace=true
    bOverrideAngularDamping=true
    bUseSphericalLimits=true
    gravityScale=0.5
    simulationSpace=RootRelative
    alphaBoolBlend {blendOption=Linear}
    alphaScaleBiasClamp {scale=0.5}
    boundBone {boneName=dyn_l_braid_05}
    componentAppliedLinearAccClamp=<5000 5000 5000>
    componentLinearAccScale {z=0.7}
    componentLinearVelScale {z=0.7}
    physicsBodyDefinitions [{
      collisionType=CustomSphere
      sphereCollisionRadius=1
      boundBone {boneName=dyn_l_braid_05}
      boxExtents=<12 22 22>
      constraintSetup {angularLimitsMax=<20 45 90>, angularLimitsMin=<-20 -45 -90>, angularTarget=<0 0 0>}
      localJointOffset=<10 0 0>
     }]
    sphericalLimits [
     {limitRadius=65.9176, drivingBone {boneName=spine_05}, sphereLocalOffset=<-12.8628 55 0>}
     {limitRadius=45, drivingBone {boneName=pelvis}, sphereLocalOffset=<-2.6826 35 0>}
     {limitRadius=28, drivingBone {boneName=spine_02}, sphereLocalOffset=<2.6549 15 0>}
     {limitRadius=1, drivingBone {boneName=pelvis}, sphereLocalOffset=<-30 0 0>}
    ]
   }
   n106 = Anim Dynamics - Bone: dyn_l_braid_06(ComponentPose=@n105, Alpha=1, GravityOverride="0,0,0", ExternalForce="0,0,0") {
    angularDampingOverride=1
    angularSpringConstant=2500
    bAngularSpring=true
    bGravityOverrideInSimSpace=true
    bOverrideAngularDamping=true
    bUseSphericalLimits=true
    gravityScale=0.5
    simulationSpace=RootRelative
    alphaBoolBlend {blendOption=Linear}
    alphaScaleBiasClamp {scale=0.35}
    boundBone {boneName=dyn_l_braid_06}
    componentAppliedLinearAccClamp=<5000 5000 5000>
    componentLinearAccScale {z=0.7}
    componentLinearVelScale {z=0.7}
    physicsBodyDefinitions [{
      collisionType=CustomSphere
      sphereCollisionRadius=1
      boundBone {boneName=dyn_l_braid_06}
      boxExtents=<12 22 22>
      constraintSetup {angularLimitsMax=<20 45 90>, angularLimitsMin=<-20 -45 -90>, angularTarget=<0 0 0>}
      localJointOffset=<10 0 0>
     }]
    planarLimits [{drivingBone {boneName=root}, planeTransform {location=<0 0 3>, rotation=<rot 0 0 0>}}]
    sphericalLimits [
     {limitRadius=65.9176, drivingBone {boneName=spine_05}, sphereLocalOffset=<-12.8628 55 0>}
     {limitRadius=45, drivingBone {boneName=pelvis}, sphereLocalOffset=<-2.6826 35 0>}
     {limitRadius=28, drivingBone {boneName=spine_02}, sphereLocalOffset=<2.6549 15 0>}
     {limitRadius=1, drivingBone {boneName=pelvis}, sphereLocalOffset=<-30 0 0>}
    ]
   }
   n107 = Anim Dynamics - Bone: dyn_l_braid_07(ComponentPose=@n106, Alpha=1, GravityOverride="0,0,0", ExternalForce="0,0,0") {
    angularDampingOverride=1
    angularSpringConstant=2500
    bAngularSpring=true
    bGravityOverrideInSimSpace=true
    bOverrideAngularDamping=true
    bUseSphericalLimits=true
    gravityScale=0.5
    simulationSpace=RootRelative
    alphaBoolBlend {blendOption=Linear}
    alphaScaleBiasClamp {scale=0.4}
    boundBone {boneName=dyn_l_braid_07}
    componentAppliedLinearAccClamp=<5000 5000 5000>
    componentLinearAccScale {z=0.7}
    componentLinearVelScale {z=0.7}
    physicsBodyDefinitions [{
      collisionType=CustomSphere
      sphereCollisionRadius=1
      boundBone {boneName=dyn_l_braid_07}
      boxExtents=<12 22 22>
      constraintSetup {angularLimitsMax=<20 90 90>, angularLimitsMin=<-20 -90 -90>, angularTarget=<0 0 0>}
      localJointOffset=<10 0 0>
     }]
    planarLimits [{drivingBone {boneName=root}, planeTransform {location=<0 0 3>, rotation=<rot 0 0 0>}}]
    sphericalLimits [
     {limitRadius=65.9176, drivingBone {boneName=spine_05}, sphereLocalOffset=<-12.8628 65 0>}
     {limitRadius=45, drivingBone {boneName=pelvis}, sphereLocalOffset=<-2.6826 35 0>}
     {limitRadius=28, drivingBone {boneName=spine_02}, sphereLocalOffset=<2.6549 17 0>}
     {limitRadius=1, drivingBone {boneName=pelvis}, sphereLocalOffset=<-30 0 0>}
    ]
   }
   n108 = Anim Dynamics - Bone: dyn_l_braid_08(ComponentPose=@n107, Alpha=1, GravityOverride="0,0,0", ExternalForce="0,0,0") {
    angularDampingOverride=1
    angularSpringConstant=2500
    bAngularSpring=true
    bGravityOverrideInSimSpace=true
    bOverrideAngularDamping=true
    bUseSphericalLimits=true
    gravityScale=0.5
    simulationSpace=RootRelative
    alphaBoolBlend {blendOption=Linear}
    alphaScaleBiasClamp {scale=0.5}
    boundBone {boneName=dyn_l_braid_08}
    componentAppliedLinearAccClamp=<5000 5000 5000>
    componentLinearAccScale {z=0.7}
    componentLinearVelScale {z=0.7}
    physicsBodyDefinitions [{
      collisionType=CustomSphere
      sphereCollisionRadius=1
      boundBone {boneName=dyn_l_braid_08}
      boxExtents=<12 22 22>
      constraintSetup {angularLimitsMax=<20 90 90>, angularLimitsMin=<-20 -90 -90>, angularTarget=<0 0 0>}
      localJointOffset=<10 0 0>
     }]
    planarLimits [{drivingBone {boneName=root}, planeTransform {location=<0 0 3>, rotation=<rot 0 0 0>}}]
    sphericalLimits [
     {limitRadius=65.9176, drivingBone {boneName=spine_05}, sphereLocalOffset=<-12.8628 65 0>}
     {limitRadius=45, drivingBone {boneName=pelvis}, sphereLocalOffset=<-2.6826 40 0>}
     {limitRadius=28, drivingBone {boneName=spine_02}, sphereLocalOffset=<2.6549 17 0>}
     {limitRadius=1, drivingBone {boneName=pelvis}, sphereLocalOffset=<-30 0 0>}
    ]
   }
   n109 = Anim Dynamics - Bone: dyn_l_braid_09(ComponentPose=@n108, Alpha=1, GravityOverride="0,0,0", ExternalForce="0,0,0") {
    angularDampingOverride=1
    angularSpringConstant=2500
    bAngularSpring=true
    bGravityOverrideInSimSpace=true
    bOverrideAngularDamping=true
    bUseSphericalLimits=true
    gravityScale=0.5
    simulationSpace=RootRelative
    alphaBoolBlend {blendOption=Linear}
    alphaScaleBiasClamp {scale=0.5}
    boundBone {boneName=dyn_l_braid_09}
    componentAppliedLinearAccClamp=<5000 5000 5000>
    componentLinearAccScale {z=0.7}
    componentLinearVelScale {z=0.7}
    physicsBodyDefinitions [{
      collisionType=CustomSphere
      sphereCollisionRadius=1
      boundBone {boneName=dyn_l_braid_09}
      boxExtents=<12 22 22>
      constraintSetup {angularLimitsMax=<20 90 90>, angularLimitsMin=<-20 -90 -90>, angularTarget=<0 0 0>}
      localJointOffset=<10 0 0>
     }]
    planarLimits [{drivingBone {boneName=root}, planeTransform {location=<0 0 3>, rotation=<rot 0 0 0>}}]
    sphericalLimits [
     {limitRadius=65.9176, drivingBone {boneName=spine_05}, sphereLocalOffset=<-12.8628 65 0>}
     {limitRadius=50, drivingBone {boneName=pelvis}, sphereLocalOffset=<-2.6826 40 0>}
     {limitRadius=28, drivingBone {boneName=spine_02}, sphereLocalOffset=<2.6549 17 0>}
     {limitRadius=1, drivingBone {boneName=pelvis}, sphereLocalOffset=<-30 0 0>}
    ]
   }
   n110 = Anim Dynamics - Bone: dyn_l_braid_010(ComponentPose=@n109, Alpha=1, GravityOverride="0,0,0", ExternalForce="0,0,0") {
    angularDampingOverride=1
    angularSpringConstant=2500
    bAngularSpring=true
    bGravityOverrideInSimSpace=true
    bOverrideAngularDamping=true
    bUseSphericalLimits=true
    gravityScale=0.5
    simulationSpace=RootRelative
    alphaBoolBlend {blendOption=Linear}
    alphaScaleBiasClamp {scale=0.5}
    boundBone {boneName=dyn_l_braid_010}
    componentAppliedLinearAccClamp=<5000 5000 5000>
    componentLinearAccScale {z=0.7}
    componentLinearVelScale {z=0.7}
    physicsBodyDefinitions [{
      collisionType=CustomSphere
      sphereCollisionRadius=1
      boundBone {boneName=dyn_l_braid_010}
      boxExtents=<12 22 22>
      constraintSetup {angularLimitsMax=<20 90 90>, angularLimitsMin=<-20 -90 -90>, angularTarget=<0 0 0>}
      localJointOffset=<10 0 0>
     }]
    planarLimits [{drivingBone {boneName=root}, planeTransform {location=<0 0 3>, rotation=<rot 0 0 0>}}]
    sphericalLimits [
     {limitRadius=65.9176, drivingBone {boneName=spine_05}, sphereLocalOffset=<-12.8628 65 0>}
     {limitRadius=50, drivingBone {boneName=pelvis}, sphereLocalOffset=<-2.6826 40 0>}
     {limitRadius=28, drivingBone {boneName=spine_02}, sphereLocalOffset=<2.6549 17 0>}
     {limitRadius=1, drivingBone {boneName=pelvis}, sphereLocalOffset=<-30 0 0>}
     {limitRadius=0, drivingBone {boneName=None}, sphereLocalOffset=<0 0 0>}
    ]
   }
   n111 = Anim Dynamics - Bone: dyn_l_braid_011(ComponentPose=@n110, Alpha=1, GravityOverride="0,0,0", ExternalForce="0,0,0") {
    angularDampingOverride=1
    angularSpringConstant=2500
    bAngularSpring=true
    bGravityOverrideInSimSpace=true
    bOverrideAngularDamping=true
    bUseSphericalLimits=true
    gravityScale=0.5
    simulationSpace=RootRelative
    alphaBoolBlend {blendOption=Linear}
    alphaScaleBiasClamp {scale=0.5}
    boundBone {boneName=dyn_l_braid_011}
    componentAppliedLinearAccClamp=<5000 5000 5000>
    componentLinearAccScale {z=0.7}
    componentLinearVelScale {z=0.7}
    physicsBodyDefinitions [{
      collisionType=CustomSphere
      sphereCollisionRadius=1
      boundBone {boneName=dyn_l_braid_011}
      boxExtents=<12 22 22>
      constraintSetup {angularLimitsMax=<20 90 90>, angularLimitsMin=<-20 -90 -90>, angularTarget=<0 0 0>}
      localJointOffset=<10 0 0>
     }]
    planarLimits [{drivingBone {boneName=root}, planeTransform {location=<0 0 3>, rotation=<rot 0 0 0>}}]
    sphericalLimits [
     {limitRadius=65.9176, drivingBone {boneName=spine_05}, sphereLocalOffset=<-12.8628 65 0>}
     {limitRadius=50, drivingBone {boneName=pelvis}, sphereLocalOffset=<-2.6826 40 0>}
     {limitRadius=28, drivingBone {boneName=spine_02}, sphereLocalOffset=<2.6549 17 0>}
     {limitRadius=1, drivingBone {boneName=pelvis}, sphereLocalOffset=<-30 0 0>}
     {limitRadius=0, drivingBone {boneName=None}, sphereLocalOffset=<0 0 0>}
    ]
   }
   n112 = Anim Dynamics - Bone: dyn_r_braid_05(ComponentPose=@n111, Alpha=1, GravityOverride="0,0,0", ExternalForce="0,0,0") {
    angularDampingOverride=1
    angularSpringConstant=2500
    bAngularSpring=true
    bGravityOverrideInSimSpace=true
    bOverrideAngularDamping=true
    bUseSphericalLimits=true
    gravityScale=0.5
    simulationSpace=RootRelative
    alphaBoolBlend {blendOption=Linear}
    alphaScaleBiasClamp {scale=0.55}
    boundBone {boneName=dyn_r_braid_05}
    componentAppliedLinearAccClamp=<5000 5000 5000>
    componentLinearAccScale {z=0.7}
    componentLinearVelScale {z=0.7}
    physicsBodyDefinitions [{
      collisionType=CustomSphere
      sphereCollisionRadius=1
      boundBone {boneName=dyn_r_braid_05}
      boxExtents=<14 28 28>
      constraintSetup {angularLimitsMax=<20 45 90>, angularLimitsMin=<-20 -45 -90>, angularTarget=<0 0 0>}
      localJointOffset=<10 -0.5 0>
     }]
    sphericalLimits [
     {limitRadius=65.9176, drivingBone {boneName=spine_05}, sphereLocalOffset=<-12.8628 55 0>}
     {limitRadius=50, drivingBone {boneName=pelvis}, sphereLocalOffset=<-2.6826 40 0>}
     {limitRadius=28, drivingBone {boneName=spine_02}, sphereLocalOffset=<2.6549 15 0>}
     {limitRadius=1, drivingBone {boneName=pelvis}, sphereLocalOffset=<-30 0 0>}
    ]
   }
   n113 = Anim Dynamics - Bone: dyn_r_braid_06(ComponentPose=@n112, Alpha=1, GravityOverride="0,0,0", ExternalForce="0,0,0") {
    angularDampingOverride=1
    angularSpringConstant=2500
    bAngularSpring=true
    bGravityOverrideInSimSpace=true
    bOverrideAngularDamping=true
    bUseSphericalLimits=true
    gravityScale=0.5
    simulationSpace=RootRelative
    alphaBoolBlend {blendOption=Linear}
    alphaScaleBiasClamp {scale=0.4}
    boundBone {boneName=dyn_r_braid_06}
    componentAppliedLinearAccClamp=<5000 5000 5000>
    componentLinearAccScale {z=0.7}
    componentLinearVelScale {z=0.7}
    physicsBodyDefinitions [{
      collisionType=CustomSphere
      sphereCollisionRadius=1
      boundBone {boneName=dyn_r_braid_06}
      boxExtents=<14 28 28>
      constraintSetup {angularLimitsMax=<20 45 90>, angularLimitsMin=<-20 -45 -90>, angularTarget=<0 0 0>}
      localJointOffset=<10 -0.5 0>
     }]
    planarLimits [{drivingBone {boneName=root}, planeTransform {location=<0 0 5>, rotation=<rot 0 0 0>}}]
    sphericalLimits [
     {limitRadius=65.9176, drivingBone {boneName=spine_05}, sphereLocalOffset=<-12.8628 55 0>}
     {limitRadius=50, drivingBone {boneName=pelvis}, sphereLocalOffset=<-2.6826 35 0>}
     {limitRadius=28, drivingBone {boneName=spine_02}, sphereLocalOffset=<2.6549 15 0>}
     {limitRadius=1, drivingBone {boneName=pelvis}, sphereLocalOffset=<-30 0 0>}
    ]
   }
   n114 = Anim Dynamics - Bone: dyn_r_braid_07(ComponentPose=@n113, Alpha=1, GravityOverride="0,0,0", ExternalForce="0,0,0") {
    angularDampingOverride=1
    angularSpringConstant=2500
    bAngularSpring=true
    bGravityOverrideInSimSpace=true
    bOverrideAngularDamping=true
    bUseSphericalLimits=true
    gravityScale=0.5
    simulationSpace=RootRelative
    alphaBoolBlend {blendOption=Linear}
    alphaScaleBiasClamp {scale=0.4}
    boundBone {boneName=dyn_r_braid_07}
    componentAppliedLinearAccClamp=<5000 5000 5000>
    componentLinearAccScale {z=0.7}
    componentLinearVelScale {z=0.7}
    physicsBodyDefinitions [{
      collisionType=CustomSphere
      sphereCollisionRadius=1
      boundBone {boneName=dyn_r_braid_07}
      boxExtents=<12 22 22>
      constraintSetup {angularLimitsMax=<20 90 90>, angularLimitsMin=<-20 -90 -90>, angularTarget=<0 0 0>}
      localJointOffset=<10 0 0>
     }]
    planarLimits [{drivingBone {boneName=root}, planeTransform {location=<0 0 3>, rotation=<rot 0 0 0>}}]
    sphericalLimits [
     {limitRadius=65.9176, drivingBone {boneName=spine_05}, sphereLocalOffset=<-12.8628 65 0>}
     {limitRadius=50, drivingBone {boneName=pelvis}, sphereLocalOffset=<-2.6826 35 0>}
     {limitRadius=28, drivingBone {boneName=spine_02}, sphereLocalOffset=<2.6549 17 0>}
     {limitRadius=1, drivingBone {boneName=pelvis}, sphereLocalOffset=<-30 0 0>}
    ]
   }
   n115 = Anim Dynamics - Bone: dyn_r_braid_08(ComponentPose=@n114, Alpha=1, GravityOverride="0,0,0", ExternalForce="0,0,0") {
    angularDampingOverride=1
    angularSpringConstant=2500
    bAngularSpring=true
    bGravityOverrideInSimSpace=true
    bOverrideAngularDamping=true
    bUseSphericalLimits=true
    gravityScale=0.5
    simulationSpace=RootRelative
    alphaBoolBlend {blendOption=Linear}
    alphaScaleBiasClamp {scale=0.5}
    boundBone {boneName=dyn_r_braid_08}
    componentAppliedLinearAccClamp=<5000 5000 5000>
    componentLinearAccScale {z=0.7}
    componentLinearVelScale {z=0.7}
    physicsBodyDefinitions [{
      collisionType=CustomSphere
      sphereCollisionRadius=1
      boundBone {boneName=dyn_r_braid_08}
      boxExtents=<12 22 22>
      constraintSetup {angularLimitsMax=<20 90 90>, angularLimitsMin=<-20 -90 -90>, angularTarget=<0 0 0>}
      localJointOffset=<10 0 0>
     }]
    planarLimits [{drivingBone {boneName=root}, planeTransform {location=<0 0 3>, rotation=<rot 0 0 0>}}]
    sphericalLimits [
     {limitRadius=65.9176, drivingBone {boneName=spine_05}, sphereLocalOffset=<-12.8628 65 0>}
     {limitRadius=50, drivingBone {boneName=pelvis}, sphereLocalOffset=<-2.6826 40 0>}
     {limitRadius=28, drivingBone {boneName=spine_02}, sphereLocalOffset=<2.6549 17 0>}
     {limitRadius=1, drivingBone {boneName=pelvis}, sphereLocalOffset=<-30 0 0>}
    ]
   }
   n116 = Anim Dynamics - Bone: dyn_r_braid_09(ComponentPose=@n115, Alpha=1, GravityOverride="0,0,0", ExternalForce="0,0,0") {
    angularDampingOverride=1
    angularSpringConstant=2500
    bAngularSpring=true
    bGravityOverrideInSimSpace=true
    bOverrideAngularDamping=true
    bUseSphericalLimits=true
    gravityScale=0.5
    simulationSpace=RootRelative
    alphaBoolBlend {blendOption=Linear}
    alphaScaleBiasClamp {scale=0.5}
    boundBone {boneName=dyn_r_braid_09}
    componentAppliedLinearAccClamp=<5000 5000 5000>
    componentLinearAccScale {z=0.7}
    componentLinearVelScale {z=0.7}
    physicsBodyDefinitions [{
      collisionType=CustomSphere
      sphereCollisionRadius=1
      boundBone {boneName=dyn_r_braid_09}
      boxExtents=<12 22 22>
      constraintSetup {angularLimitsMax=<20 90 90>, angularLimitsMin=<-20 -90 -90>, angularTarget=<0 0 0>}
      localJointOffset=<1 0 0>
     }]
    planarLimits [{drivingBone {boneName=root}, planeTransform {location=<0 0 3>, rotation=<rot 0 0 0>}}]
    sphericalLimits [
     {limitRadius=65.9176, drivingBone {boneName=spine_05}, sphereLocalOffset=<-12.8628 65 0>}
     {limitRadius=50, drivingBone {boneName=pelvis}, sphereLocalOffset=<-2.6826 40 0>}
     {limitRadius=28, drivingBone {boneName=spine_02}, sphereLocalOffset=<2.6549 17 0>}
     {limitRadius=1, drivingBone {boneName=pelvis}, sphereLocalOffset=<-30 0 0>}
    ]
   }
   n117 = Anim Dynamics - Bone: dyn_r_braid_010(ComponentPose=@n116, Alpha=1, GravityOverride="0,0,0", ExternalForce="0,0,0") {
    angularDampingOverride=1
    angularSpringConstant=2500
    bAngularSpring=true
    bGravityOverrideInSimSpace=true
    bOverrideAngularDamping=true
    bUseSphericalLimits=true
    gravityScale=0.5
    simulationSpace=RootRelative
    alphaBoolBlend {blendOption=Linear}
    alphaScaleBiasClamp {scale=0.5}
    boundBone {boneName=dyn_r_braid_010}
    componentAppliedLinearAccClamp=<5000 5000 5000>
    componentLinearAccScale {z=0.7}
    componentLinearVelScale {z=0.7}
    physicsBodyDefinitions [{
      collisionType=CustomSphere
      sphereCollisionRadius=1
      boundBone {boneName=dyn_r_braid_010}
      boxExtents=<12 22 22>
      constraintSetup {angularLimitsMax=<20 45 15>, angularLimitsMin=<-20 -45 -15>, angularTarget=<0 0 0>}
      localJointOffset=<5 0 0>
     }]
    planarLimits [{drivingBone {boneName=root}, planeTransform {location=<0 0 5>, rotation=<rot 0 0 0>}}]
    sphericalLimits [
     {limitRadius=65.9176, drivingBone {boneName=spine_05}, sphereLocalOffset=<-12.8628 65 0>}
     {limitRadius=50, drivingBone {boneName=pelvis}, sphereLocalOffset=<-2.6826 40 0>}
     {limitRadius=28, drivingBone {boneName=spine_02}, sphereLocalOffset=<2.6549 17 0>}
     {limitRadius=1, drivingBone {boneName=pelvis}, sphereLocalOffset=<-30 0 0>}
    ]
   }
   n118 = Anim Dynamics - Bone: dyn_r_braid_011(ComponentPose=@n117, Alpha=1, GravityOverride="0,0,0", ExternalForce="0,0,0") {
    angularDampingOverride=1
    angularSpringConstant=2500
    bAngularSpring=true
    bGravityOverrideInSimSpace=true
    bOverrideAngularDamping=true
    bUseSphericalLimits=true
    gravityScale=0.5
    simulationSpace=RootRelative
    alphaBoolBlend {blendOption=Linear}
    alphaScaleBiasClamp {scale=0.5}
    boundBone {boneName=dyn_r_braid_011}
    componentAppliedLinearAccClamp=<5000 5000 5000>
    componentLinearAccScale {z=0.7}
    componentLinearVelScale {z=0.7}
    physicsBodyDefinitions [{
      collisionType=CustomSphere
      sphereCollisionRadius=1
      boundBone {boneName=dyn_r_braid_011}
      boxExtents=<12 22 22>
      constraintSetup {angularLimitsMax=<20 45 15>, angularLimitsMin=<-20 -45 -15>, angularTarget=<0 0 0>}
      localJointOffset=<5 0 0>
     }]
    planarLimits [{drivingBone {boneName=root}, planeTransform {location=<0 0 5>, rotation=<rot 0 0 0>}}]
    sphericalLimits [
     {limitRadius=65.9176, drivingBone {boneName=spine_05}, sphereLocalOffset=<-12.8628 65 0>}
     {limitRadius=50, drivingBone {boneName=pelvis}, sphereLocalOffset=<-2.6826 40 0>}
     {limitRadius=28, drivingBone {boneName=spine_02}, sphereLocalOffset=<2.6549 17 0>}
     {limitRadius=1, drivingBone {boneName=pelvis}, sphereLocalOffset=<-30 0 0>}
     {limitRadius=0, drivingBone {boneName=None}, sphereLocalOffset=<0 0 0>}
    ]
   }
   n119 = Component To Local(ComponentPose=@n118)
   n120 = Get bIsDBNOCarried()
   n121 = "Transform (Modify) Bone - Bone: dyn_r_braid_06"(
    ComponentPose=Local To Component(LocalPose=Blend Poses by bool(BlendPose_0=@n103, BlendPose_1=@n119, bActiveValue=@n120) {blendType=Linear})
    Alpha=1
    Translation="0,0,0"
    Rotation="15,0,0"
    Scale="1,1,1"
   ) {
    rotationMode=BMM_Additive
    rotationSpace=BCS_BoneSpace
    scaleMode=BMM_Additive
    translationMode=BMM_Additive
    alphaBoolBlend {blendOption=Linear}
    boneToModify {boneName=dyn_r_braid_06}
    rotation {pitch=15}
   }
   n122 = "Transform (Modify) Bone - Bone: dyn_l_braid_06"(ComponentPose=@n121, Alpha=1, Translation="0,0,0", Rotation="15,0,0", Scale="1,1,1") {
    rotationMode=BMM_Additive
    rotationSpace=BCS_BoneSpace
    scaleMode=BMM_Additive
    translationMode=BMM_Additive
    alphaBoolBlend {blendOption=Linear}
    boneToModify {boneName=dyn_l_braid_06}
    rotation {pitch=15}
   }
   n123 = "Transform (Modify) Bone - Bone: dyn_l_braid_02"!(ComponentPose=@n122, bAlphaBoolEnabled:bool=Get bIsDBNOCarried(), Translation="0,0,0", Rotation="5,0,0", Scale="1,1,1") -> (Pose) {
    alphaInputType=Bool
    rotationMode=BMM_Additive
    rotationSpace=BCS_BoneSpace
    scaleMode=BMM_Additive
    translationMode=BMM_Additive
    alphaBoolBlend {blendOption=Linear}
    boneToModify {boneName=dyn_l_braid_02}
    rotation {pitch=5}
   }
   Save cached pose 'dbno'(Pose=Component To Local(ComponentPose=@n123)) {cacheName=dbno}
   n124 = Control Rig(Source=Use cached pose 'Prebraid sim'()) {
    alphaBoolBlend {blendOption=Linear}
    controlRigAssetReference {blueprintRigClass=&F_MED_Uproar_Braids_FaceAcc_CtrlRig_C}
   }
   n125 = "Transform (Modify) Bone - Bone: dyn_r_braid_04"(ComponentPose=Local To Component(LocalPose=@n124), Alpha=1, Translation="0,0,0", Rotation="0,0,8.135", Scale="0.9,1,1") {
    rotationMode=BMM_Additive
    translationMode=BMM_Additive
    alphaBoolBlend {blendOption=Linear}
    boneToModify {boneName=dyn_r_braid_04}
    rotation {roll=8.135}
    scale {x=0.9}
   }
   n126 = "Transform (Modify) Bone - Bone: dyn_r_braid_06"(ComponentPose=@n125, Alpha=1, Translation="0,0,0", Rotation="0,0,6.167", Scale="0.9,1,1") {
    rotationMode=BMM_Additive
    translationMode=BMM_Additive
    alphaBoolBlend {blendOption=Linear}
    boneToModify {boneName=dyn_r_braid_06}
    rotation {roll=6.167}
    scale {x=0.9}
   }
   n127 = "Transform (Modify) Bone - Bone: dyn_r_braid_08"(ComponentPose=@n126, Alpha=1, Translation="0,0,0", Rotation="0,0,0", Scale="0.9,1,1") {
    rotationMode=BMM_Additive
    translationMode=BMM_Additive
    alphaBoolBlend {blendOption=Linear}
    boneToModify {boneName=dyn_r_braid_08}
    scale {x=0.9}
   }
   n128 = "Transform (Modify) Bone - Bone: dyn_r_braid_010"(ComponentPose=@n127, Alpha=1, Translation="0,0,0", Rotation="0,0,0", Scale="0.9,1,1") {
    rotationMode=BMM_Additive
    translationMode=BMM_Additive
    alphaBoolBlend {blendOption=Linear}
    boneToModify {boneName=dyn_r_braid_010}
    scale {x=0.9}
   }
   n129 = "Transform (Modify) Bone - Bone: dyn_l_braid_04"(ComponentPose=@n128, Alpha=1, Translation="0,0,0", Rotation="0.000,0.000,8.135", Scale="0.9,1,1") {
    rotationMode=BMM_Additive
    translationMode=BMM_Additive
    alphaBoolBlend {blendOption=Linear}
    boneToModify {boneName=dyn_l_braid_04}
    rotation {roll=8.135}
    scale {x=0.9}
   }
   n130 = "Transform (Modify) Bone - Bone: dyn_l_braid_06"(ComponentPose=@n129, Alpha=1, Translation="0,0,0", Rotation="0.000,0.000,6.167", Scale="0.9,1,1") {
    rotationMode=BMM_Additive
    translationMode=BMM_Additive
    alphaBoolBlend {blendOption=Linear}
    boneToModify {boneName=dyn_l_braid_06}
    rotation {roll=6.167}
    scale {x=0.9}
   }
   n131 = "Transform (Modify) Bone - Bone: dyn_l_braid_08"(ComponentPose=@n130, Alpha=1, Translation="0,0,0", Rotation="0,0,0", Scale="0.9,1,1") {
    rotationMode=BMM_Additive
    translationMode=BMM_Additive
    alphaBoolBlend {blendOption=Linear}
    boneToModify {boneName=dyn_l_braid_08}
    scale {x=0.9}
   }
   n132 = "Transform (Modify) Bone - Bone: dyn_l_braid_010"(ComponentPose=@n131, Alpha=1, Translation="0,0,0", Rotation="0,0,0", Scale="0.9,1,1") {
    rotationMode=BMM_Additive
    translationMode=BMM_Additive
    alphaBoolBlend {blendOption=Linear}
    boneToModify {boneName=dyn_l_braid_010}
    scale {x=0.9}
   }
   n133 = Get GravityOverride()
   n134 = Anim Dynamics - Bone: dyn_l_braid_04(ComponentPose=@n132, Alpha=1, GravityOverride=@n133, ExternalForce="0,0,0") {
    angularDampingOverride=1
    angularSpringConstant=2500
    bAngularSpring=true
    bGravityOverrideInSimSpace=true
    bOverrideAngularDamping=true
    bUseSphericalLimits=true
    gravityScale=0.5
    simulationSpace=RootRelative
    alphaBoolBlend {blendOption=Linear}
    alphaScaleBiasClamp {scale=0.35}
    boundBone {boneName=dyn_l_braid_04}
    componentAppliedLinearAccClamp=<2500 2500 2500>
    componentLinearAccScale {z=0.35}
    componentLinearVelScale {z=0.7}
    physicsBodyDefinitions [{
      collisionType=CustomSphere
      sphereCollisionRadius=1
      boundBone {boneName=dyn_l_braid_04}
      boxExtents=<12 22 22>
      constraintSetup {angularLimitsMax=<20 45 2.5>, angularLimitsMin=<-20 -45 -2.5>, angularTarget=<0 0 0>}
      localJointOffset=<10 0 0>
     }]
    sphericalLimits [
     {limitRadius=65.9176, drivingBone {boneName=spine_05}, sphereLocalOffset=<-12.8628 55 0>}
     {limitRadius=50, drivingBone {boneName=pelvis}, sphereLocalOffset=<-2.6826 40 0>}
     {limitRadius=28, drivingBone {boneName=spine_02}, sphereLocalOffset=<2.6549 15 0>}
     {limitRadius=1, drivingBone {boneName=pelvis}, sphereLocalOffset=<-30 0 0>}
    ]
   }
   n135 = Anim Dynamics - Bone: dyn_l_braid_05(ComponentPose=@n134, Alpha=1, GravityOverride=@n133, ExternalForce="0,0,0") {
    angularDampingOverride=1
    angularSpringConstant=2500
    bAngularSpring=true
    bGravityOverrideInSimSpace=true
    bOverrideAngularDamping=true
    bUseSphericalLimits=true
    gravityScale=0.5
    simulationSpace=RootRelative
    alphaBoolBlend {blendOption=Linear}
    alphaScaleBiasClamp {scale=0.35}
    boundBone {boneName=dyn_l_braid_05}
    componentAppliedLinearAccClamp=<2500 2500 2500>
    componentLinearAccScale {z=0.35}
    componentLinearVelScale {z=0.7}
    physicsBodyDefinitions [{
      collisionType=CustomSphere
      sphereCollisionRadius=1
      boundBone {boneName=dyn_l_braid_05}
      boxExtents=<12 22 22>
      constraintSetup {angularLimitsMax=<20 45 2.5>, angularLimitsMin=<-20 -45 -2.5>, angularTarget=<0 0 0>}
      localJointOffset=<10 0 0>
     }]
    sphericalLimits [
     {limitRadius=65.9176, drivingBone {boneName=spine_05}, sphereLocalOffset=<-12.8628 55 0>}
     {limitRadius=50, drivingBone {boneName=pelvis}, sphereLocalOffset=<-2.6826 35 0>}
     {limitRadius=28, drivingBone {boneName=spine_02}, sphereLocalOffset=<2.6549 15 0>}
     {limitRadius=1, drivingBone {boneName=pelvis}, sphereLocalOffset=<-30 0 0>}
    ]
   }
   n136 = Anim Dynamics - Bone: dyn_l_braid_06(ComponentPose=@n135, Alpha=1, GravityOverride=@n133, ExternalForce="0,0,0") {
    angularDampingOverride=1
    angularSpringConstant=2500
    bAngularSpring=true
    bGravityOverrideInSimSpace=true
    bOverrideAngularDamping=true
    bUsePlanarLimit=false
    bUseSphericalLimits=true
    gravityScale=0.5
    simulationSpace=RootRelative
    alphaBoolBlend {blendOption=Linear}
    alphaScaleBiasClamp {scale=0.35}
    boundBone {boneName=dyn_l_braid_06}
    componentAppliedLinearAccClamp=<1500 1500 1500>
    componentLinearAccScale {z=0.35}
    componentLinearVelScale {z=0.35}
    physicsBodyDefinitions [{
      collisionType=CustomSphere
      sphereCollisionRadius=1
      boundBone {boneName=dyn_l_braid_06}
      boxExtents=<12 22 22>
      constraintSetup {angularLimitsMax=<10 45 5>, angularLimitsMin=<-10 -45 -5>, angularTarget=<0 0 0>}
      localJointOffset=<10 0 0>
     }]
    sphericalLimits [
     {limitRadius=65.9176, drivingBone {boneName=spine_05}, sphereLocalOffset=<-12.8628 55 0>}
     {limitRadius=50, drivingBone {boneName=pelvis}, sphereLocalOffset=<-2.6826 35 0>}
     {limitRadius=28, drivingBone {boneName=spine_02}, sphereLocalOffset=<2.6549 15 0>}
     {limitRadius=1, drivingBone {boneName=pelvis}, sphereLocalOffset=<-30 0 0>}
    ]
   }
   n137 = Anim Dynamics - Bone: dyn_l_braid_07(ComponentPose=@n136, Alpha=1, GravityOverride=@n133, ExternalForce="0,0,0") {
    angularDampingOverride=1
    angularSpringConstant=2500
    bAngularSpring=true
    bGravityOverrideInSimSpace=true
    bOverrideAngularDamping=true
    bUsePlanarLimit=false
    bUseSphericalLimits=true
    gravityScale=0.5
    simulationSpace=RootRelative
    alphaBoolBlend {blendOption=Linear}
    alphaScaleBiasClamp {scale=0.4}
    boundBone {boneName=dyn_l_braid_07}
    componentAppliedLinearAccClamp=<1500 1500 1500>
    componentLinearAccScale {z=0.35}
    componentLinearVelScale {z=0.35}
    physicsBodyDefinitions [{
      collisionType=CustomSphere
      sphereCollisionRadius=1
      boundBone {boneName=dyn_l_braid_07}
      boxExtents=<12 22 22>
      constraintSetup {angularLimitsMax=<10 45 5>, angularLimitsMin=<-10 -45 -5>, angularTarget=<0 0 0>}
      localJointOffset=<10 0 0>
     }]
    sphericalLimits [
     {limitRadius=65.9176, drivingBone {boneName=spine_05}, sphereLocalOffset=<-12.8628 65 0>}
     {limitRadius=50, drivingBone {boneName=pelvis}, sphereLocalOffset=<-2.6826 35 0>}
     {limitRadius=28, drivingBone {boneName=spine_02}, sphereLocalOffset=<2.6549 17 0>}
     {limitRadius=1, drivingBone {boneName=pelvis}, sphereLocalOffset=<-30 0 0>}
    ]
   }
   n138 = Anim Dynamics - Bone: dyn_l_braid_08(ComponentPose=@n137, Alpha=1, GravityOverride=@n133, ExternalForce="0,0,0") {
    angularDampingOverride=1
    angularSpringConstant=2500
    bAngularSpring=true
    bGravityOverrideInSimSpace=true
    bOverrideAngularDamping=true
    bUsePlanarLimit=false
    bUseSphericalLimits=true
    gravityScale=0.5
    simulationSpace=RootRelative
    alphaBoolBlend {blendOption=Linear}
    alphaScaleBiasClamp {scale=0.5}
    boundBone {boneName=dyn_l_braid_08}
    componentAppliedLinearAccClamp=<1500 1500 1500>
    componentLinearAccScale {z=0.35}
    componentLinearVelScale {z=0.35}
    physicsBodyDefinitions [{
      collisionType=CustomSphere
      sphereCollisionRadius=1
      boundBone {boneName=dyn_l_braid_08}
      boxExtents=<12 22 22>
      constraintSetup {angularLimitsMax=<10 45 5>, angularLimitsMin=<-10 -45 -5>, angularTarget=<0 0 0>}
      localJointOffset=<10 0 0>
     }]
    sphericalLimits [
     {limitRadius=65.9176, drivingBone {boneName=spine_05}, sphereLocalOffset=<-12.8628 65 0>}
     {limitRadius=50, drivingBone {boneName=pelvis}, sphereLocalOffset=<-2.6826 40 0>}
     {limitRadius=28, drivingBone {boneName=spine_02}, sphereLocalOffset=<2.6549 17 0>}
     {limitRadius=1, drivingBone {boneName=pelvis}, sphereLocalOffset=<-30 0 0>}
    ]
   }
   n139 = Anim Dynamics - Bone: dyn_l_braid_09(ComponentPose=@n138, Alpha=1, GravityOverride=@n133, ExternalForce="0,0,0") {
    angularDampingOverride=1
    angularSpringConstant=2500
    bAngularSpring=true
    bGravityOverrideInSimSpace=true
    bOverrideAngularDamping=true
    bUsePlanarLimit=false
    bUseSphericalLimits=true
    gravityScale=0.5
    simulationSpace=RootRelative
    alphaBoolBlend {blendOption=Linear}
    alphaScaleBiasClamp {scale=0.5}
    boundBone {boneName=dyn_l_braid_09}
    componentAppliedLinearAccClamp=<1500 1500 1500>
    componentLinearAccScale {z=0.35}
    componentLinearVelScale {z=0.35}
    physicsBodyDefinitions [{
      collisionType=CustomSphere
      sphereCollisionRadius=1
      boundBone {boneName=dyn_l_braid_09}
      boxExtents=<12 22 22>
      constraintSetup {angularLimitsMax=<10 45 5>, angularLimitsMin=<-10 -45 -5>, angularTarget=<0 0 0>}
      localJointOffset=<10 0 0>
     }]
    sphericalLimits [
     {limitRadius=65.9176, drivingBone {boneName=spine_05}, sphereLocalOffset=<-12.8628 65 0>}
     {limitRadius=50, drivingBone {boneName=pelvis}, sphereLocalOffset=<-2.6826 40 0>}
     {limitRadius=28, drivingBone {boneName=spine_02}, sphereLocalOffset=<2.6549 17 0>}
     {limitRadius=1, drivingBone {boneName=pelvis}, sphereLocalOffset=<-30 0 0>}
    ]
   }
   n140 = Anim Dynamics - Bone: dyn_l_braid_010(ComponentPose=@n139, Alpha=1, GravityOverride=@n133, ExternalForce="0,0,0") {
    angularDampingOverride=1
    angularSpringConstant=2500
    bAngularSpring=true
    bGravityOverrideInSimSpace=true
    bOverrideAngularDamping=true
    bUsePlanarLimit=false
    bUseSphericalLimits=true
    gravityScale=0.5
    simulationSpace=RootRelative
    alphaBoolBlend {blendOption=Linear}
    alphaScaleBiasClamp {scale=0.5}
    boundBone {boneName=dyn_l_braid_010}
    componentAppliedLinearAccClamp=<1500 1500 1500>
    componentLinearAccScale {z=0.35}
    componentLinearVelScale {z=0.35}
    physicsBodyDefinitions [{
      collisionType=CustomSphere
      sphereCollisionRadius=1
      boundBone {boneName=dyn_l_braid_010}
      boxExtents=<12 22 22>
      constraintSetup {angularLimitsMax=<10 45 5>, angularLimitsMin=<-10 -45 -5>, angularTarget=<0 0 0>}
      localJointOffset=<10 0 0>
     }]
    sphericalLimits [
     {limitRadius=65.9176, drivingBone {boneName=spine_05}, sphereLocalOffset=<-12.8628 65 0>}
     {limitRadius=50, drivingBone {boneName=pelvis}, sphereLocalOffset=<-2.6826 40 0>}
     {limitRadius=28, drivingBone {boneName=spine_02}, sphereLocalOffset=<2.6549 17 0>}
     {limitRadius=1, drivingBone {boneName=pelvis}, sphereLocalOffset=<-30 0 0>}
     {limitRadius=0, drivingBone {boneName=None}, sphereLocalOffset=<0 0 0>}
    ]
   }
   n141 = Anim Dynamics - Bone: dyn_l_braid_011(ComponentPose=@n140, Alpha=1, GravityOverride=@n133, ExternalForce="0,0,0") {
    angularDampingOverride=1
    angularSpringConstant=2500
    bAngularSpring=true
    bGravityOverrideInSimSpace=true
    bOverrideAngularDamping=true
    bUsePlanarLimit=false
    bUseSphericalLimits=true
    gravityScale=0.5
    simulationSpace=RootRelative
    alphaBoolBlend {blendOption=Linear}
    alphaScaleBiasClamp {scale=0.5}
    boundBone {boneName=dyn_l_braid_011}
    componentAppliedLinearAccClamp=<1500 1500 1500>
    componentLinearAccScale {z=0.35}
    componentLinearVelScale {z=0.35}
    physicsBodyDefinitions [{
      collisionType=CustomSphere
      sphereCollisionRadius=1
      boundBone {boneName=dyn_l_braid_011}
      boxExtents=<12 22 22>
      constraintSetup {angularLimitsMax=<10 45 5>, angularLimitsMin=<-10 -45 -5>, angularTarget=<0 0 0>}
      localJointOffset=<10 0 0>
     }]
    sphericalLimits [
     {limitRadius=65.9176, drivingBone {boneName=spine_05}, sphereLocalOffset=<-12.8628 65 0>}
     {limitRadius=50, drivingBone {boneName=pelvis}, sphereLocalOffset=<-2.6826 40 0>}
     {limitRadius=28, drivingBone {boneName=spine_02}, sphereLocalOffset=<2.6549 17 0>}
     {limitRadius=1, drivingBone {boneName=pelvis}, sphereLocalOffset=<-30 0 0>}
     {limitRadius=0, drivingBone {boneName=None}, sphereLocalOffset=<0 0 0>}
    ]
   }
   n142 = Anim Dynamics - Bone: dyn_r_braid_04(ComponentPose=@n141, Alpha=1, GravityOverride=@n133, ExternalForce="0,0,0") {
    angularDampingOverride=1
    angularSpringConstant=2500
    bAngularSpring=true
    bGravityOverrideInSimSpace=true
    bOverrideAngularDamping=true
    bUseSphericalLimits=true
    gravityScale=0.5
    simulationSpace=RootRelative
    alphaBoolBlend {blendOption=Linear}
    alphaScaleBiasClamp {scale=0.35}
    boundBone {boneName=dyn_r_braid_04}
    componentAppliedLinearAccClamp=<2500 2500 2500>
    componentLinearAccScale {z=0.35}
    componentLinearVelScale {z=0.7}
    physicsBodyDefinitions [{
      collisionType=CustomSphere
      sphereCollisionRadius=1
      boundBone {boneName=dyn_r_braid_04}
      boxExtents=<14 28 28>
      constraintSetup {angularLimitsMax=<20 45 2.5>, angularLimitsMin=<-20 -45 -2.5>, angularTarget=<0 0 0>}
      localJointOffset=<10 -0.5 0>
     }]
    sphericalLimits [
     {limitRadius=65.9176, drivingBone {boneName=spine_05}, sphereLocalOffset=<-12.8628 55 0>}
     {limitRadius=50, drivingBone {boneName=pelvis}, sphereLocalOffset=<-2.6826 40 0>}
     {limitRadius=28, drivingBone {boneName=spine_02}, sphereLocalOffset=<2.6549 15 0>}
     {limitRadius=1, drivingBone {boneName=pelvis}, sphereLocalOffset=<-30 0 0>}
    ]
   }
   n143 = Anim Dynamics - Bone: dyn_r_braid_05(ComponentPose=@n142, Alpha=1, GravityOverride=@n133, ExternalForce="0,0,0") {
    angularDampingOverride=1
    angularSpringConstant=2500
    bAngularSpring=true
    bGravityOverrideInSimSpace=true
    bOverrideAngularDamping=true
    bUseSphericalLimits=true
    gravityScale=0.5
    simulationSpace=RootRelative
    alphaBoolBlend {blendOption=Linear}
    alphaScaleBiasClamp {scale=0.35}
    boundBone {boneName=dyn_r_braid_05}
    componentAppliedLinearAccClamp=<2500 2500 2500>
    componentLinearAccScale {z=0.35}
    componentLinearVelScale {z=0.7}
    physicsBodyDefinitions [{
      collisionType=CustomSphere
      sphereCollisionRadius=1
      boundBone {boneName=dyn_r_braid_05}
      boxExtents=<14 28 28>
      constraintSetup {angularLimitsMax=<20 45 2.5>, angularLimitsMin=<-20 -45 -2.5>, angularTarget=<0 0 0>}
      localJointOffset=<10 -0.5 0>
     }]
    sphericalLimits [
     {limitRadius=65.9176, drivingBone {boneName=spine_05}, sphereLocalOffset=<-12.8628 55 0>}
     {limitRadius=50, drivingBone {boneName=pelvis}, sphereLocalOffset=<-2.6826 40 0>}
     {limitRadius=28, drivingBone {boneName=spine_02}, sphereLocalOffset=<2.6549 15 0>}
     {limitRadius=1, drivingBone {boneName=pelvis}, sphereLocalOffset=<-30 0 0>}
    ]
   }
   n144 = Anim Dynamics - Bone: dyn_r_braid_06(ComponentPose=@n143, Alpha=1, GravityOverride=@n133, ExternalForce="0,0,0") {
    angularDampingOverride=1
    angularSpringConstant=2500
    bAngularSpring=true
    bGravityOverrideInSimSpace=true
    bOverrideAngularDamping=true
    bUsePlanarLimit=false
    bUseSphericalLimits=true
    gravityScale=0.5
    simulationSpace=RootRelative
    alphaBoolBlend {blendOption=Linear}
    alphaScaleBiasClamp {scale=0.35}
    boundBone {boneName=dyn_r_braid_06}
    componentAppliedLinearAccClamp=<1500 1500 1500>
    componentLinearAccScale {z=0.35}
    componentLinearVelScale {z=0.35}
    physicsBodyDefinitions [{
      collisionType=CustomSphere
      sphereCollisionRadius=1
      boundBone {boneName=dyn_r_braid_06}
      boxExtents=<14 28 28>
      constraintSetup {angularLimitsMax=<10 45 5>, angularLimitsMin=<-10 -45 -5>, angularTarget=<0 0 0>}
      localJointOffset=<10 -0.5 0>
     }]
    sphericalLimits [
     {limitRadius=65.9176, drivingBone {boneName=spine_05}, sphereLocalOffset=<-12.8628 55 0>}
     {limitRadius=50, drivingBone {boneName=pelvis}, sphereLocalOffset=<-2.6826 35 0>}
     {limitRadius=28, drivingBone {boneName=spine_02}, sphereLocalOffset=<2.6549 15 0>}
     {limitRadius=1, drivingBone {boneName=pelvis}, sphereLocalOffset=<-30 0 0>}
    ]
   }
   n145 = Anim Dynamics - Bone: dyn_r_braid_07(ComponentPose=@n144, Alpha=1, GravityOverride=@n133, ExternalForce="0,0,0") {
    angularDampingOverride=1
    angularSpringConstant=2500
    bAngularSpring=true
    bGravityOverrideInSimSpace=true
    bOverrideAngularDamping=true
    bUsePlanarLimit=false
    bUseSphericalLimits=true
    gravityScale=0.5
    simulationSpace=RootRelative
    alphaBoolBlend {blendOption=Linear}
    alphaScaleBiasClamp {scale=0.4}
    boundBone {boneName=dyn_r_braid_07}
    componentAppliedLinearAccClamp=<1500 1500 1500>
    componentLinearAccScale {z=0.35}
    componentLinearVelScale {z=0.35}
    physicsBodyDefinitions [{
      collisionType=CustomSphere
      sphereCollisionRadius=1
      boundBone {boneName=dyn_r_braid_07}
      boxExtents=<14 28 28>
      constraintSetup {angularLimitsMax=<10 45 5>, angularLimitsMin=<-10 -45 -5>, angularTarget=<0 0 0>}
      localJointOffset=<10 -0.5 0>
     }]
   }
   n146 = Anim Dynamics - Bone: dyn_r_braid_08(ComponentPose=@n145, Alpha=1, GravityOverride=@n133, ExternalForce="0,0,0") {
    angularDampingOverride=1
    angularSpringConstant=2500
    bAngularSpring=true
    bGravityOverrideInSimSpace=true
    bOverrideAngularDamping=true
    bUsePlanarLimit=false
    bUseSphericalLimits=true
    gravityScale=0.5
    simulationSpace=RootRelative
    alphaBoolBlend {blendOption=Linear}
    alphaScaleBiasClamp {scale=0.5}
    boundBone {boneName=dyn_r_braid_08}
    componentAppliedLinearAccClamp=<1500 1500 1500>
    componentLinearAccScale {z=0.35}
    componentLinearVelScale {z=0.35}
    physicsBodyDefinitions [{
      collisionType=CustomSphere
      sphereCollisionRadius=1
      boundBone {boneName=dyn_r_braid_08}
      boxExtents=<14 28 28>
      constraintSetup {angularLimitsMax=<10 45 5>, angularLimitsMin=<-10 -45 -5>, angularTarget=<0 0 0>}
      localJointOffset=<10 -0.5 0>
     }]
   }
   n147 = Anim Dynamics - Bone: dyn_r_braid_09(ComponentPose=@n146, Alpha=1, GravityOverride=@n133, ExternalForce="0,0,0") {
    angularDampingOverride=1
    angularSpringConstant=2500
    bAngularSpring=true
    bGravityOverrideInSimSpace=true
    bOverrideAngularDamping=true
    bUsePlanarLimit=false
    bUseSphericalLimits=true
    gravityScale=0.5
    simulationSpace=RootRelative
    alphaBoolBlend {blendOption=Linear}
    alphaScaleBiasClamp {scale=0.5}
    boundBone {boneName=dyn_r_braid_09}
    componentAppliedLinearAccClamp=<1500 1500 1500>
    componentLinearAccScale {z=0.35}
    componentLinearVelScale {z=0.35}
    physicsBodyDefinitions [{
      collisionType=CustomSphere
      sphereCollisionRadius=1
      boundBone {boneName=dyn_r_braid_09}
      boxExtents=<14 28 28>
      constraintSetup {angularLimitsMax=<10 45 5>, angularLimitsMin=<-10 -45 -5>, angularTarget=<0 0 0>}
      localJointOffset=<10 -0.5 0>
     }]
   }
   n148 = Anim Dynamics - Bone: dyn_r_braid_010(ComponentPose=@n147, Alpha=1, GravityOverride=@n133, ExternalForce="0,0,0") {
    angularDampingOverride=1
    angularSpringConstant=2500
    bAngularSpring=true
    bGravityOverrideInSimSpace=true
    bOverrideAngularDamping=true
    bUsePlanarLimit=false
    bUseSphericalLimits=true
    gravityScale=0.5
    simulationSpace=RootRelative
    alphaBoolBlend {blendOption=Linear}
    alphaScaleBiasClamp {scale=0.5}
    boundBone {boneName=dyn_r_braid_010}
    componentAppliedLinearAccClamp=<1500 1500 1500>
    componentLinearAccScale {z=0.35}
    componentLinearVelScale {z=0.35}
    physicsBodyDefinitions [{
      collisionType=CustomSphere
      sphereCollisionRadius=1
      boundBone {boneName=dyn_r_braid_010}
      boxExtents=<14 28 28>
      constraintSetup {angularLimitsMax=<10 45 5>, angularLimitsMin=<-10 -45 -5>, angularTarget=<0 0 0>}
      localJointOffset=<10 -0.5 0>
     }]
   }
   n149 = Anim Dynamics - Bone: dyn_r_braid_011(ComponentPose=@n148, Alpha=1, GravityOverride=@n133, ExternalForce="0,0,0") {
    angularDampingOverride=1
    angularSpringConstant=2500
    bAngularSpring=true
    bGravityOverrideInSimSpace=true
    bOverrideAngularDamping=true
    bUsePlanarLimit=false
    bUseSphericalLimits=true
    gravityScale=0.5
    simulationSpace=RootRelative
    alphaBoolBlend {blendOption=Linear}
    alphaScaleBiasClamp {scale=0.5}
    boundBone {boneName=dyn_r_braid_011}
    componentAppliedLinearAccClamp=<1500 1500 1500>
    componentLinearAccScale {z=0.35}
    componentLinearVelScale {z=0.35}
    physicsBodyDefinitions [{
      collisionType=CustomSphere
      sphereCollisionRadius=1
      boundBone {boneName=dyn_r_braid_011}
      boxExtents=<14 28 28>
      constraintSetup {angularLimitsMax=<10 45 5>, angularLimitsMin=<-10 -45 -5>, angularTarget=<0 0 0>}
      localJointOffset=<10 -0.5 0>
     }]
   }
   Save cached pose 'Cape Diving'(Pose=Component To Local(ComponentPose=@n149)) {cacheName=Cape Diving}
   n150 = Control Rig(Source=Use cached pose 'Prebraid sim'()) {
    alphaBoolBlend {blendOption=Linear}
    controlRigAssetReference {blueprintRigClass=&F_MED_Uproar_Braids_FaceAcc_CtrlRig_C}
   }
   n151 = Get bIsCrouching()
   n152 = "Transform (Modify) Bone - Bone: dyn_l_braid_06"!(ComponentPose=Local To Component(LocalPose=@n150), bAlphaBoolEnabled:bool=@n151, Translation="0,0,0", Rotation="10,0.000,30", Scale="1,1,1") -> (Pose) {
    alphaInputType=Bool
    rotationMode=BMM_Additive
    alphaBoolBlend {blendInTime=1, blendOption=Linear, blendOutTime=0.5}
    boneToModify {boneName=dyn_l_braid_06}
    rotation {pitch=10, roll=30}
   }
   n153 = "Transform (Modify) Bone - Bone: dyn_l_braid_07"!(ComponentPose=@n152, bAlphaBoolEnabled:bool=@n151, Translation="0,0,0", Rotation="0.000,0.000,30", Scale="1,1,1") -> (Pose) {
    alphaInputType=Bool
    rotationMode=BMM_Additive
    alphaBoolBlend {blendInTime=1, blendOption=Linear, blendOutTime=0.5}
    boneToModify {boneName=dyn_l_braid_07}
    rotation {roll=30}
   }
   n154 = "Transform (Modify) Bone - Bone: dyn_l_braid_08"!(ComponentPose=@n153, bAlphaBoolEnabled:bool=@n151, Translation="0,0,0", Rotation="0.000,0.000,45", Scale="1,1,1") -> (Pose) {
    alphaInputType=Bool
    rotationMode=BMM_Additive
    alphaBoolBlend {blendInTime=1, blendOption=Linear, blendOutTime=0.5}
    boneToModify {boneName=dyn_l_braid_08}
    rotation {roll=45}
   }
   n155 = "Transform (Modify) Bone - Bone: dyn_l_braid_09"!(ComponentPose=@n154, bAlphaBoolEnabled:bool=@n151, Translation="0,0,0", Rotation="0.000,0.000,45", Scale="1,1,1") -> (Pose) {
    alphaInputType=Bool
    rotationMode=BMM_Additive
    alphaBoolBlend {blendInTime=1, blendOption=Linear, blendOutTime=0.5}
    boneToModify {boneName=dyn_l_braid_09}
    rotation {roll=45}
   }
   n156 = "Transform (Modify) Bone - Bone: dyn_l_braid_010"!(ComponentPose=@n155, bAlphaBoolEnabled:bool=@n151, Translation="0,0,0", Rotation="0.000,0.000,45", Scale="1,1,1") -> (Pose) {
    alphaInputType=Bool
    rotationMode=BMM_Additive
    alphaBoolBlend {blendInTime=1, blendOption=Linear, blendOutTime=0.5}
    boneToModify {boneName=dyn_l_braid_010}
    rotation {roll=45}
   }
   n157 = "Transform (Modify) Bone - Bone: dyn_r_braid_06"!(ComponentPose=@n156, bAlphaBoolEnabled:bool=@n151, Translation="0,0,0", Rotation="10,0.000,30", Scale="1,1,1") -> (Pose) {
    alphaInputType=Bool
    rotationMode=BMM_Additive
    alphaBoolBlend {blendInTime=1, blendOption=Linear, blendOutTime=0.5}
    boneToModify {boneName=dyn_r_braid_06}
    rotation {pitch=10, roll=30}
   }
   n158 = "Transform (Modify) Bone - Bone: dyn_r_braid_07"!(ComponentPose=@n157, bAlphaBoolEnabled:bool=@n151, Translation="0,0,0", Rotation="0.000,0.000,45", Scale="1,1,1") -> (Pose) {
    alphaInputType=Bool
    rotationMode=BMM_Additive
    alphaBoolBlend {blendInTime=1, blendOption=Linear, blendOutTime=0.5}
    boneToModify {boneName=dyn_r_braid_07}
    rotation {roll=45}
   }
   n159 = "Transform (Modify) Bone - Bone: dyn_r_braid_08"!(ComponentPose=@n158, bAlphaBoolEnabled:bool=@n151, Translation="0,0,0", Rotation="0.000,0.000,45", Scale="1,1,1") -> (Pose) {
    alphaInputType=Bool
    rotationMode=BMM_Additive
    alphaBoolBlend {blendInTime=1, blendOption=Linear, blendOutTime=0.5}
    boneToModify {boneName=dyn_r_braid_08}
    rotation {roll=45}
   }
   n160 = "Transform (Modify) Bone - Bone: dyn_r_braid_09"!(ComponentPose=@n159, bAlphaBoolEnabled:bool=@n151, Translation="0,0,0", Rotation="0.000,0.000,30", Scale="1,1,1") -> (Pose) {
    alphaInputType=Bool
    rotationMode=BMM_Additive
    alphaBoolBlend {blendInTime=1, blendOption=Linear, blendOutTime=0.5}
    boneToModify {boneName=dyn_r_braid_09}
    rotation {roll=30}
   }
   n161 = "Transform (Modify) Bone - Bone: dyn_r_braid_010"!(ComponentPose=@n160, bAlphaBoolEnabled:bool=@n151, Translation="0,0,0", Rotation="0.000,0.000,30", Scale="1,1,1") -> (Pose) {
    alphaInputType=Bool
    rotationMode=BMM_Additive
    alphaBoolBlend {blendInTime=1, blendOption=Linear, blendOutTime=0.5}
    boneToModify {boneName=dyn_r_braid_010}
    rotation {roll=30}
   }
   n162 = "Transform (Modify) Bone - Bone: dyn_r_braid_04"(ComponentPose=@n161, Alpha=1, Translation="0,0,0", Rotation="0,0,8.135", Scale="0.9,1,1") {
    rotationMode=BMM_Additive
    translationMode=BMM_Additive
    alphaBoolBlend {blendOption=Linear}
    boneToModify {boneName=dyn_r_braid_04}
    rotation {roll=8.135}
    scale {x=0.9}
   }
   n163 = "Transform (Modify) Bone - Bone: dyn_r_braid_06"!(ComponentPose=@n162, Alpha:float=1, Translation="0,0,0", Rotation="0,0,6.167", Scale="0.9,1,1") -> (Pose) {
    rotationMode=BMM_Additive
    translationMode=BMM_Additive
    alphaBoolBlend {blendOption=Linear}
    boneToModify {boneName=dyn_r_braid_06}
    rotation {roll=6.167}
    scale {x=0.9}
   }
   n164 = "Transform (Modify) Bone - Bone: dyn_r_braid_08"!(ComponentPose=@n163, Alpha:float=1, Translation="0,0,0", Rotation="0,0,0", Scale="0.9,1,1") -> (Pose) {
    rotationMode=BMM_Additive
    translationMode=BMM_Additive
    alphaBoolBlend {blendOption=Linear}
    boneToModify {boneName=dyn_r_braid_08}
    scale {x=0.9}
   }
   n165 = "Transform (Modify) Bone - Bone: dyn_r_braid_010"!(ComponentPose=@n164, Alpha:float=1, Translation="0,0,0", Rotation="0,0,0", Scale="0.9,1,1") -> (Pose) {
    rotationMode=BMM_Additive
    translationMode=BMM_Additive
    alphaBoolBlend {blendOption=Linear}
    boneToModify {boneName=dyn_r_braid_010}
    scale {x=0.9}
   }
   n166 = "Transform (Modify) Bone - Bone: dyn_l_braid_04"(ComponentPose=@n165, Alpha=1, Translation="0,0,0", Rotation="0.000,0.000,8.135", Scale="0.9,1,1") {
    rotationMode=BMM_Additive
    translationMode=BMM_Additive
    alphaBoolBlend {blendOption=Linear}
    boneToModify {boneName=dyn_l_braid_04}
    rotation {roll=8.135}
    scale {x=0.9}
   }
   n167 = "Transform (Modify) Bone - Bone: dyn_l_braid_06"!(ComponentPose=@n166, Alpha:float=1, Translation="0,0,0", Rotation="0.000,0.000,6.167", Scale="0.9,1,1") -> (Pose) {
    rotationMode=BMM_Additive
    translationMode=BMM_Additive
    alphaBoolBlend {blendOption=Linear}
    boneToModify {boneName=dyn_l_braid_06}
    rotation {roll=6.167}
    scale {x=0.9}
   }
   n168 = "Transform (Modify) Bone - Bone: dyn_l_braid_08"!(ComponentPose=@n167, Alpha:float=1, Translation="0,0,0", Rotation="0,0,0", Scale="0.9,1,1") -> (Pose) {
    rotationMode=BMM_Additive
    translationMode=BMM_Additive
    alphaBoolBlend {blendOption=Linear}
    boneToModify {boneName=dyn_l_braid_08}
    scale {x=0.9}
   }
   n169 = "Transform (Modify) Bone - Bone: dyn_l_braid_010"!(ComponentPose=@n168, Alpha:float=1, Translation="0,0,0", Rotation="0,0,0", Scale="0.9,1,1") -> (Pose) {
    rotationMode=BMM_Additive
    translationMode=BMM_Additive
    alphaBoolBlend {blendOption=Linear}
    boneToModify {boneName=dyn_l_braid_010}
    scale {x=0.9}
   }
   n170 = Get GravityOverride()
   n171 = Get simulateBase()
   n172 = Anim Dynamics - Bone: dyn_l_braid_02(ComponentPose=@n169, Alpha=@n171, GravityOverride=@n170, ExternalForce="0,0,0") {
    angularDampingOverride=1
    angularSpringConstant=2500
    bAngularSpring=true
    bGravityOverrideInSimSpace=true
    bOverrideAngularDamping=true
    bUseSphericalLimits=true
    gravityScale=0.5
    simulationSpace=RootRelative
    alphaBoolBlend {blendInTime=1, blendOption=Linear, blendOutTime=1}
    alphaScaleBiasClamp {bInterpResult=true, scale=0.75}
    boundBone {boneName=dyn_l_braid_02}
    componentAppliedLinearAccClamp=<5000 5000 5000>
    componentLinearAccScale {z=0.7}
    componentLinearVelScale {z=0.7}
    physicsBodyDefinitions [{
      collisionType=CustomSphere
      sphereCollisionRadius=1
      boundBone {boneName=dyn_l_braid_02}
      boxExtents=<14 28 28>
      constraintSetup {angularLimitsMax=<10 30 20>, angularLimitsMin=<-10 -60 -20>, angularTarget=<0 0 0>}
      localJointOffset=<5 0 0>
     }]
    sphericalLimits [
     {limitRadius=65.9176, drivingBone {boneName=spine_05}, sphereLocalOffset=<-12.8628 55 0>}
     {limitRadius=50, drivingBone {boneName=pelvis}, sphereLocalOffset=<-2.6826 40 0>}
     {limitRadius=28, drivingBone {boneName=spine_02}, sphereLocalOffset=<2.6549 15 0>}
     {limitRadius=1, drivingBone {boneName=pelvis}, sphereLocalOffset=<-30 0 0>}
    ]
   }
   n173 = Anim Dynamics - Bone: dyn_l_braid_03(ComponentPose=@n172, Alpha=@n171, GravityOverride=@n170, ExternalForce="0,0,0") {
    angularDampingOverride=1
    angularSpringConstant=2500
    bAngularSpring=true
    bGravityOverrideInSimSpace=true
    bOverrideAngularDamping=true
    bUseSphericalLimits=true
    gravityScale=0.5
    simulationSpace=RootRelative
    alphaBoolBlend {blendInTime=1, blendOption=Linear, blendOutTime=1}
    alphaScaleBiasClamp {bInterpResult=true, scale=0.75}
    boundBone {boneName=dyn_l_braid_03}
    componentAppliedLinearAccClamp=<5000 5000 5000>
    componentLinearAccScale {z=0.7}
    componentLinearVelScale {z=0.7}
    physicsBodyDefinitions [{
      collisionType=CustomSphere
      sphereCollisionRadius=1
      boundBone {boneName=dyn_l_braid_03}
      boxExtents=<12 22 22>
      constraintSetup {angularLimitsMax=<10 30 20>, angularLimitsMin=<-10 -60 -20>, angularTarget=<0 0 0>}
      localJointOffset=<10 0 0>
     }]
    sphericalLimits [
     {limitRadius=65.9176, drivingBone {boneName=spine_05}, sphereLocalOffset=<-12.8628 55 0>}
     {limitRadius=50, drivingBone {boneName=pelvis}, sphereLocalOffset=<-2.6826 40 0>}
     {limitRadius=28, drivingBone {boneName=spine_02}, sphereLocalOffset=<2.6549 15 0>}
     {limitRadius=1, drivingBone {boneName=pelvis}, sphereLocalOffset=<-30 0 0>}
    ]
   }
   n174 = Anim Dynamics - Bone: dyn_r_braid_02(ComponentPose=@n173, Alpha=@n171, GravityOverride=@n170, ExternalForce="0,0,0") {
    angularDampingOverride=1
    angularSpringConstant=2500
    bAngularSpring=true
    bGravityOverrideInSimSpace=true
    bOverrideAngularDamping=true
    bUseSphericalLimits=true
    gravityScale=0.5
    simulationSpace=RootRelative
    alphaBoolBlend {blendInTime=1, blendOption=Linear, blendOutTime=1}
    alphaScaleBiasClamp {bInterpResult=true, scale=0.8}
    boundBone {boneName=dyn_r_braid_02}
    componentAppliedLinearAccClamp=<5000 5000 5000>
    componentLinearAccScale {z=0.7}
    componentLinearVelScale {z=0.7}
    physicsBodyDefinitions [{
      collisionType=CustomSphere
      sphereCollisionRadius=1
      boundBone {boneName=dyn_r_braid_02}
      boxExtents=<14 28 28>
      constraintSetup {angularLimitsMax=<10 30 20>, angularLimitsMin=<-10 -60 -20>, angularTarget=<0 0 0>}
      localJointOffset=<5 -0.5 0>
     }]
    sphericalLimits [
     {limitRadius=65.9176, drivingBone {boneName=spine_05}, sphereLocalOffset=<-12.8628 55 0>}
     {limitRadius=50, drivingBone {boneName=pelvis}, sphereLocalOffset=<-2.6826 40 0>}
     {limitRadius=28, drivingBone {boneName=spine_02}, sphereLocalOffset=<2.6549 15 0>}
     {limitRadius=1, drivingBone {boneName=pelvis}, sphereLocalOffset=<-30 0 0>}
    ]
   }
   n175 = Anim Dynamics - Bone: dyn_r_braid_03(ComponentPose=@n174, Alpha=@n171, GravityOverride=@n170, ExternalForce="0,0,0") {
    angularDampingOverride=1
    angularSpringConstant=2500
    bAngularSpring=true
    bGravityOverrideInSimSpace=true
    bOverrideAngularDamping=true
    bUseSphericalLimits=true
    gravityScale=0.5
    simulationSpace=RootRelative
    alphaBoolBlend {blendInTime=1, blendOption=Linear, blendOutTime=1}
    alphaScaleBiasClamp {bInterpResult=true, scale=0.8}
    boundBone {boneName=dyn_r_braid_03}
    componentAppliedLinearAccClamp=<5000 5000 5000>
    componentLinearAccScale {z=0.7}
    componentLinearVelScale {z=0.7}
    physicsBodyDefinitions [{
      collisionType=CustomSphere
      sphereCollisionRadius=1
      boundBone {boneName=dyn_r_braid_03}
      boxExtents=<14 28 28>
      constraintSetup {angularLimitsMax=<10 30 20>, angularLimitsMin=<-10 -60 -20>, angularTarget=<0 0 0>}
      localJointOffset=<10 -0.5 0>
     }]
    sphericalLimits [
     {limitRadius=65.9176, drivingBone {boneName=spine_05}, sphereLocalOffset=<-12.8628 55 0>}
     {limitRadius=50, drivingBone {boneName=pelvis}, sphereLocalOffset=<-2.6826 40 0>}
     {limitRadius=28, drivingBone {boneName=spine_02}, sphereLocalOffset=<2.6549 15 0>}
     {limitRadius=1, drivingBone {boneName=pelvis}, sphereLocalOffset=<-30 0 0>}
    ]
   }
   n176 = Anim Dynamics - Bone: dyn_l_braid_04(ComponentPose=@n175, Alpha=1, GravityOverride=@n170, ExternalForce="0,0,0") {
    angularDampingOverride=1
    angularSpringConstant=2500
    bAngularSpring=true
    bGravityOverrideInSimSpace=true
    bOverrideAngularDamping=true
    bUseSphericalLimits=true
    gravityScale=0.5
    simulationSpace=RootRelative
    alphaBoolBlend {blendOption=Linear}
    alphaScaleBiasClamp {scale=0.5}
    boundBone {boneName=dyn_l_braid_04}
    componentAppliedLinearAccClamp=<5000 5000 5000>
    componentLinearAccScale {z=0.7}
    componentLinearVelScale {z=0.7}
    physicsBodyDefinitions [{
      collisionType=CustomSphere
      sphereCollisionRadius=1
      boundBone {boneName=dyn_l_braid_04}
      boxExtents=<12 22 22>
      constraintSetup {angularLimitsMax=<20 45 90>, angularLimitsMin=<-20 -45 -90>, angularTarget=<0 0 0>}
      localJointOffset=<10 0 0>
     }]
    sphericalLimits [
     {limitRadius=65.9176, drivingBone {boneName=spine_05}, sphereLocalOffset=<-12.8628 55 0>}
     {limitRadius=50, drivingBone {boneName=pelvis}, sphereLocalOffset=<-2.6826 40 0>}
     {limitRadius=28, drivingBone {boneName=spine_02}, sphereLocalOffset=<2.6549 15 0>}
     {limitRadius=1, drivingBone {boneName=pelvis}, sphereLocalOffset=<-30 0 0>}
    ]
   }
   n177 = Anim Dynamics - Bone: dyn_l_braid_05(ComponentPose=@n176, Alpha=1, GravityOverride=@n170, ExternalForce="0,0,0") {
    angularDampingOverride=1
    angularSpringConstant=2500
    bAngularSpring=true
    bGravityOverrideInSimSpace=true
    bOverrideAngularDamping=true
    bUseSphericalLimits=true
    gravityScale=0.5
    simulationSpace=RootRelative
    alphaBoolBlend {blendOption=Linear}
    alphaScaleBiasClamp {scale=0.5}
    boundBone {boneName=dyn_l_braid_05}
    componentAppliedLinearAccClamp=<5000 5000 5000>
    componentLinearAccScale {z=0.7}
    componentLinearVelScale {z=0.7}
    physicsBodyDefinitions [{
      collisionType=CustomSphere
      sphereCollisionRadius=1
      boundBone {boneName=dyn_l_braid_05}
      boxExtents=<12 22 22>
      constraintSetup {angularLimitsMax=<20 45 90>, angularLimitsMin=<-20 -45 -90>, angularTarget=<0 0 0>}
      localJointOffset=<10 0 0>
     }]
    sphericalLimits [
     {limitRadius=65.9176, drivingBone {boneName=spine_05}, sphereLocalOffset=<-12.8628 55 0>}
     {limitRadius=50, drivingBone {boneName=pelvis}, sphereLocalOffset=<-2.6826 35 0>}
     {limitRadius=28, drivingBone {boneName=spine_02}, sphereLocalOffset=<2.6549 15 0>}
     {limitRadius=1, drivingBone {boneName=pelvis}, sphereLocalOffset=<-30 0 0>}
    ]
   }
   n178 = Anim Dynamics - Bone: dyn_l_braid_06(ComponentPose=@n177, Alpha=1, GravityOverride=@n170, ExternalForce="0,0,0") {
    angularDampingOverride=1
    angularSpringConstant=2500
    bAngularSpring=true
    bGravityOverrideInSimSpace=true
    bOverrideAngularDamping=true
    bUseSphericalLimits=true
    gravityScale=0.5
    simulationSpace=RootRelative
    alphaBoolBlend {blendOption=Linear}
    alphaScaleBiasClamp {scale=0.35}
    boundBone {boneName=dyn_l_braid_06}
    componentAppliedLinearAccClamp=<5000 5000 5000>
    componentLinearAccScale {z=0.7}
    componentLinearVelScale {z=0.7}
    physicsBodyDefinitions [{
      collisionType=CustomSphere
      sphereCollisionRadius=1
      boundBone {boneName=dyn_l_braid_06}
      boxExtents=<12 22 22>
      constraintSetup {angularLimitsMax=<20 45 90>, angularLimitsMin=<-20 -45 -90>, angularTarget=<0 0 0>}
      localJointOffset=<10 0 0>
     }]
    planarLimits [{drivingBone {boneName=root}, planeTransform {location=<0 0 3>, rotation=<rot 0 0 0>}}]
    sphericalLimits [
     {limitRadius=65.9176, drivingBone {boneName=spine_05}, sphereLocalOffset=<-12.8628 55 0>}
     {limitRadius=50, drivingBone {boneName=pelvis}, sphereLocalOffset=<-2.6826 35 0>}
     {limitRadius=28, drivingBone {boneName=spine_02}, sphereLocalOffset=<2.6549 15 0>}
     {limitRadius=1, drivingBone {boneName=pelvis}, sphereLocalOffset=<-30 0 0>}
    ]
   }
   n179 = Anim Dynamics - Bone: dyn_l_braid_07(ComponentPose=@n178, Alpha=1, GravityOverride=@n170, ExternalForce="0,0,0") {
    angularDampingOverride=1
    angularSpringConstant=2500
    bAngularSpring=true
    bGravityOverrideInSimSpace=true
    bOverrideAngularDamping=true
    bUseSphericalLimits=true
    gravityScale=0.5
    simulationSpace=RootRelative
    alphaBoolBlend {blendOption=Linear}
    alphaScaleBiasClamp {scale=0.4}
    boundBone {boneName=dyn_l_braid_07}
    componentAppliedLinearAccClamp=<5000 5000 5000>
    componentLinearAccScale {z=0.7}
    componentLinearVelScale {z=0.7}
    physicsBodyDefinitions [{
      collisionType=CustomSphere
      sphereCollisionRadius=1
      boundBone {boneName=dyn_l_braid_07}
      boxExtents=<12 22 22>
      constraintSetup {angularLimitsMax=<20 90 90>, angularLimitsMin=<-20 -90 -90>, angularTarget=<0 0 0>}
      localJointOffset=<10 0 0>
     }]
    planarLimits [{drivingBone {boneName=root}, planeTransform {location=<0 0 3>, rotation=<rot 0 0 0>}}]
    sphericalLimits [
     {limitRadius=65.9176, drivingBone {boneName=spine_05}, sphereLocalOffset=<-12.8628 65 0>}
     {limitRadius=50, drivingBone {boneName=pelvis}, sphereLocalOffset=<-2.6826 35 0>}
     {limitRadius=28, drivingBone {boneName=spine_02}, sphereLocalOffset=<2.6549 17 0>}
     {limitRadius=1, drivingBone {boneName=pelvis}, sphereLocalOffset=<-30 0 0>}
    ]
   }
   n180 = Anim Dynamics - Bone: dyn_l_braid_08(ComponentPose=@n179, Alpha=1, GravityOverride=@n170, ExternalForce="0,0,0") {
    angularDampingOverride=1
    angularSpringConstant=2500
    bAngularSpring=true
    bGravityOverrideInSimSpace=true
    bOverrideAngularDamping=true
    bUseSphericalLimits=true
    gravityScale=0.5
    simulationSpace=RootRelative
    alphaBoolBlend {blendOption=Linear}
    alphaScaleBiasClamp {scale=0.5}
    boundBone {boneName=dyn_l_braid_08}
    componentAppliedLinearAccClamp=<5000 5000 5000>
    componentLinearAccScale {z=0.7}
    componentLinearVelScale {z=0.7}
    physicsBodyDefinitions [{
      collisionType=CustomSphere
      sphereCollisionRadius=1
      boundBone {boneName=dyn_l_braid_08}
      boxExtents=<12 22 22>
      constraintSetup {angularLimitsMax=<20 90 90>, angularLimitsMin=<-20 -90 -90>, angularTarget=<0 0 0>}
      localJointOffset=<10 0 0>
     }]
    planarLimits [{drivingBone {boneName=root}, planeTransform {location=<0 0 3>, rotation=<rot 0 0 0>}}]
    sphericalLimits [
     {limitRadius=65.9176, drivingBone {boneName=spine_05}, sphereLocalOffset=<-12.8628 65 0>}
     {limitRadius=50, drivingBone {boneName=pelvis}, sphereLocalOffset=<-2.6826 40 0>}
     {limitRadius=28, drivingBone {boneName=spine_02}, sphereLocalOffset=<2.6549 17 0>}
     {limitRadius=1, drivingBone {boneName=pelvis}, sphereLocalOffset=<-30 0 0>}
    ]
   }
   n181 = Anim Dynamics - Bone: dyn_l_braid_09(ComponentPose=@n180, Alpha=1, GravityOverride=@n170, ExternalForce="0,0,0") {
    angularDampingOverride=1
    angularSpringConstant=2500
    bAngularSpring=true
    bGravityOverrideInSimSpace=true
    bOverrideAngularDamping=true
    bUseSphericalLimits=true
    gravityScale=0.5
    simulationSpace=RootRelative
    alphaBoolBlend {blendOption=Linear}
    alphaScaleBiasClamp {scale=0.5}
    boundBone {boneName=dyn_l_braid_09}
    componentAppliedLinearAccClamp=<5000 5000 5000>
    componentLinearAccScale {z=0.7}
    componentLinearVelScale {z=0.7}
    physicsBodyDefinitions [{
      collisionType=CustomSphere
      sphereCollisionRadius=1
      boundBone {boneName=dyn_l_braid_09}
      boxExtents=<12 22 22>
      constraintSetup {angularLimitsMax=<20 90 90>, angularLimitsMin=<-20 -90 -90>, angularTarget=<0 0 0>}
      localJointOffset=<10 0 0>
     }]
    planarLimits [{drivingBone {boneName=root}, planeTransform {location=<0 0 3>, rotation=<rot 0 0 0>}}]
    sphericalLimits [
     {limitRadius=65.9176, drivingBone {boneName=spine_05}, sphereLocalOffset=<-12.8628 65 0>}
     {limitRadius=50, drivingBone {boneName=pelvis}, sphereLocalOffset=<-2.6826 40 0>}
     {limitRadius=28, drivingBone {boneName=spine_02}, sphereLocalOffset=<2.6549 17 0>}
     {limitRadius=1, drivingBone {boneName=pelvis}, sphereLocalOffset=<-30 0 0>}
    ]
   }
   n182 = Anim Dynamics - Bone: dyn_l_braid_010(ComponentPose=@n181, Alpha=1, GravityOverride=@n170, ExternalForce="0,0,0") {
    angularDampingOverride=1
    angularSpringConstant=2500
    bAngularSpring=true
    bGravityOverrideInSimSpace=true
    bOverrideAngularDamping=true
    bUseSphericalLimits=true
    gravityScale=0.5
    simulationSpace=RootRelative
    alphaBoolBlend {blendOption=Linear}
    alphaScaleBiasClamp {scale=0.5}
    boundBone {boneName=dyn_l_braid_010}
    componentAppliedLinearAccClamp=<5000 5000 5000>
    componentLinearAccScale {z=0.7}
    componentLinearVelScale {z=0.7}
    physicsBodyDefinitions [{
      collisionType=CustomSphere
      sphereCollisionRadius=1
      boundBone {boneName=dyn_l_braid_010}
      boxExtents=<12 22 22>
      constraintSetup {angularLimitsMax=<20 90 90>, angularLimitsMin=<-20 -90 -90>, angularTarget=<0 0 0>}
      localJointOffset=<10 0 0>
     }]
    planarLimits [{drivingBone {boneName=root}, planeTransform {location=<0 0 3>, rotation=<rot 0 0 0>}}]
    sphericalLimits [
     {limitRadius=65.9176, drivingBone {boneName=spine_05}, sphereLocalOffset=<-12.8628 65 0>}
     {limitRadius=50, drivingBone {boneName=pelvis}, sphereLocalOffset=<-2.6826 40 0>}
     {limitRadius=28, drivingBone {boneName=spine_02}, sphereLocalOffset=<2.6549 17 0>}
     {limitRadius=1, drivingBone {boneName=pelvis}, sphereLocalOffset=<-30 0 0>}
     {limitRadius=0, drivingBone {boneName=None}, sphereLocalOffset=<0 0 0>}
    ]
   }
   n183 = Anim Dynamics - Bone: dyn_l_braid_011(ComponentPose=@n182, Alpha=1, GravityOverride=@n170, ExternalForce="0,0,0") {
    angularDampingOverride=1
    angularSpringConstant=2500
    bAngularSpring=true
    bGravityOverrideInSimSpace=true
    bOverrideAngularDamping=true
    bUseSphericalLimits=true
    gravityScale=0.5
    simulationSpace=RootRelative
    alphaBoolBlend {blendOption=Linear}
    alphaScaleBiasClamp {scale=0.5}
    boundBone {boneName=dyn_l_braid_011}
    componentAppliedLinearAccClamp=<5000 5000 5000>
    componentLinearAccScale {z=0.7}
    componentLinearVelScale {z=0.7}
    physicsBodyDefinitions [{
      collisionType=CustomSphere
      sphereCollisionRadius=1
      boundBone {boneName=dyn_l_braid_011}
      boxExtents=<12 22 22>
      constraintSetup {angularLimitsMax=<20 90 90>, angularLimitsMin=<-20 -90 -90>, angularTarget=<0 0 0>}
      localJointOffset=<10 0 0>
     }]
    planarLimits [{drivingBone {boneName=root}, planeTransform {location=<0 0 3>, rotation=<rot 0 0 0>}}]
    sphericalLimits [
     {limitRadius=65.9176, drivingBone {boneName=spine_05}, sphereLocalOffset=<-12.8628 65 0>}
     {limitRadius=50, drivingBone {boneName=pelvis}, sphereLocalOffset=<-2.6826 40 0>}
     {limitRadius=28, drivingBone {boneName=spine_02}, sphereLocalOffset=<2.6549 17 0>}
     {limitRadius=1, drivingBone {boneName=pelvis}, sphereLocalOffset=<-30 0 0>}
     {limitRadius=0, drivingBone {boneName=None}, sphereLocalOffset=<0 0 0>}
    ]
   }
   n184 = Anim Dynamics - Bone: dyn_r_braid_04(ComponentPose=@n183, Alpha=1, GravityOverride=@n170, ExternalForce="0,0,0") {
    angularDampingOverride=1
    angularSpringConstant=2500
    bAngularSpring=true
    bGravityOverrideInSimSpace=true
    bOverrideAngularDamping=true
    bUseSphericalLimits=true
    gravityScale=0.5
    simulationSpace=RootRelative
    alphaBoolBlend {blendOption=Linear}
    alphaScaleBiasClamp {scale=0.55}
    boundBone {boneName=dyn_r_braid_04}
    componentAppliedLinearAccClamp=<5000 5000 5000>
    componentLinearAccScale {z=0.7}
    componentLinearVelScale {z=0.7}
    physicsBodyDefinitions [{
      collisionType=CustomSphere
      sphereCollisionRadius=1
      boundBone {boneName=dyn_r_braid_04}
      boxExtents=<14 28 28>
      constraintSetup {angularLimitsMax=<20 45 90>, angularLimitsMin=<-20 -45 -90>, angularTarget=<0 0 0>}
      localJointOffset=<10 -0.5 0>
     }]
    sphericalLimits [
     {limitRadius=65.9176, drivingBone {boneName=spine_05}, sphereLocalOffset=<-12.8628 55 0>}
     {limitRadius=50, drivingBone {boneName=pelvis}, sphereLocalOffset=<-2.6826 40 0>}
     {limitRadius=28, drivingBone {boneName=spine_02}, sphereLocalOffset=<2.6549 15 0>}
     {limitRadius=1, drivingBone {boneName=pelvis}, sphereLocalOffset=<-30 0 0>}
    ]
   }
   n185 = Anim Dynamics - Bone: dyn_r_braid_05(ComponentPose=@n184, Alpha=1, GravityOverride=@n170, ExternalForce="0,0,0") {
    angularDampingOverride=1
    angularSpringConstant=2500
    bAngularSpring=true
    bGravityOverrideInSimSpace=true
    bOverrideAngularDamping=true
    bUseSphericalLimits=true
    gravityScale=0.5
    simulationSpace=RootRelative
    alphaBoolBlend {blendOption=Linear}
    alphaScaleBiasClamp {scale=0.55}
    boundBone {boneName=dyn_r_braid_05}
    componentAppliedLinearAccClamp=<5000 5000 5000>
    componentLinearAccScale {z=0.7}
    componentLinearVelScale {z=0.7}
    physicsBodyDefinitions [{
      collisionType=CustomSphere
      sphereCollisionRadius=1
      boundBone {boneName=dyn_r_braid_05}
      boxExtents=<14 28 28>
      constraintSetup {angularLimitsMax=<20 45 90>, angularLimitsMin=<-20 -45 -90>, angularTarget=<0 0 0>}
      localJointOffset=<10 -0.5 0>
     }]
    sphericalLimits [
     {limitRadius=65.9176, drivingBone {boneName=spine_05}, sphereLocalOffset=<-12.8628 55 0>}
     {limitRadius=50, drivingBone {boneName=pelvis}, sphereLocalOffset=<-2.6826 40 0>}
     {limitRadius=28, drivingBone {boneName=spine_02}, sphereLocalOffset=<2.6549 15 0>}
     {limitRadius=1, drivingBone {boneName=pelvis}, sphereLocalOffset=<-30 0 0>}
    ]
   }
   n186 = Anim Dynamics - Bone: dyn_r_braid_06(ComponentPose=@n185, Alpha=1, GravityOverride=@n170, ExternalForce="0,0,0") {
    angularDampingOverride=1
    angularSpringConstant=2500
    bAngularSpring=true
    bGravityOverrideInSimSpace=true
    bOverrideAngularDamping=true
    bUseSphericalLimits=true
    gravityScale=0.5
    simulationSpace=RootRelative
    alphaBoolBlend {blendOption=Linear}
    alphaScaleBiasClamp {scale=0.4}
    boundBone {boneName=dyn_r_braid_06}
    componentAppliedLinearAccClamp=<5000 5000 5000>
    componentLinearAccScale {z=0.7}
    componentLinearVelScale {z=0.7}
    physicsBodyDefinitions [{
      collisionType=CustomSphere
      sphereCollisionRadius=1
      boundBone {boneName=dyn_r_braid_06}
      boxExtents=<14 28 28>
      constraintSetup {angularLimitsMax=<20 45 90>, angularLimitsMin=<-20 -45 -90>, angularTarget=<0 0 0>}
      localJointOffset=<10 -0.5 0>
     }]
    planarLimits [{drivingBone {boneName=root}, planeTransform {location=<0 0 5>, rotation=<rot 0 0 0>}}]
    sphericalLimits [
     {limitRadius=65.9176, drivingBone {boneName=spine_05}, sphereLocalOffset=<-12.8628 55 0>}
     {limitRadius=50, drivingBone {boneName=pelvis}, sphereLocalOffset=<-2.6826 35 0>}
     {limitRadius=28, drivingBone {boneName=spine_02}, sphereLocalOffset=<2.6549 15 0>}
     {limitRadius=1, drivingBone {boneName=pelvis}, sphereLocalOffset=<-30 0 0>}
    ]
   }
   n187 = Anim Dynamics - Bone: dyn_r_braid_07(ComponentPose=@n186, Alpha=1, GravityOverride="0,0,0", ExternalForce="0,0,0") {
    angularDampingOverride=1
    angularSpringConstant=2500
    bAngularSpring=true
    bGravityOverrideInSimSpace=true
    bOverrideAngularDamping=true
    bUseSphericalLimits=true
    gravityScale=0.5
    simulationSpace=RootRelative
    alphaBoolBlend {blendOption=Linear}
    alphaScaleBiasClamp {scale=0.4}
    boundBone {boneName=dyn_r_braid_07}
    componentAppliedLinearAccClamp=<5000 5000 5000>
    componentLinearAccScale {z=0.7}
    componentLinearVelScale {z=0.7}
    physicsBodyDefinitions [{
      collisionType=CustomSphere
      sphereCollisionRadius=1
      boundBone {boneName=dyn_r_braid_07}
      boxExtents=<12 22 22>
      constraintSetup {angularLimitsMax=<20 90 90>, angularLimitsMin=<-20 -90 -90>, angularTarget=<0 0 0>}
      localJointOffset=<10 0 0>
     }]
    planarLimits [{drivingBone {boneName=root}, planeTransform {location=<0 0 3>, rotation=<rot 0 0 0>}}]
    sphericalLimits [
     {limitRadius=65.9176, drivingBone {boneName=spine_05}, sphereLocalOffset=<-12.8628 65 0>}
     {limitRadius=50, drivingBone {boneName=pelvis}, sphereLocalOffset=<-2.6826 35 0>}
     {limitRadius=28, drivingBone {boneName=spine_02}, sphereLocalOffset=<2.6549 17 0>}
     {limitRadius=1, drivingBone {boneName=pelvis}, sphereLocalOffset=<-30 0 0>}
    ]
   }
   n188 = Anim Dynamics - Bone: dyn_r_braid_08(ComponentPose=@n187, Alpha=1, GravityOverride="0,0,0", ExternalForce="0,0,0") {
    angularDampingOverride=1
    angularSpringConstant=2500
    bAngularSpring=true
    bGravityOverrideInSimSpace=true
    bOverrideAngularDamping=true
    bUseSphericalLimits=true
    gravityScale=0.5
    simulationSpace=RootRelative
    alphaBoolBlend {blendOption=Linear}
    alphaScaleBiasClamp {scale=0.5}
    boundBone {boneName=dyn_r_braid_08}
    componentAppliedLinearAccClamp=<5000 5000 5000>
    componentLinearAccScale {z=0.7}
    componentLinearVelScale {z=0.7}
    physicsBodyDefinitions [{
      collisionType=CustomSphere
      sphereCollisionRadius=1
      boundBone {boneName=dyn_r_braid_08}
      boxExtents=<12 22 22>
      constraintSetup {angularLimitsMax=<20 90 90>, angularLimitsMin=<-20 -90 -90>, angularTarget=<0 0 0>}
      localJointOffset=<10 0 0>
     }]
    planarLimits [{drivingBone {boneName=root}, planeTransform {location=<0 0 3>, rotation=<rot 0 0 0>}}]
    sphericalLimits [
     {limitRadius=65.9176, drivingBone {boneName=spine_05}, sphereLocalOffset=<-12.8628 65 0>}
     {limitRadius=50, drivingBone {boneName=pelvis}, sphereLocalOffset=<-2.6826 40 0>}
     {limitRadius=28, drivingBone {boneName=spine_02}, sphereLocalOffset=<2.6549 17 0>}
     {limitRadius=1, drivingBone {boneName=pelvis}, sphereLocalOffset=<-30 0 0>}
    ]
   }
   n189 = Anim Dynamics - Bone: dyn_r_braid_09(ComponentPose=@n188, Alpha=1, GravityOverride="0,0,0", ExternalForce="0,0,0") {
    angularDampingOverride=1
    angularSpringConstant=2500
    bAngularSpring=true
    bGravityOverrideInSimSpace=true
    bOverrideAngularDamping=true
    bUseSphericalLimits=true
    gravityScale=0.5
    simulationSpace=RootRelative
    alphaBoolBlend {blendOption=Linear}
    alphaScaleBiasClamp {scale=0.5}
    boundBone {boneName=dyn_r_braid_09}
    componentAppliedLinearAccClamp=<5000 5000 5000>
    componentLinearAccScale {z=0.7}
    componentLinearVelScale {z=0.7}
    physicsBodyDefinitions [{
      collisionType=CustomSphere
      sphereCollisionRadius=1
      boundBone {boneName=dyn_r_braid_09}
      boxExtents=<12 22 22>
      constraintSetup {angularLimitsMax=<20 90 90>, angularLimitsMin=<-20 -90 -90>, angularTarget=<0 0 0>}
      localJointOffset=<1 0 0>
     }]
    planarLimits [{drivingBone {boneName=root}, planeTransform {location=<0 0 3>, rotation=<rot 0 0 0>}}]
    sphericalLimits [
     {limitRadius=65.9176, drivingBone {boneName=spine_05}, sphereLocalOffset=<-12.8628 65 0>}
     {limitRadius=50, drivingBone {boneName=pelvis}, sphereLocalOffset=<-2.6826 40 0>}
     {limitRadius=28, drivingBone {boneName=spine_02}, sphereLocalOffset=<2.6549 17 0>}
     {limitRadius=1, drivingBone {boneName=pelvis}, sphereLocalOffset=<-30 0 0>}
    ]
   }
   n190 = Anim Dynamics - Bone: dyn_r_braid_010(ComponentPose=@n189, Alpha=1, GravityOverride="0,0,0", ExternalForce="0,0,0") {
    angularDampingOverride=1
    angularSpringConstant=2500
    bAngularSpring=true
    bGravityOverrideInSimSpace=true
    bOverrideAngularDamping=true
    bUseSphericalLimits=true
    gravityScale=0.5
    simulationSpace=RootRelative
    alphaBoolBlend {blendOption=Linear}
    alphaScaleBiasClamp {scale=0.5}
    boundBone {boneName=dyn_r_braid_010}
    componentAppliedLinearAccClamp=<5000 5000 5000>
    componentLinearAccScale {z=0.7}
    componentLinearVelScale {z=0.7}
    physicsBodyDefinitions [{
      collisionType=CustomSphere
      sphereCollisionRadius=1
      boundBone {boneName=dyn_r_braid_010}
      boxExtents=<12 22 22>
      constraintSetup {angularLimitsMax=<20 45 15>, angularLimitsMin=<-20 -45 -15>, angularTarget=<0 0 0>}
      localJointOffset=<5 0 0>
     }]
    planarLimits [{drivingBone {boneName=root}, planeTransform {location=<0 0 5>, rotation=<rot 0 0 0>}}]
    sphericalLimits [
     {limitRadius=65.9176, drivingBone {boneName=spine_05}, sphereLocalOffset=<-12.8628 65 0>}
     {limitRadius=50, drivingBone {boneName=pelvis}, sphereLocalOffset=<-2.6826 40 0>}
     {limitRadius=28, drivingBone {boneName=spine_02}, sphereLocalOffset=<2.6549 17 0>}
     {limitRadius=1, drivingBone {boneName=pelvis}, sphereLocalOffset=<-30 0 0>}
    ]
   }
   n191 = Anim Dynamics - Bone: dyn_r_braid_011(ComponentPose=@n190, Alpha=1, GravityOverride="0,0,0", ExternalForce="0,0,0") {
    angularDampingOverride=1
    angularSpringConstant=2500
    bAngularSpring=true
    bGravityOverrideInSimSpace=true
    bOverrideAngularDamping=true
    bUseSphericalLimits=true
    gravityScale=0.5
    simulationSpace=RootRelative
    alphaBoolBlend {blendOption=Linear}
    alphaScaleBiasClamp {scale=0.5}
    boundBone {boneName=dyn_r_braid_011}
    componentAppliedLinearAccClamp=<5000 5000 5000>
    componentLinearAccScale {z=0.7}
    componentLinearVelScale {z=0.7}
    physicsBodyDefinitions [{
      collisionType=CustomSphere
      sphereCollisionRadius=1
      boundBone {boneName=dyn_r_braid_011}
      boxExtents=<12 22 22>
      constraintSetup {angularLimitsMax=<20 45 15>, angularLimitsMin=<-20 -45 -15>, angularTarget=<0 0 0>}
      localJointOffset=<5 0 0>
     }]
    planarLimits [{drivingBone {boneName=root}, planeTransform {location=<0 0 5>, rotation=<rot 0 0 0>}}]
    sphericalLimits [
     {limitRadius=65.9176, drivingBone {boneName=spine_05}, sphereLocalOffset=<-12.8628 65 0>}
     {limitRadius=50, drivingBone {boneName=pelvis}, sphereLocalOffset=<-2.6826 40 0>}
     {limitRadius=28, drivingBone {boneName=spine_02}, sphereLocalOffset=<2.6549 17 0>}
     {limitRadius=1, drivingBone {boneName=pelvis}, sphereLocalOffset=<-30 0 0>}
     {limitRadius=0, drivingBone {boneName=None}, sphereLocalOffset=<0 0 0>}
    ]
   }
   n192 = Get kunaiAiming()
   n193 = "Transform (Modify) Bone - Bone: dyn_r_braid_02"(ComponentPose=@n191, Alpha=@n192, Translation="0,0,0", Rotation="15,15,0", Scale="1,1,1") {
    rotationMode=BMM_Additive
    rotationSpace=BCS_BoneSpace
    alphaBoolBlend {blendInTime=1, blendOption=Linear, blendOutTime=0.5}
    alphaScaleBiasClamp {bInterpResult=true, interpSpeedIncreasing=25}
    boneToModify {boneName=dyn_r_braid_02}
    rotation {pitch=15, yaw=15}
   }
   n194 = "Transform (Modify) Bone - Bone: dyn_r_braid_04"(ComponentPose=@n193, Alpha=@n192, Translation="0,0,0", Rotation="-15,-30,0.000", Scale="1,1,1") {
    rotationMode=BMM_Additive
    rotationSpace=BCS_BoneSpace
    alphaBoolBlend {blendInTime=1, blendOption=Linear, blendOutTime=0.5}
    alphaScaleBiasClamp {bInterpResult=true, interpSpeedIncreasing=25}
    boneToModify {boneName=dyn_r_braid_04}
    rotation {pitch=-15, yaw=-30}
   }
   n195 = "Transform (Modify) Bone - Bone: dyn_r_braid_05"(ComponentPose=@n194, Alpha=@n192, Translation="0,0,0", Rotation="10,0,0.000", Scale="1,1,1") {
    rotationMode=BMM_Additive
    rotationSpace=BCS_BoneSpace
    alphaBoolBlend {blendInTime=1, blendOption=Linear, blendOutTime=0.5}
    alphaScaleBiasClamp {bInterpResult=true, interpSpeedIncreasing=25}
    boneToModify {boneName=dyn_r_braid_05}
    rotation {pitch=10}
   }
   Save cached pose 'rigids disabled'(Pose=Component To Local(ComponentPose=@n195)) {cacheName=rigids disabled}
  ]}
 EventGraph {nodes [
   ResetDynamics(InTeleportType=TeleportPhysics)
   > ResetDynamics(InTeleportType=ResetPhysics)
   Event BlueprintUpdateAnimation()
   > n1 = Sequence!(execute) -> (then_0, then_1, then_2:Exec, then_3:Exec, then_4:Exec, then_5:Exec, then_6:Exec, then_7:Exec, then_8:Exec)
   > n2 = Gravity Override State Function DEPRECATED(CharacterPartAnimInstance=Self-Reference(), GameTimeInSeconds=GetGameTimeInSeconds(), GravityOverrideParams=Get Gravity Override Params())
   > Set GravityOverride(GravityOverride=@n2)
   > Set animDynamicsEnabledOrFrontend(animDynamicsEnabledOrFrontend:bool=OR Boolean(A=Get bAnimDynamicsEnabled() -> (bAnimDynamicsEnabled:bool), B=Get bIsInFrontEnd())) -> (Output_Get:bool)
   n3 = ComplexSimSpaceFunction DEPRECATED!(
    execute=@n1.then_1
    CharacterPartAnimInstance:Custom Character Part Anim Instance Object Reference=Self-Reference()
    InterpSpeedMax:double=9
    PawnSpeedMax:double=20
    RInterpErrorTolerance:double=1
    DeltaTime:double=0
    DebugDraw:bool=false
    sim_space_rotation_in:Rotator=Get sim_space_rotation() -> (sim_space_rotation:Rotator)
    inComponentSpaceIn:bool=Get InComponentSpace() -> (InComponentSpace:bool)
    inComponentSpaceLastFrameIn:bool=Get InComponentSpaceLastFrame() -> (InComponentSpaceLastFrame:bool)
    lockedOnSimSpaceIn:bool=Get LockedOnSimSpace() -> (LockedOnSimSpace:bool)
    InterpSpeedIn:double=Get InterpSpeed() -> (InterpSpeed:double)
   ) -> (
    then
    sim_space_rotation:Rotator
    inComponentSpaceOut:bool
    inComponentSpaceLastFrameOut:bool
    lockedOnSimSpaceOut:bool
    InterpSpeedOut:double
   )
   > Set sim_space_rotation(sim_space_rotation:Rotator=@n3) -> (Output_Get:Rotator)
   > Set InComponentSpace(InComponentSpace:bool=@n3.inComponentSpaceOut) -> (Output_Get:bool)
   > Set InComponentSpaceLastFrame(InComponentSpaceLastFrame:bool=@n3.inComponentSpaceLastFrameOut) -> (Output_Get:bool)
   > Set LockedOnSimSpace(LockedOnSimSpace:bool=@n3.lockedOnSimSpaceOut) -> (Output_Get:bool)
   > Set InterpSpeed(InterpSpeed:double=@n3.InterpSpeedOut) -> (Output_Get:double)
   n4 = CheckForBackbling()
   Set isWearingBackbling(execute=@n1.then_2, isWearingBackbling=OR Boolean(A=@n4, B=@n4.bIsBackblingLong))
   n5 = CheckForBackbling()
   Set isWearingBackbling(execute=@n1.then_3, isWearingBackbling=@n5)
   > Set isNotWearingBackbling(isNotWearingBackbling:bool=NOT Boolean(A=@n5)) -> (Output_Get:bool)
   > n6 = Branch(Condition=@n5)
   > Set WearingCape(WearingCape:bool=@n5.bIsBackblingACape) -> (Output_Get:bool)
   > Set simulateBase(simulateBase:double=0) -> (Output_Get:double)
   Set simulateBase(execute=@n6.else, simulateBase=1)
   Branch(execute=@n1.then_4, Condition=Get bIsCrouching())
   > Set simulateBase(simulateBase=0)
   n7 = Set Skydiving(
    execute=@n1.then_5
    Skydiving:bool=OR Boolean!(A=Get bIsSkyDiving(), B=Get bIsSkydiveFloating(), C:bool=Get bIsSkydiveDiveMode()) -> (ReturnValue)
   ) -> (Output_Get:bool)
   > Set CapeSkydive(CapeSkydive:bool=AND Boolean(A=@n7, B=Get WearingCape())) -> (Output_Get:bool)
   n8 = Branch(execute=@n1.then_6, Condition=Get bIsParachuteOpened())
   > Set SkyDiveDynLookAtPos(SkyDiveDynLookAtPos:Vector="0,-100,0") -> (Output_Get:Vector)
   Set SkyDiveDynLookAtPos(execute=@n8.else, SkyDiveDynLookAtPos="0,-60,100")
   n9 = Cast To FortPlayerAnimInstance(execute=@n1.then_8, Object=GetAnimInstance(self=GetOwnerSkeletalMeshComponent()))
   > Set kunaiAiming(kunaiAiming:double=0) -> (Output_Get:double)
   > n10 = Switch on EFortWeaponCoreAnimation!(execute, Selection:EFortWeaponCoreAnimation Enum=Get WeaponCoreAnim(self=@n9)) -> (
    Melee:Exec
    Pistol:Exec
    Shotgun:Exec
    PaperBlueprint:Exec
    Rifle:Exec
    MeleeOneHand:Exec
    MachinePistol:Exec
    RocketLauncher:Exec
    GrenadeLauncher:Exec
    GoingCommando:Exec
    AssaultRifle:Exec
    TacticalShotgun:Exec
    SniperRifle:Exec
    TrapPlacement:Exec
    ShoulderLauncher:Exec
    AbilityDecoTool:Exec
    Crossbow:Exec
    C4:Exec
    RemoteControl:Exec
    DualWield:Exec
    AR_BullPup:Exec
    AR_ForwardGrip:Exec
    MedPackPaddles:Exec
    SMG_P90:Exec
    AR_DrumGun:Exec
    Consumable_Small:Exec
    Consumable_Large:Exec
    Balloon:Exec
    MountedTurret:Exec
    CreativeTool:Exec
    ExplosiveBow:Exec
    AshtonIndigo:Exec
    AshtonChicago:Exec
    MeleeDualWield:Exec
    Unarmed:Exec
   )
   n11 = NOT Boolean(A=Get isWearingBackbling())
   n12 = Get bIsCrouching()
   n13 = Branch(execute=@n10.AshtonChicago, Condition=OR Boolean(A=Get isWearingBackbling(), B=Get WearingCape(), C=AND Boolean(A=@n11, B=@n12)))
   n14 = BreakTransform(InTransform=GetDeltaTransformFromRefPose(self=GetOwnerSkeletalMeshComponent(), BoneName=clavicle_l, BaseName=spine_05))
   > Set kunaiAiming(kunaiAiming=MapRangeClamped(Value=BreakRotator(InRot=@n14.Rotation).Pitch, InRangeA=0, InRangeB=14))
   Set kunaiAiming(execute=@n13.else, kunaiAiming=0)
   Event BlueprintInitializeAnimation()
   > Set Skel_Mesh(Skel_Mesh=GetOwnerSkeletalMeshComponent())
   Event OnCharacterCustomizationLoaded()
  ]}
]
variables [
 animDynamicsEnabledOrFrontend {type=boolean}
 CapeSkydive {type=boolean}
 Gravity Override Params {type=GravityOverrideParamsStruct}
 GravityOverride {type=Vector}
 InComponentSpace {type=boolean}
 InComponentSpaceLastFrame {type=boolean}
 InterpSpeed {type=number}
 isNotWearingBackbling {type=boolean}
 isWearingBackbling {type=boolean}
 isWearingCape {type=number}
 kunaiAiming {type=number}
 LockedOnSimSpace {type=boolean}
 sim_space_rotation {type=Rotator}
 simulateBase {type=number}
 Skel_Mesh {type=SkeletalMeshComponent}
 SkyDiveDynLookAtPos {type=Vector}
 Skydiving {type=boolean}
 Spine05_twist {type=number}
 WearingCape {type=boolean}
]
